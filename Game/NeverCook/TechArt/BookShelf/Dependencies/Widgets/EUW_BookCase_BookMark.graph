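# Editor utility widget for one BookShelf bookmark row: a colored, named asset shortcut with type icon and browse, edit and remove buttons.
# Init takes bookmark data and looks up the asset class in the Asset Registry to pick the icon.
# Clicking loads levels, runs editor utilities via inline Python or syncs to folders; Ctrl/Alt opens the asset editor.
# Config-mode, edit and remove clicks reach the owner through dispatchers.
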EditorUtilityWidgetBlueprint /Game/NeverCook/TechArt/BookShelf/Dependencies/Widgets/EUW_BookCase_BookMark
graphs [
 BindEditBookmark {nodes [BindEditBookmark!() -> (then, Item:EUW Book Case Book Mark Object Reference) {is=title K2Node_FunctionEntry}]}
 BindEvents {nodes [
   n1 = BindEvents!() -> (then, RemoveSelf:"Delegate (by ref)", BeginEditBookmark:"Delegate (by ref)", CommitEditBookmark:"Delegate (by ref)") {is=title K2Node_FunctionEntry}
   > Bind Event to Remove Bookmark(self:EUW Book Case Book Mark Object Reference, Delegate=@n1) {is=title K2Node_AddDelegate}
   > Bind Event to Load Button for Config(self:EUW Book Case Book Mark Object Reference, Delegate=@n1.BeginEditBookmark) {is=title K2Node_AddDelegate}
   > Bind Event to Bind Edit Bookmark(self:EUW Book Case Book Mark Object Reference, Delegate=@n1.CommitEditBookmark) {is=title K2Node_AddDelegate}
  ]}
 ChangeTextSize {nodes [
   n1 = ChangeTextSize!() -> (then, Size:float=~) {is=title K2Node_FunctionEntry}
   n2 = Get TX_BookmarkName() -> (TX_BookmarkName:Editor Utility Editable Text Object Reference)
   > n3 = Set font(font:Slate Font Info Structure=GetFont!(self:Editable Text Object Reference=@n2) -> (ReturnValue:"Slate Font Info Structure (by ref)")) -> (Output_Get:Slate Font Info Structure)
   > n4 = Set members in Slate Font Info!(
    execute:Exec
    StructRef:"Slate Font Info Structure (by ref)"=@n3
    Size:float="Min (Float)"(A:double="Max (Float)"(A=@n1, B=4), B:double=72) -> (ReturnValue:double) {is=title K2Node_CommutativeAssociativeBinaryOperator}
   ) -> (then:Exec, StructOut:"Slate Font Info Structure (by ref)") {is=title K2Node_SetFieldsInStruct}
   > SetFont(self:Editable Text Object Reference=@n2, InFontInfo=@n4)
  ]}
 EventGraph {nodes [
   n1 = "Clicked (BTN_SelectAssetBookmark)"!() -> (OutputDelegate, then, Event:Pointer Event Structure) {is=title K2Node_ComponentBoundEvent}
   n2 = Break STRUCT Book Shelf Asset Data!(STRUCT_BookShelf_AssetData:"STRUCT Book Shelf Asset Data Structure (by ref)"=Get BookmarkData() -> (BookmarkData:STRUCT Book Shelf Asset Data Structure)) -> (BoxColor_15_#1:Color Structure) {is=title K2Node_BreakStruct}
   > SetBrushColor(
    self=Get Background() -> (Background:Border Object Reference)
    InBrushColor="Lerp (LinearColor)"(A="To LinearColor (Color)"!(InColor:Color Structure=@n2) -> (ReturnValue:Linear Color Structure), Alpha=0.9)
   )
   > n3 = Sequence()
   > n4 = Branch(Condition=Get ConfigModeEnabled() -> (ConfigModeEnabled:bool))
   > Call Load Button for Config!(execute:Exec, self:EUW Book Case Book Mark Object Reference, Item:EUW Book Case Book Mark Object Reference=Self-Reference()) -> (then:Exec) {is=title K2Node_CallDelegate}
   Branch(execute=@n4.else, Condition=Get ProcessClicks() -> (ProcessClicks:bool))
   n5 = GetInputEventFromPointerEvent!(Event:"Pointer Event Structure (by ref)"=@n1.Event) -> (ReturnValue:Input Event Structure)
   > OpenAssetByType!(
    execute
    self
    Open Editor:bool=OR Boolean(
     A=Is Control Down!(Input:"Input Event Structure (by ref)"=@n5) -> (ReturnValue:bool)
     B=Is Alt Down!(Input:"Input Event Structure (by ref)"=@n5) -> (ReturnValue:bool)
    )
   ) -> (then)
   Delay!(execute=@n3.then_1, WorldContextObject:Object Reference, Duration=0.01) -> (then)
   n6 = Get BookmarkData()
   > SetBrushColor(
    self=Get Background()
    InBrushColor="To LinearColor (Color)"(InColor=Break STRUCT Book Shelf Asset Data!(STRUCT_BookShelf_AssetData=@n6) -> (AssetPath_3_#2:String, BoxColor_15_#1).BoxColor_15_#1)
   )
   "On Clicked (BN_EditBookmark)"() {is=title K2Node_ComponentBoundEvent}
   > Branch(Condition=Get ProcessClicks())
   > Call Bind Edit Bookmark!(execute:Exec, self:EUW Book Case Book Mark Object Reference, Item:EUW Book Case Book Mark Object Reference=Self-Reference()) -> (then:Exec) {is=title K2Node_CallDelegate}
   "CTRL Clicked (BTN_SelectAssetBookmark)"!() -> (OutputDelegate, then, Event:Pointer Event Structure) {is=title K2Node_ComponentBoundEvent}
   > Call Bind Edit Bookmark(Item=Self-Reference())
   "On Clicked (BN_RemoveBookmark)"() {is=title K2Node_ComponentBoundEvent}
   > Call Remove Bookmark!(execute:Exec, self:EUW Book Case Book Mark Object Reference, Item:EUW Book Case Book Mark Object Reference=Self-Reference()) -> (then:Exec) {is=title K2Node_CallDelegate}
   "On Clicked (BN_BrowseTo)"() {is=title K2Node_ComponentBoundEvent}
   > Branch(Condition=Get ProcessClicks())
   > SyncBrowserToObjects(AssetPaths=Make Array("[0]":String=Get!() -> (
      BookmarkData_AssetPath_3_#2:String
      BookmarkData_AssetCategory_5_#3:String
      BookmarkData_BoxColor_15_#1:Color Structure
      BookmarkData_Arguments_10_#4:Array of Strings
      BookmarkData_KeywordArguments_14_#5:Map of Strings to Strings
     )) -> (Array:Array of Strings))
   Event OnMouseEnter()
   n7 = "To LinearColor (Color)"(InColor=Break STRUCT Book Shelf Asset Data!(STRUCT_BookShelf_AssetData=Get BookmarkData()) -> (BoxColor_15_#1))
   > SetBrushColor(self=Get Background(), InBrushColor="Lerp (LinearColor)"(A=@n7, Alpha=0.3))
   Event OnMouseLeave!() -> (OutputDelegate, then, MouseEvent:"Pointer Event Structure (by ref)") {is=title K2Node_Event}
   > SetBrushColor(self=Get Background(), InBrushColor=@n7)
   Call Load Button for Config(Item=Self-Reference())
   Call Load Button for Config()
  ]}
 F_Init_Bookmark {nodes [
   n1 = F_Init_Bookmark!() -> (then, Bookmark:Name, Data:STRUCT Book Shelf Asset Data Structure, Parent:EUW Book Case Category Object Reference) {is=title K2Node_FunctionEntry}
   > Set ParentCategory(ParentCategory:EUW Book Case Category Object Reference=@n1.Parent) -> (Output_Get:EUW Book Case Category Object Reference)
   > n2 = Set Bookmark Name(Bookmark Name:Name=@n1) -> (Output_Get:Name=None)
   > Set BookmarkData(BookmarkData:STRUCT Book Shelf Asset Data Structure=@n1.Data) -> (Output_Get:STRUCT Book Shelf Asset Data Structure)
   > "SetText (Editable Text)"!(execute, self:Editable Text Object Reference=Get TX_BookmarkName(), InText:Text="To Text (Name)"!(InName:Name=@n2) -> (ReturnValue:Text)) -> (then)
   n3 = Break STRUCT Book Shelf Asset Data!(STRUCT_BookShelf_AssetData=@n1.Data) -> (
    AssetPath_3_#2:String
    AssetCategory_5_#3:String
    BoxColor_15_#1
    Arguments_10_#4:Array of Strings
    KeywordArguments_14_#5:Map of Strings to Strings
   )
   > Set Asset Name(Asset Name:Name=String To Name(InString=F_NameFromPath!(path:String=@n3) -> (name:String))) -> (Output_Get:Name=None)
   > SetToolTipText!(
    execute
    self:Widget Object Reference=Get BTN_SelectAssetBookmark() -> (BTN_SelectAssetBookmark:EUW Book Case Button Object Reference)
    InToolTipText:"Text (by ref)"=Format Text!(Format="LOCTEXT(\"({path})\")", path:String=@n3) -> (Result)
   ) -> (then)
   > SetBrushColor(self=Get Background(), InBrushColor="To LinearColor (Color)"(InColor=@n3.BoxColor_15_#1))
   > GetAssetTypeFromData()
   Return Node()
  ]}
 F_NameFromPath {nodes [
   n1 = F_NameFromPath!() -> (then, path:String) {is=title K2Node_FunctionEntry}
   n2 = Split(SourceString=@n1, InStr=., SearchDir=FromStart)
   n3 = Split(SourceString=Replace(SourceString=SelectString(A=@n2, B=@n1, bPickA=@n2.ReturnValue), From="\\", To=/), InStr=/, SearchDir=FromEnd)
   > Return Node!(execute, name:String=Replace(SourceString=Replace(SourceString=@n3, From='), From=.))
  ]}
 GetAssetTypeFromData {nodes [
   GetAssetTypeFromData() {is=title K2Node_FunctionEntry}
   > n1 = Sequence()
   n2 = "To String (Text)"(InText=GetBookmarkCleanPath!() -> (Result:Text))
   n3 = Split(SourceString=@n2, InStr=., SearchDir=FromStart)
   n4 = Replace(SourceString=SelectString(A=@n3, B=@n2, bPickA=@n3.ReturnValue), From=")")
   > n5 = GetAssetsByPackageName!(
    execute
    self:Asset Registry Interface=GetAssetRegistry()
    PackageName:Name=String To Name(InString=Replace(SourceString=@n4, From="("))
    bIncludeOnlyOnDiskAssets:bool=false
    bSkipARFilteredAssets:bool=true
   ) -> (then, OutAssetData:Array of Asset Data Structures, ReturnValue:bool)
   > n6 = Branch(Condition=Is Not Empty!(TargetArray:Array of Asset Data Structures=@n5) -> (ReturnValue:bool))
   n7 = Break Asset Data!(AssetData="Get (a copy)"(Array:Array of Asset Data Structures=@n5) -> (Output:Asset Data Structure)) -> (AssetClassPath)
   > n8 = Set AssetClass(AssetClass:String=Split(
     SourceString=BreakTopLevelAssetPath!(TopLevelAssetPath:"Top Level Asset Path Structure (by ref)"=@n7) -> (PathString:String)
     InStr=.
     SearchDir=FromStart
    ).RightS) -> (Output_Get:String)
   n9 = Switch on String!(execute:Exec=[@n16.CastFailed, @n8.then], Selection:String=Get AssetClass() -> (AssetClass:String)) -> (
    Default:Exec
    EditorUtilityWidgetBlueprint:Exec
    EditorUtilityBlueprint:Exec
    UserDefinedStruct:Exec
    Blueprint:Exec
    AtomModel:Exec
    Texture2D:Exec
    Texture:Exec
    DataTable:Exec
    GeometryCollection:Exec
    StaticMesh:Exec
    SkeletalMesh:Exec
    UserDefinedEnum:Exec
   ) {pinNames [
     EditorUtilityWidgetBlueprint
     EditorUtilityBlueprint
     UserDefinedStruct
     Blueprint
     AtomModel
     Texture2D
     Texture
     DataTable
     GeometryCollection
     StaticMesh
     SkeletalMesh
     UserDefinedEnum
    ]}
   Set option(execute=@n9.AtomModel, option:String=lm) -> (Output_Get:String)
   Set option(execute=[@n9.Blueprint, @n9.UserDefinedEnum, @n9.UserDefinedStruct], option=bp)
   Set option(execute=@n9.DataTable, option=data)
   Set option(execute=@n9.EditorUtilityBlueprint, option=task)
   Set option(execute=@n9.EditorUtilityWidgetBlueprint, option=euw)
   Set option(execute=@n9.GeometryCollection, option=gc)
   Set option(execute=@n9.SkeletalMesh, option=sk)
   Set option(execute=@n9.StaticMesh, option=sm)
   Set option(execute=[@n9.Texture, @n9.Texture2D], option=tex)
   PrintString(execute=@n6.else, InString=No Assets Found)
   n10 = Branch(execute=@n1.then_1, Condition=IsEmpty(InString=Get AssetClass()))
   > Set AssetClass(AssetClass=Folder)
   > n11 = Set option(option=folder)
   SetToolTipText(
    execute=[@n10.else, @n11.then]
    self=Get Icon() -> (Icon:Image Object Reference)
    InToolTipText=Format Text!(Format="LOCTEXT(\"Asset Class - {a}\")", a:String=Get AssetClass()) -> (Result)
   )
   n12 = Find(
    TargetMap:Map of Strings to Texture Object References=Get Options() -> (Options:Map of Strings to Texture Object References)
    Key=Get option() -> (option:String)
   ) -> (Value:Texture Object Reference)
   > Branch(Condition=@n12.ReturnValue)
   > n13 = Cast To Texture2D(Object=@n12)
   > n14 = Set IconTypeTexture(IconTypeTexture:Texture 2D Object Reference=@n13) -> (Output_Get:Texture 2D Object Reference)
   > SetBrushFromTexture!(execute, self:Image Object Reference=Get Icon(), Texture:Texture 2D Object Reference=@n14, bMatchSize:bool=false) -> (then)
   n15 = Branch(Condition="Equal (Class)"(B=/Script/Engine.World))
   > Set option(option=level)
   n16 = Cast To PrimaryDataAsset!(execute=@n15.else, Object:Wildcard) -> (then, CastFailed, AsPrimary Data Asset:Primary Data Asset Object Reference) {is=title K2Node_DynamicCast}
   > Set option(option=data)
   Branch(Condition="Equal (Name)"(B=WorldItem))
   > Set option(option=data)
  ]}
 GetBookmarkCleanPath {nodes [
   GetBookmarkCleanPath() {is=title K2Node_FunctionEntry}
   n1 = Break STRUCT Book Shelf Asset Data!(STRUCT_BookShelf_AssetData=Get BookmarkData()) -> (AssetPath_3_#2)
   n2 = Split(SourceString=@n1, InStr=', SearchDir=FromStart)
   n3 = Split(SourceString=@n2.RightS, InStr=., SearchDir=FromEnd)
   n4 = AND Boolean(A=@n3.ReturnValue, B=@n2.ReturnValue)
   > Return Node!(execute, Result:Text=Format Text!(Format="LOCTEXT(\"({f})\")", f:String=SelectString(A=@n3, B=@n1, bPickA=@n4)) -> (Result))
  ]}
 GetPackagePathFromPluginPath {nodes [
   n1 = GetPackagePathFromPluginPath!() -> (then, Plugin:String) {is=title K2Node_FunctionEntry}
   > n2 = Branch(Condition=StartsWith(SourceString=@n1, InPrefix=/All/Game/))
   > Return Node!(execute, Package:String=Append(A=/Game/, B=Split(SourceString=@n1, InStr=/Game/, SearchDir=FromStart).RightS))
   n3 = Branch(execute=@n2.else, Condition=NOT Boolean(A=StartsWith(SourceString=@n1, InPrefix=/Game/)))
   > n4 = For Each Loop(Array=ParseIntoArray(SourceString=Replace(SourceString=@n1, From="\\", To=/), Delimiter=/))
   Return Node(execute=@n4.Completed, Package=Get Path() -> (Path:String))
   > n5 = GetPluginBaseDir!(execute, PluginName:String=@n4) -> (then, OutBaseDir:String, ReturnValue:bool)
   > n6 = Branch(Condition=@n5.ReturnValue)
   > n7 = Set Start(Start:bool=true) -> (Output_Get:bool)
   Branch(execute=[@n6.else, @n7.then], Condition=Get Start() -> (Start:bool))
   > Set Path(Path:String=Append!(A=Get Path(), B=/, C:String=@n4) -> (ReturnValue)) -> (Output_Get:String)
   Return Node(execute=@n3.else, Package=@n1)
  ]}
 LoadButtonForConfig {nodes [LoadButtonForConfig!() -> (then, Item:EUW Book Case Book Mark Object Reference) {is=title K2Node_FunctionEntry}]}
 OpenAssetByType {nodes [
   n1 = OpenAssetByType!() -> (then, Open Editor:bool=~) {is=title K2Node_FunctionEntry}
   > n2 = Set Path(Path=Break STRUCT Book Shelf Asset Data(STRUCT_BookShelf_AssetData=Get BookmarkData()))
   > n3 = Validate Active Path!(execute:Exec, Base Path:String=@n2) -> (Valid:Exec, Invalid:Exec) {is=title K2Node_MacroInstance}
   PrintString(
    execute=@n3.Invalid
    InString=Append!(A="Warning: Failed to Find: ", B=Get Path()) -> (ReturnValue)
    TextColor=(R=0.887,G=0.119,B=0.0787,A=1)
    Duration=5
   )
   n4 = Branch(execute=@n3.Valid, Condition=@n1)
   > n5 = LoadAsset(AssetPath=Get Path())
   > n6 = Is Valid(InputObject=@n5)
   PrintString(
    execute=[@n8.else, @n6.Is Not Valid]
    InString=Append(A="Error: Failed To Open Asset: ", B=Get Path())
    TextColor=Break Slate Color!(SlateColor_SpecifiedColor:Linear Color Structure=(R=0.375,G=0.0339,B=0.0059,A=1), SlateColor_ColorUseRule:ESlateColorStylingMode Enum) -> (SpecifiedColor:Linear Color Structure, ColorUseRule:ESlateColorStylingMode Enum) {is=title K2Node_BreakStruct}
    Duration=Make Literal Float(Value=5)
   )
   n7 = OpenEditorForAssets!(
    execute=@n6.Is Valid
    self:Asset Editor Subsystem Object Reference=Get AssetEditorSubsystem()
    Assets:Array of Object References=Make Array("[0]":Object Reference=@n5) -> (Array:Array of Object References)
    OpenedMethod:EAssetTypeActivationOpenedMethod Enum=Edit
   ) -> (then, ReturnValue:bool)
   > n8 = Branch(Condition=@n7)
   n9 = Get Path()
   n10 = DoesAssetExist(execute=@n4.else, AssetPath=@n9)
   > n11 = Branch(Condition=@n10)
   > n12 = Branch(Condition="Not Equal Exactly (String)"(A=Get AssetClass(), B=World))
   n13 = Get()
   > n14 = Execute Python Script!(
    execute
    PythonScript="import unreal\r\n\r\nran = True\r\n\r\nasset = unreal.EditorAssetLibrary.load_asset(path)\r\neus = unreal.get_editor_subsystem(unreal.EditorUtilitySubsystem)\r\n\r\nif not asset:\r\n    ran=True\r\n    \r\nelif 'EditorUtilityWidgetBlueprint' in str(type(asset)):\r\n    ## Opens editor widget\r\n    eus.spawn_and_register_tab(asset)\r\n    \r\nelse: ## Run editor task\r\n    ran = eus.try_run(asset)\r\n    ran = True if ran else False"
    path:String=@n13
   ) -> (then, ReturnValue, ran:bool=~, asset:Object Reference) {inputs [path], outputs [ran,asset]}
   > n15 = Branch(Condition=@n14.ran)
   Is Valid(exec=@n15.else, InputObject=@n14.asset)
   > OpenEditorForAssets(self=Get AssetEditorSubsystem(), Assets=Make Array("[0]"=@n14.asset), OpenedMethod=Edit)
   n16 = LoadLevel!(execute=@n12.else, self:Level Editor Subsystem Object Reference=Get LevelEditorSubsystem(), AssetPath:String=@n13) -> (then, ReturnValue:bool)
   > Branch(Condition=@n16)
   > PrintString(InString=Append(A="Warning: Failed To Load Level: ", B=@n13), TextColor=(R=0.887,G=0.119,B=0.0787,A=1), Duration=5)
   SyncBrowserToFolders!(execute=@n11.else, FolderList:Array of Strings=Make Array("[0]":String=@n9) -> (Array:Array of Strings)) -> (then)
  ]}
 RemoveBookmark {nodes [RemoveBookmark!() -> (then, Item:EUW Book Case Book Mark Object Reference) {is=title K2Node_FunctionEntry}]}
 SetConfigMode {nodes [
   n1 = SetConfigMode!() -> (then, Active:bool=~) {is=title K2Node_FunctionEntry}
   > Set ConfigModeEnabled(ConfigModeEnabled:bool=@n1) -> (Output_Get:bool)
   n2 = Get ConfigModeEnabled()
   n3 = Get img-editmode() -> (img-editmode:Image Object Reference)
   > SetBrushFromTexture(
    self=@n3
    Texture=Select(
     Option 0:Texture 2D Object Reference=Get NewLocalVar() -> (NewLocalVar:Texture 2D Object Reference)
     Option 1:Texture 2D Object Reference=Get Edit Mode Texture() -> (Edit Mode Texture:Texture 2D Object Reference)
     Index=@n2
    ) -> (ReturnValue:Texture 2D Object Reference)
    bMatchSize=false
   )
   > SetBrushTintColor!(
    execute
    self:Image Object Reference=@n3
    TintColor:Slate Color Structure=Select(
     Option 0:Slate Color Structure=Get Default Icon Type Color() -> (Default Icon Type Color:Slate Color Structure)
     Option 1:Slate Color Structure=Get Edit Mode Icon Color() -> (Edit Mode Icon Color:Slate Color Structure)
     Index=@n2
    ) -> (ReturnValue:Slate Color Structure)
   ) -> (then)
   > SetToolTipText(
    self=[Get BTN_SelectAssetBookmark(), @n3]
    InToolTipText=SelectText(
     A="LOCTEXT(\"This bookmark will be updated when you click \\'Update Bookmark\\'\")"
     B=GetBookmarkCleanPath() {is=title K2Node_CallFunction}
     bPickA=Get ConfigModeEnabled()
    )
   )
  ]}
 ValidateActivePath {nodes [
   n1 = Inputs!() -> (execute, Base Path:String)
   n2 = DirectoryExists!(InPath:String=@n1) -> (ReturnValue:bool)
   n3 = GetPluginNameForAssetPath!(AssetPath:String=@n1) -> (bIsInMainModule:bool, ReturnValue:String)
   > n4 = Branch(Condition=AND Boolean(A=IsEmpty(InString=@n3.ReturnValue), B=NOT Boolean(A=@n2)))
   Outputs!(Valid:Exec=@n4.else, Invalid:Exec=@n4.then)
  ]}
]
variables [
 Asset Name {type=string, default=None}
 AssetClass {type=string, default=""}
 BindEditBookmark {default=()}
 Bookmark Name {type=string, default=None}
 BookmarkData {
  type=STRUCT_BookShelf_AssetData
  default {asset Category="", asset Path="", arguments [], boxColor=<rgba 0.2902 0.2902 0.298 1>, keywordArguments {}}
 }
 ConfigModeEnabled {type=boolean}
 Default Icon Type Color {type=SlateColor, default {colorUseRule=UseColor_Specified, specifiedColor=<rgba 1 1 1 1>}}
 IconTypeTexture {type=Texture2D}
 LoadButtonForConfig {default=()}
 ParentCategory {type=EUW_BookCase_Category_C, default=null}
 ProcessClicks {type=boolean, default=true}
 RemoveBookmark {default=()}
]
widgets {
 info {namedSlotCount=1, rootWidgetClass=&Border, widgetCount=19}
 widgets [
  SpacerBackground {
   widgetClassPath=&Border
   widget=SpacerBackground
   properties {verticalAlignment=VAlign_Center, brushColor=<rgba 1 1 1 0>, padding=<ltrb 0 0.5 0 0.5>}
  }
  Background {
   widgetClassPath=&Border
   bIsVariable=true
   parent=SpacerBackground
   widget=Background
   properties {verticalAlignment=VAlign_Center, brushColor=<rgba 0.0162 0.0178 0.0261 1>, padding=<ltrb 2 2 2 2>}
   slot {class=BorderSlot, properties {verticalAlignment=VAlign_Center, padding=<ltrb 0 0.5 0 0.5>}}
  }
  HorizontalBox_0 {
   widgetClassPath=&HorizontalBox
   parent=Background
   widget=HorizontalBox_0
   properties {bOverride_Cursor=true}
   slot {class=BorderSlot, properties {verticalAlignment=VAlign_Center, padding=<ltrb 2 2 2 2>}}
  }
  Border_65 {
   widgetClassPath=&Border
   parent=HorizontalBox_0
   widget=Border_65
   properties {horizontalAlignment=HAlign_Left, brushColor=<rgba 0 0 0 0.2>, padding=<ltrb 0 0 0 0>}
   slot {class=HorizontalBoxSlot, properties {padding=<ltrb 0 0 2 0>}}
  }
  ScaleBox_1 {
   widgetClassPath=&ScaleBox
   parent=Border_65
   widget=ScaleBox_1
   properties {ignoreInheritedScale=true, stretch=ScaleToFitY}
   slot {class=BorderSlot, properties {horizontalAlignment=HAlign_Left, padding=<ltrb 0 0 0 0>}}
  }
  BN_BrowseTo {
   widgetClassPath=&EditorUtilityButton
   parent=ScaleBox_1
   widget=BN_BrowseTo
   properties {
    accessibleBehavior=NotAccessible
    bCanChildrenBeAccessible=true
    bOverride_Cursor=true
    toolTipText=Browse to asset
    backgroundColor=<rgba 0 0 0 0.2>
    widgetStyle {
     clickedSlateSound {resourceObject=null}
     disabled {
      bIsDynamicallyLoaded=false
      drawAs=Box
      imageType=NoImage
      mirroring=NoMirror
      resourceName=None
      resourceObject=null
      tiling=NoTile
      imageSize=<32 32>
      margin=<ltrb 0 0 0 0>
      outlineSettings {
       bUseBrushTransparency=false
       roundingType=FixedRadius
       width=1
       color {colorUseRule=UseColor_ColorTable, specifiedColor=<rgba 0 0 0 0>}
       cornerRadii=<4 4 4 4>
      }
      tintColor {colorUseRule=UseColor_Specified, specifiedColor=<rgba 1 1 1 1>}
      uVRegion {bIsValid=false, max=<0 0>, min=<0 0>}
     }
     disabledForeground {colorUseRule=UseColor_ColorTable, specifiedColor=<rgba 0 0 0 0>}
     hovered {
      bIsDynamicallyLoaded=false
      drawAs=Box
      imageType=NoImage
      mirroring=NoMirror
      resourceName=None
      resourceObject=null
      tiling=NoTile
      imageSize=<32 32>
      margin=<ltrb 0 0 0 0>
      outlineSettings {
       bUseBrushTransparency=false
       roundingType=FixedRadius
       width=1
       color {colorUseRule=UseColor_ColorTable, specifiedColor=<rgba 0 0 0 0>}
       cornerRadii=<4 4 4 4>
      }
      tintColor {colorUseRule=UseColor_Specified, specifiedColor=<rgba 0.8879 0.624 1 1>}
      uVRegion {bIsValid=false, max=<0 0>, min=<0 0>}
     }
     hoveredForeground {colorUseRule=UseColor_ColorTable, specifiedColor=<rgba 0 0 0 0>}
     hoveredSlateSound {resourceObject=null}
     normal {
      bIsDynamicallyLoaded=false
      drawAs=Box
      imageType=NoImage
      mirroring=NoMirror
      resourceName=None
      resourceObject=null
      tiling=NoTile
      imageSize=<32 32>
      margin=<ltrb 0 0 0 0>
      outlineSettings {
       bUseBrushTransparency=false
       roundingType=FixedRadius
       width=1
       color {colorUseRule=UseColor_ColorTable, specifiedColor=<rgba 0 0 0 0>}
       cornerRadii=<4 4 4 4>
      }
      tintColor {colorUseRule=UseColor_Specified, specifiedColor=<rgba 1 1 1 0>}
      uVRegion {bIsValid=false, max=<0 0>, min=<0 0>}
     }
     normalForeground {colorUseRule=UseColor_ColorTable, specifiedColor=<rgba 0 0 0 0>}
     normalPadding=<ltrb 0 0 0 0>
     pressed {
      bIsDynamicallyLoaded=false
      drawAs=Box
      imageType=NoImage
      mirroring=NoMirror
      resourceName=None
      resourceObject=null
      tiling=NoTile
      imageSize=<32 32>
      margin=<ltrb 0 0 0 0>
      outlineSettings {
       bUseBrushTransparency=false
       roundingType=FixedRadius
       width=1
       color {colorUseRule=UseColor_ColorTable, specifiedColor=<rgba 0 0 0 0>}
       cornerRadii=<4 4 4 4>
      }
      tintColor {colorUseRule=UseColor_Specified, specifiedColor=<rgba 1 1 1 1>}
      uVRegion {bIsValid=false, max=<0 0>, min=<0 0>}
     }
     pressedForeground {colorUseRule=UseColor_ColorTable, specifiedColor=<rgba 0 0 0 0>}
     pressedPadding=<ltrb 0 0 0 0>
     pressedSlateSound {resourceObject=null}
    }
   }
   slot {class=ScaleBoxSlot, properties {horizontalAlignment=HAlign_Fill, verticalAlignment=VAlign_Fill}}
  }
  img-goTo {
   widgetClassPath=&Image
   parent=BN_BrowseTo
   widget=img-goTo
   properties {
    brush {
     bIsDynamicallyLoaded=false
     drawAs=Image
     imageType=FullColor
     mirroring=NoMirror
     resourceName=None
     resourceObject=/Game/Athena/UI/InGame/Creative/PropertyEditor/Icon_Picker/Textures_V2/64x64/T_UI_IconLibrary_FastForward_64
     tiling=NoTile
     imageSize=<12 12>
     margin=<ltrb 0 0 0 0>
     outlineSettings {
      bUseBrushTransparency=false
      roundingType=HalfHeightRadius
      width=0
      color {colorUseRule=UseColor_Specified, specifiedColor=<rgba 0 0 0 0>}
      cornerRadii=<0 0 0 0>
     }
     tintColor {colorUseRule=UseColor_Specified, specifiedColor=<rgba 1 1 1 1>}
     uVRegion {bIsValid=false, max=<0 0>, min=<0 0>}
    }
    colorAndOpacity=<rgba 1 1 1 0.9>
    renderTransform {angle=0, scale=<0.7 0.7>, shear=<0 0>, translation=<0 0>}
   }
   slot {class=ButtonSlot, properties {horizontalAlignment=HAlign_Fill, verticalAlignment=VAlign_Fill, padding=<ltrb 1 0 1 0>}}
  }
  BTN_SelectAssetBookmark {
   widgetClassPath=&/EUW_BookCase_Button
   bIsVariable=true
   parent=HorizontalBox_0
   widget=BTN_SelectAssetBookmark
   properties {
    visibility=Visible
    brush {
     bIsDynamicallyLoaded=false
     drawAs=Box
     imageType=NoImage
     mirroring=NoMirror
     resourceName=None
     resourceObject=null
     tiling=NoTile
     imageSize=<32 32>
     margin=<ltrb 0 0 0 0>
     outlineSettings {
      bUseBrushTransparency=false
      roundingType=HalfHeightRadius
      width=0
      color {colorUseRule=UseColor_Specified, specifiedColor=<rgba 0 0 0 0>}
      cornerRadii=<2 2 2 2>
     }
     tintColor {colorUseRule=UseColor_Specified, specifiedColor=<rgba 1 1 1 1>}
     uVRegion {bIsValid=false, max=<0 0>, min=<0 0>}
    }
    brush Color=<rgba 0 0 0 0.2>
    button Padding=<ltrb 2 2.25 2 2>
    foregroundColor {colorUseRule=UseColor_Specified, specifiedColor=<rgba 1 1 1 1>}
   }
   slot {class=HorizontalBoxSlot, properties {size {sizeRule=Fill, value=1}}}
  }
  Overlay_1 {widgetClassPath=&Overlay, namedSlotHost=BTN_SelectAssetBookmark, widget=Overlay_1}
  TX_BookmarkName {
   widgetClassPath=/Script/Blutility.EditorUtilityEditableText
   bIsVariable=true
   parent=Overlay_1
   widget=TX_BookmarkName
   properties {
    hintText=Bookmark Name
    justification=Center
    text=Test Name
    visibility=HitTestInvisible
    widgetStyle {
     backgroundImageComposing {
      bIsDynamicallyLoaded=false
      drawAs=NoDrawType
      imageType=FullColor
      mirroring=NoMirror
      resourceName=../../../Engine/Content/Slate/Old/HyperlinkDotted.png
      resourceObject=null
      tiling=Both
      imageSize=<0 0>
      margin=<ltrb 0 0 0 0.1875>
      outlineSettings {
       bUseBrushTransparency=false
       roundingType=HalfHeightRadius
       width=0
       color {colorUseRule=UseColor_Specified, specifiedColor=<rgba 0 0 0 0>}
       cornerRadii=<0 0 0 0>
      }
      tintColor {colorUseRule=UseColor_Specified, specifiedColor=<rgba 1 1 1 1>}
      uVRegion {bIsValid=false, max=<0 0>, min=<0 0>}
     }
     backgroundImageSelected {
      bIsDynamicallyLoaded=false
      drawAs=Image
      imageType=NoImage
      mirroring=NoMirror
      resourceName=None
      resourceObject=null
      tiling=NoTile
      imageSize=<0 0>
      margin=<ltrb 0 0 0 0>
      outlineSettings {
       bUseBrushTransparency=false
       roundingType=HalfHeightRadius
       width=0
       color {colorUseRule=UseColor_Specified, specifiedColor=<rgba 0 0 0 0>}
       cornerRadii=<0 0 0 0>
      }
      tintColor {colorUseRule=UseColor_Specified, specifiedColor=<rgba 0 0.162 0.7454 1>}
      uVRegion {bIsValid=false, max=<0 0>, min=<0 0>}
     }
     caretImage {
      bIsDynamicallyLoaded=false
      drawAs=Image
      imageType=NoImage
      mirroring=NoMirror
      resourceName=None
      resourceObject=null
      tiling=NoTile
      imageSize=<0 0>
      margin=<ltrb 0 0 0 0>
      outlineSettings {
       bUseBrushTransparency=false
       roundingType=HalfHeightRadius
       width=0
       color {colorUseRule=UseColor_Specified, specifiedColor=<rgba 0 0 0 0>}
       cornerRadii=<0 0 0 0>
      }
      tintColor {colorUseRule=UseColor_Specified, specifiedColor=<rgba 1 1 1 1>}
      uVRegion {bIsValid=false, max=<0 0>, min=<0 0>}
     }
     colorAndOpacity {colorUseRule=UseColor_Specified, specifiedColor=<rgba 0.9 0.9 0.9 1>}
     font {
      bForceMonospaced=false
      bMaterialIsStencil=false
      fontMaterial=null
      fontObject=&Roboto
      letterSpacing=0
      monospacedWidth=1
      size=8.25
      skewAmount=0
      typefaceFontName=Regular
      outlineSettings {
       bApplyOutlineToDropShadows=false
       bMiteredCorners=false
       bSeparateFillAlpha=false
       outlineMaterial=null
       outlineSize=0
       outlineColor=<rgba 1 1 1 1>
      }
     }
     hintColorAndOpacity {colorUseRule=UseColor_Foreground, specifiedColor=<rgba 1 0 1 1>}
     selectedBackgroundColor {colorUseRule=UseColor_Foreground, specifiedColor=<rgba 1 0 1 1>}
    }
   }
   slot {class=OverlaySlot, properties {horizontalAlignment=HAlign_Fill, verticalAlignment=VAlign_Center, padding=<ltrb 0 0.3 0 0>}}
  }
  ScaleBox_2 {
   widgetClassPath=&ScaleBox
   parent=Overlay_1
   widget=ScaleBox_2
   properties {ignoreInheritedScale=true, stretch=ScaleToFitY}
   slot {class=OverlaySlot, properties {horizontalAlignment=HAlign_Fill, verticalAlignment=VAlign_Fill}}
  }
  Icon {
   widgetClassPath=&Image
   bIsVariable=true
   parent=ScaleBox_2
   widget=Icon
   properties {
    clipping=ClipToBounds
    toolTipText=test
    brush {
     bIsDynamicallyLoaded=false
     drawAs=Image
     imageType=FullColor
     mirroring=NoMirror
     resourceName=None
     resourceObject=/Game/Athena/UI/InGame/Creative/PropertyEditor/Icon_Picker/Textures_V2/64x64/T_UI_IconLibrary_Cube_64
     tiling=NoTile
     imageSize=<12 12>
     margin=<ltrb 0 0 0 0>
     outlineSettings {
      bUseBrushTransparency=false
      roundingType=HalfHeightRadius
      width=0
      color {colorUseRule=UseColor_Specified, specifiedColor=<rgba 0 0 0 0>}
      cornerRadii=<0 0 0 0>
     }
     tintColor {colorUseRule=UseColor_Specified, specifiedColor=<rgba 1 1 1 1>}
     uVRegion {bIsValid=false, max=<0 0>, min=<0 0>}
    }
    colorAndOpacity=<rgba 1 1 1 0.75>
   }
   slot {class=ScaleBoxSlot, properties {horizontalAlignment=HAlign_Left, verticalAlignment=VAlign_Fill}}
  }
  ScaleBox {
   widgetClassPath=&ScaleBox
   parent=Overlay_1
   widget=ScaleBox
   properties {ignoreInheritedScale=true, stretch=ScaleToFitY}
   slot {class=OverlaySlot, properties {horizontalAlignment=HAlign_Right, verticalAlignment=VAlign_Fill}}
  }
  BN_EditBookmark {
   widgetClassPath=&EditorUtilityButton
   parent=ScaleBox
   widget=BN_EditBookmark
   properties {
    accessibleBehavior=NotAccessible
    bCanChildrenBeAccessible=true
    bOverride_Cursor=true
    toolTipText=Edit Bookmark
    backgroundColor=<rgba 0 0 0 0.2>
    widgetStyle {
     clickedSlateSound {resourceObject=null}
     disabled {
      bIsDynamicallyLoaded=false
      drawAs=Box
      imageType=NoImage
      mirroring=NoMirror
      resourceName=None
      resourceObject=null
      tiling=NoTile
      imageSize=<32 32>
      margin=<ltrb 0 0 0 0>
      outlineSettings {
       bUseBrushTransparency=false
       roundingType=FixedRadius
       width=1
       color {colorUseRule=UseColor_ColorTable, specifiedColor=<rgba 0 0 0 0>}
       cornerRadii=<4 4 4 4>
      }
      tintColor {colorUseRule=UseColor_Specified, specifiedColor=<rgba 1 1 1 1>}
      uVRegion {bIsValid=false, max=<0 0>, min=<0 0>}
     }
     disabledForeground {colorUseRule=UseColor_ColorTable, specifiedColor=<rgba 0 0 0 0>}
     hovered {
      bIsDynamicallyLoaded=false
      drawAs=Box
      imageType=NoImage
      mirroring=NoMirror
      resourceName=None
      resourceObject=null
      tiling=NoTile
      imageSize=<32 32>
      margin=<ltrb 0 0 0 0>
      outlineSettings {
       bUseBrushTransparency=false
       roundingType=FixedRadius
       width=1
       color {colorUseRule=UseColor_ColorTable, specifiedColor=<rgba 0 0 0 0>}
       cornerRadii=<4 4 4 4>
      }
      tintColor {colorUseRule=UseColor_Specified, specifiedColor=<rgba 0.8879 0.624 1 1>}
      uVRegion {bIsValid=false, max=<0 0>, min=<0 0>}
     }
     hoveredForeground {colorUseRule=UseColor_ColorTable, specifiedColor=<rgba 0 0 0 0>}
     hoveredSlateSound {resourceObject=null}
     normal {
      bIsDynamicallyLoaded=false
      drawAs=Box
      imageType=NoImage
      mirroring=NoMirror
      resourceName=None
      resourceObject=null
      tiling=NoTile
      imageSize=<32 32>
      margin=<ltrb 0 0 0 0>
      outlineSettings {
       bUseBrushTransparency=false
       roundingType=FixedRadius
       width=1
       color {colorUseRule=UseColor_ColorTable, specifiedColor=<rgba 0 0 0 0>}
       cornerRadii=<4 4 4 4>
      }
      tintColor {colorUseRule=UseColor_Specified, specifiedColor=<rgba 1 1 1 0>}
      uVRegion {bIsValid=false, max=<0 0>, min=<0 0>}
     }
     normalForeground {colorUseRule=UseColor_ColorTable, specifiedColor=<rgba 0 0 0 0>}
     normalPadding=<ltrb 0 0 0 0>
     pressed {
      bIsDynamicallyLoaded=false
      drawAs=Box
      imageType=NoImage
      mirroring=NoMirror
      resourceName=None
      resourceObject=null
      tiling=NoTile
      imageSize=<32 32>
      margin=<ltrb 0 0 0 0>
      outlineSettings {
       bUseBrushTransparency=false
       roundingType=FixedRadius
       width=1
       color {colorUseRule=UseColor_ColorTable, specifiedColor=<rgba 0 0 0 0>}
       cornerRadii=<4 4 4 4>
      }
      tintColor {colorUseRule=UseColor_Specified, specifiedColor=<rgba 1 1 1 1>}
      uVRegion {bIsValid=false, max=<0 0>, min=<0 0>}
     }
     pressedForeground {colorUseRule=UseColor_ColorTable, specifiedColor=<rgba 0 0 0 0>}
     pressedPadding=<ltrb 0 0 0 0>
     pressedSlateSound {resourceObject=null}
    }
   }
   slot {class=ScaleBoxSlot, properties {horizontalAlignment=HAlign_Fill, verticalAlignment=VAlign_Fill}}
  }
  img-editmode {
   widgetClassPath=&Image
   bIsVariable=true
   parent=BN_EditBookmark
   widget=img-editmode
   properties {
    visibility=SelfHitTestInvisible
    brush {
     bIsDynamicallyLoaded=false
     drawAs=Box
     imageType=FullColor
     mirroring=NoMirror
     resourceName=None
     resourceObject=/Game/Athena/UI/InGame/Creative/PropertyEditor/Icon_Picker/Textures_V2/64x64/T_UI_IconLibrary_Pencil_64
     tiling=NoTile
     imageSize=<12 12>
     margin=<ltrb 0 0 0 0>
     outlineSettings {
      bUseBrushTransparency=false
      roundingType=HalfHeightRadius
      width=0
      color {colorUseRule=UseColor_Specified, specifiedColor=<rgba 0 0 0 0>}
      cornerRadii=<0 0 0 0>
     }
     tintColor {colorUseRule=UseColor_Specified, specifiedColor=<rgba 0.77 0.77 0.77 1>}
     uVRegion {bIsValid=false, max=<0 0>, min=<0 0>}
    }
    colorAndOpacity=<rgba 1 1 1 0.9>
    renderTransform {angle=0, scale=<0.8 0.75>, shear=<0 0>, translation=<0 0>}
   }
   slot {class=ButtonSlot, properties {horizontalAlignment=HAlign_Right, verticalAlignment=VAlign_Fill, padding=<ltrb 2 2 2 2>}}
  }
  Border {
   widgetClassPath=&Border
   parent=HorizontalBox_0
   widget=Border
   properties {brushColor=<rgba 0 0 0 0.2>, padding=<ltrb 0 0 0 0>}
   slot {class=HorizontalBoxSlot, properties {horizontalAlignment=HAlign_Right, padding=<ltrb 2 0 0 0>}}
  }
  ScaleBox_0 {
   widgetClassPath=&ScaleBox
   parent=Border
   widget=ScaleBox_0
   properties {ignoreInheritedScale=true, stretch=ScaleToFitY}
   slot {class=BorderSlot, properties {padding=<ltrb 0 0 0 0>}}
  }
  BN_RemoveBookmark {
   widgetClassPath=&EditorUtilityButton
   parent=ScaleBox_0
   widget=BN_RemoveBookmark
   properties {
    accessibleBehavior=NotAccessible
    bCanChildrenBeAccessible=true
    bOverride_Cursor=true
    toolTipText=Remove Bookmark
    backgroundColor=<rgba 0 0 0 0.2>
    widgetStyle {
     clickedSlateSound {resourceObject=null}
     disabled {
      bIsDynamicallyLoaded=false
      drawAs=Box
      imageType=NoImage
      mirroring=NoMirror
      resourceName=None
      resourceObject=null
      tiling=NoTile
      imageSize=<32 32>
      margin=<ltrb 0 0 0 0>
      outlineSettings {
       bUseBrushTransparency=false
       roundingType=FixedRadius
       width=1
       color {colorUseRule=UseColor_ColorTable, specifiedColor=<rgba 0 0 0 0>}
       cornerRadii=<4 4 4 4>
      }
      tintColor {colorUseRule=UseColor_Specified, specifiedColor=<rgba 1 1 1 1>}
      uVRegion {bIsValid=false, max=<0 0>, min=<0 0>}
     }
     disabledForeground {colorUseRule=UseColor_ColorTable, specifiedColor=<rgba 0 0 0 0>}
     hovered {
      bIsDynamicallyLoaded=false
      drawAs=Box
      imageType=NoImage
      mirroring=NoMirror
      resourceName=None
      resourceObject=null
      tiling=NoTile
      imageSize=<32 32>
      margin=<ltrb 0 0 0 0>
      outlineSettings {
       bUseBrushTransparency=false
       roundingType=FixedRadius
       width=1
       color {colorUseRule=UseColor_ColorTable, specifiedColor=<rgba 0 0 0 0>}
       cornerRadii=<4 4 4 4>
      }
      tintColor {colorUseRule=UseColor_Specified, specifiedColor=<rgba 0.8879 0.624 1 1>}
      uVRegion {bIsValid=false, max=<0 0>, min=<0 0>}
     }
     hoveredForeground {colorUseRule=UseColor_ColorTable, specifiedColor=<rgba 0 0 0 0>}
     hoveredSlateSound {resourceObject=null}
     normal {
      bIsDynamicallyLoaded=false
      drawAs=Box
      imageType=NoImage
      mirroring=NoMirror
      resourceName=None
      resourceObject=null
      tiling=NoTile
      imageSize=<32 32>
      margin=<ltrb 0 0 0 0>
      outlineSettings {
       bUseBrushTransparency=false
       roundingType=FixedRadius
       width=1
       color {colorUseRule=UseColor_ColorTable, specifiedColor=<rgba 0 0 0 0>}
       cornerRadii=<4 4 4 4>
      }
      tintColor {colorUseRule=UseColor_Specified, specifiedColor=<rgba 1 1 1 0>}
      uVRegion {bIsValid=false, max=<0 0>, min=<0 0>}
     }
     normalForeground {colorUseRule=UseColor_ColorTable, specifiedColor=<rgba 0 0 0 0>}
     normalPadding=<ltrb 0 0 0 0>
     pressed {
      bIsDynamicallyLoaded=false
      drawAs=Box
      imageType=NoImage
      mirroring=NoMirror
      resourceName=None
      resourceObject=null
      tiling=NoTile
      imageSize=<32 32>
      margin=<ltrb 0 0 0 0>
      outlineSettings {
       bUseBrushTransparency=false
       roundingType=FixedRadius
       width=1
       color {colorUseRule=UseColor_ColorTable, specifiedColor=<rgba 0 0 0 0>}
       cornerRadii=<4 4 4 4>
      }
      tintColor {colorUseRule=UseColor_Specified, specifiedColor=<rgba 1 1 1 1>}
      uVRegion {bIsValid=false, max=<0 0>, min=<0 0>}
     }
     pressedForeground {colorUseRule=UseColor_ColorTable, specifiedColor=<rgba 0 0 0 0>}
     pressedPadding=<ltrb 0 0 0 0>
     pressedSlateSound {resourceObject=null}
    }
   }
   slot {class=ScaleBoxSlot, properties {horizontalAlignment=HAlign_Fill, verticalAlignment=VAlign_Fill}}
  }
  img-X {
   widgetClassPath=&Image
   parent=BN_RemoveBookmark
   widget=img-X
   properties {
    brush {
     bIsDynamicallyLoaded=false
     drawAs=Box
     imageType=FullColor
     mirroring=NoMirror
     resourceName=None
     resourceObject=/Game/Athena/UI/InGame/Creative/PropertyEditor/Icon_Picker/Textures/64x64/T_UI_IconLibrary_X_64
     tiling=NoTile
     imageSize=<12 12>
     margin=<ltrb 0 0 0 0>
     outlineSettings {
      bUseBrushTransparency=false
      roundingType=HalfHeightRadius
      width=0
      color {colorUseRule=UseColor_Specified, specifiedColor=<rgba 0 0 0 0>}
      cornerRadii=<0 0 0 0>
     }
     tintColor {colorUseRule=UseColor_Specified, specifiedColor=<rgba 0.77 0.77 0.77 1>}
     uVRegion {bIsValid=false, max=<0 0>, min=<0 0>}
    }
    colorAndOpacity=<rgba 1 1 1 0.9>
    renderTransform {angle=0, scale=<0.8 0.75>, shear=<0 0>, translation=<0 0>}
   }
   slot {class=ButtonSlot, properties {horizontalAlignment=HAlign_Fill, verticalAlignment=VAlign_Fill, padding=<ltrb 2 2 2 2>}}
  }
 ]
}
