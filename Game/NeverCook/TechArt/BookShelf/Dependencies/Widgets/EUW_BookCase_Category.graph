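# Editor utility widget for one BookShelf category: a labelled expandable area holding bookmark widgets.
# Preview mouse-down picks the bookmark under the cursor, except over its edit button, and allows dragging with Left Shift or config mode.
# Drop and cancel share a handler that hit-tests cursor geometry for a neighbor bookmark, records this category as target and broadcasts Bookmark_Dropped.
# It also propagates config mode to bookmarks and tints its border from their averaged, desaturated colors.
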EditorUtilityWidgetBlueprint /Game/NeverCook/TechArt/BookShelf/Dependencies/Widgets/EUW_BookCase_Category
graphs [
 BookcaseDragDropped {nodes [
   n1 = BookcaseDragDropped!() -> (then, PointerEvent:Pointer Event Structure, Operation:Object Reference) {is=title K2Node_FunctionEntry}
   > n2 = Cast To BP_BookCase_DragDrop!(execute, Object=@n1.Operation) -> (then, CastFailed, AsBP Book Case Drag Drop:BP Book Case Drag Drop Object Reference) {is=title K2Node_DynamicCast}
   > n3 = Branch(Condition=IsWidgetInBounds!(OwningEvent:"Pointer Event Structure (by ref)"=@n1, Widget:Widget Object Reference=Get Label() -> (Label:Text Object Reference)) -> (ReturnValue:bool))
   Set NewCategory!(
    execute=[@n7.then, @n4.Completed, @n3.then]
    NewCategory:EUW Book Case Category Object Reference=Self-Reference()
    self:BP Book Case Drag Drop Object Reference=@n2
   ) -> (then, Output_Get:EUW Book Case Category Object Reference)
   > Call Bookmark Dropped!(execute:Exec, self:EUW Book Case Category Object Reference, Operation:BP Book Case Drag Drop Object Reference=@n2) -> (then:Exec) {is=title K2Node_CallDelegate}
   > Set DraggedBookmark(DraggedBookmark:EUW Book Case Book Mark Object Reference) -> (Output_Get:EUW Book Case Book Mark Object Reference)
   > Set DraggedBookmark!(
    execute
    DraggedBookmark
    self:EUW Book Case Category Object Reference=Get Category!(self:BP Book Case Drag Drop Object Reference=@n2) -> (Category:EUW Book Case Category Object Reference)
   ) -> (then, Output_Get)
   n4 = For Each Loop(
    Exec=@n3.else
    Array:Array of Widget Object References=GetAllChildren(self=Get VBox_Content() -> (VBox_Content:Vertical Box Object Reference))
   ) -> (Array Element:Widget Object Reference)
   > n5 = Cast To EUW_BookCase_BookMark!(execute, Object=@n4) -> (then, CastFailed, AsEUW Book Case Book Mark:EUW Book Case Book Mark Object Reference) {is=title K2Node_DynamicCast}
   > Branch(Condition=IsWidgetInBounds(OwningEvent=Get PointerEvent() -> (PointerEvent:Pointer Event Structure), Widget=@n5))
   > n6 = Branch(Condition="Equal (Object)"(A=@n5, B=Get Bookmark!(self:BP Book Case Drag Drop Object Reference=@n2) -> (Bookmark:EUW Book Case Book Mark Object Reference)))
   n7 = Set NeighborBookmark!(execute=@n6.else, NeighborBookmark:EUW Book Case Book Mark Object Reference=@n5, self:BP Book Case Drag Drop Object Reference=@n2) -> (then, Output_Get:EUW Book Case Book Mark Object Reference)
  ]}
 Bookmark_Dropped {nodes [Bookmark_Dropped!() -> (then, Operation:BP Book Case Drag Drop Object Reference) {is=title K2Node_FunctionEntry}]}
 EventGraph {nodes [
   n1 = Event OnDragCancelled!() -> (OutputDelegate, then, PointerEvent:"Pointer Event Structure (by ref)", Operation:Drag Drop Operation Object Reference) {is=title K2Node_Event}
   > PrintString(InString=Cancelled)
   > BookcaseDragDropped!(execute, self, PointerEvent:Pointer Event Structure=@n1.PointerEvent, Operation:Object Reference=@n1.Operation) -> (then) {is=title K2Node_CallFunction}
   Event PreConstruct()
   Event Construct()
   Event Tick()
  ]}
 IsWidgetInBounds {nodes [
   n1 = IsWidgetInBounds!() -> (then, OwningEvent:"Pointer Event Structure (by ref)", Widget:Widget Object Reference) {is=title K2Node_FunctionEntry}
   n2 = GetCachedGeometry!(self:Widget Object Reference=@n1.Widget) -> (ReturnValue:"Geometry Structure (by ref)")
   n3 = BreakVector2D(InVec="Absolute to Local (Coordinates)"!(
     Geometry:"Geometry Structure (by ref)"=@n2
     AbsoluteCoordinate:Vector 2D Structure=Get Screen Space Position!(Input:"Pointer Event Structure (by ref)"=@n1) -> (ReturnValue:Vector 2D Structure) {is=title K2Node_CallFunction}
    ) -> (ReturnValue:Vector 2D Structure))
   n4 = BreakVector2D(InVec=GetLocalSize!(Geometry:"Geometry Structure (by ref)"=@n2) -> (ReturnValue:Vector 2D Structure))
   > Return Node!(
    execute
    ReturnValue:bool=AND Boolean!(
     A="float >= float"(A=@n3)
     B="float >= float"(A=@n3.Y)
     C:bool="float <= float"(A=@n3, B=@n4) -> (ReturnValue=~)
     D:bool="float <= float"(A=@n3.Y, B=@n4.Y) -> (ReturnValue=~)
    ) -> (ReturnValue)
   )
  ]}
 OnDragDetected {nodes [
   OnDragDetected!() -> (then, MyGeometry:Geometry Structure, PointerEvent:"Pointer Event Structure (by ref)") {is=title K2Node_FunctionEntry}
   > PrintString(InString=Dragging detected)
   n1 = Get DraggedBookmark() -> (DraggedBookmark:EUW Book Case Book Mark Object Reference)
   > n2 = "Create Drag && Drop Operation"!(
    execute:Exec
    WorldContextObject:Object Reference
    Class:Drag Drop Operation Class Reference=&/BP_BookCase_DragDrop.BP_BookCase_DragDrop_C
    Tag:String
    Payload:Object Reference
    DefaultDragVisual:Widget Object Reference=@n1
    Pivot:EDragPivot Enum=CenterCenter
    Offset:Vector 2D Structure=(X=0,Y=0)
    Bookmark:EUW Book Case Book Mark Object Reference=@n1
    Category:EUW Book Case Category Object Reference=Self-Reference()
    NewCategory:EUW Book Case Category Object Reference
    NeighborBookmark:EUW Book Case Book Mark Object Reference
   ) -> (then:Exec, ReturnValue:BP Book Case Drag Drop Object Reference) {is=title K2Node_CreateDragDropOperation}
   > Set Bookmark!(execute, Bookmark:EUW Book Case Book Mark Object Reference=Get DraggedBookmark(), self:BP Book Case Drag Drop Object Reference=@n2) -> (then, Output_Get:EUW Book Case Book Mark Object Reference)
   > Return Node!(execute, Operation:Drag Drop Operation Object Reference=@n2)
  ]}
 OnDrop {nodes [
   n1 = OnDrop!() -> (then, MyGeometry:Geometry Structure, PointerEvent:Pointer Event Structure, Operation:Drag Drop Operation Object Reference) {is=title K2Node_FunctionEntry}
   > BookcaseDragDropped(PointerEvent=@n1.PointerEvent, Operation=@n1.Operation)
   > Return Node!(execute, ReturnValue:bool=true)
  ]}
 OnPreviewMouseButtonDown {nodes [
   n1 = OnPreviewMouseButtonDown!() -> (then, MyGeometry:Geometry Structure, MouseEvent:"Pointer Event Structure (by ref)") {is=title K2Node_FunctionEntry}
   > n2 = Branch(Condition=Is Mouse Button Down!(Input:"Pointer Event Structure (by ref)"=@n1.MouseEvent, MouseButton:Key Structure=LeftMouseButton) -> (ReturnValue:bool))
   > n3 = For Each Loop(Array=GetAllChildren(self=Get VBox_Content()))
   n4 = DetectDragIfPressed!(
    execute=@n3.Completed
    PointerEvent:"Pointer Event Structure (by ref)"=@n1.MouseEvent
    WidgetDetectingDrag:Widget Object Reference
    DragKey:Key Structure=LeftMouseButton
   ) -> (then, ReturnValue:Event Reply Structure)
   n5 = Is Left Shift Down!(Input:"Input Event Structure (by ref)"=GetInputEventFromPointerEvent!(Event:"Pointer Event Structure (by ref)"=@n1.MouseEvent) -> (ReturnValue:Input Event Structure)) -> (ReturnValue:bool)
   n6 = Get ConfigModeEnabled!(self:EUW Book Case Book Mark Object Reference=Get DraggedBookmark()) -> (ConfigModeEnabled:bool)
   > Return Node(ReturnValue:Event Reply Structure=Select(Option 0:Event Reply Structure=Unhandled(), Option 1:Event Reply Structure=@n4, Index=OR Boolean(A=@n6, B=@n5)) -> (ReturnValue:Event Reply Structure))
   n7 = Cast To EUW_BookCase_BookMark(execute=@n3.LoopBody, Object=@n3)
   > Branch(Condition=IsWidgetInBounds(OwningEvent=Get MouseEvent() -> (MouseEvent:"Pointer Event Structure (by ref)"), Widget=@n7) {is=title K2Node_CallFunction})
   > n8 = Branch(Condition=IsWidgetInBounds(
     OwningEvent=Get MouseEvent()
     Widget=Get BN_EditBookmark!(self:EUW Book Case Book Mark Object Reference=@n7) -> (BN_EditBookmark:Editor Utility Button Object Reference)
    ))
   > Set DraggedBookmark!(execute, DraggedBookmark) -> (then, Output_Get)
   Set DraggedBookmark(execute=@n8.else, DraggedBookmark=@n7)
   Set DraggedBookmark(execute=@n2.else)
   > Return Node(ReturnValue=Unhandled())
   n9 = BreakVector2D(InVec="Absolute to Local (Coordinates)"())
   AND Boolean!(A="float >= float"(A=@n9), B="float >= float"(A=@n9.Y)) -> (ReturnValue)
  ]}
 SetCategoryHeader {nodes [
   n1 = SetCategoryHeader!() -> (then, Name:String) {is=title K2Node_FunctionEntry}
   > Set Text(Text="To Text (String)"(InString=@n1), self=Get Label())
  ]}
 SetConfigurationMode {nodes [
   n1 = SetConfigurationMode!() -> (then, enable:bool=~) {is=title K2Node_FunctionEntry}
   > n2 = For Each Loop(Array=GetAllChildren(self=Get VBox_Content()))
   UI_AverageChildColors(execute=@n2.Completed)
   n3 = Cast To EUW_BookCase_BookMark(execute=@n2.LoopBody, Object=@n2)
   > SetConfigMode!(execute, self:EUW Book Case Book Mark Object Reference=@n3, Active:bool=@n1) -> (then)
  ]}
 UI_AverageChildColors {nodes [
   UI_AverageChildColors() {is=title K2Node_FunctionEntry}
   > n1 = For Each Loop(Array=GetAllChildren(self=Get VBox_Content()))
   n2 = GetChildrenCount!(self:Panel Widget Object Reference=Get VBox_Content()) -> (ReturnValue:int=0)
   n3 = RGB to HSV!(InColor:Linear Color Structure=LinearColor - LinearColor!(
     A:Linear Color Structure=Get cumulative_color() -> (cumulative_color:Linear Color Structure)
     B_R:float=1
     B_G:float=1
     B_B:float=1
     B_A:float=0
    ) -> (ReturnValue:Linear Color Structure) {is=title K2Node_PromotableOperator}) -> (H:float, S:float, V:float, A:float)
   n4 = LinearColor / LinearColor!(
    A:Linear Color Structure=HSV to RGB!(H:float=@n3, S:float=float * float(A=@n3.S, B=0.5) {is=title K2Node_PromotableOperator}, V:float=@n3.V, A:float=1) -> (ReturnValue:Linear Color Structure)
    B_R:float="To Float (Integer)"(InInt=@n2)
    B_G:float="To Float (Integer)"(InInt=@n2)
    B_B:float="To Float (Integer)"(InInt=@n2)
    B_A:float=0
   ) -> (ReturnValue_R:float=~, ReturnValue_G:float=~, ReturnValue_B:float=~, ReturnValue_A:float=~) {is=title K2Node_PromotableOperator}
   Set!(
    execute=@n1.Completed
    BorderColor_SpecifiedColor_R:float=@n4
    BorderColor_SpecifiedColor_G:float=@n4.ReturnValue_G
    BorderColor_SpecifiedColor_B:float=@n4.ReturnValue_B
    BorderColor_SpecifiedColor_A:float=0.7
    BorderColor_ColorUseRule:ESlateColorStylingMode Enum
    self:Expandable Area Object Reference=Get Expandable() -> (Expandable:Editor Utility Expandable Area Object Reference)
   ) -> (then, Output_Get:Slate Color Structure)
   n5 = Cast To EUW_BookCase_BookMark(execute=@n1.LoopBody, Object=@n1)
   n6 = Get!(self:EUW Book Case Book Mark Object Reference=@n5) -> (
    BookmarkData_AssetPath_3_#1:String
    BookmarkData_AssetCategory_5_#2:String
    BookmarkData_BoxColor_15_#3:Color Structure
    BookmarkData_Arguments_10_#4:Array of Strings
    BookmarkData_KeywordArguments_14_#5:Map of Strings to Strings
   )
   > Set cumulative_color(cumulative_color:Linear Color Structure=LinearColor + LinearColor(
     A:Linear Color Structure="To LinearColor (Color)"!(InColor:Color Structure=@n6.BookmarkData_BoxColor_15_#3) -> (ReturnValue:Linear Color Structure)
     B:Linear Color Structure=Get cumulative_color()
    ) -> (ReturnValue:Linear Color Structure) {is=title K2Node_PromotableOperator}) -> (Output_Get:Linear Color Structure)
  ]}
]
variables [
 Bookmark_Dropped {default=()}
 DraggedBookmark {type=EUW_BookCase_BookMark_C, default=null}
]
widgets {
 info {namedSlotCount=2, rootWidgetClass=&Border, widgetCount=4}
 widgets [
  Border_0 {widgetClassPath=&Border, widget=Border_0, properties {brushColor=<rgba 0 0 0 0>, padding=<ltrb 0 1 0 0>}}
  Expandable {
   widgetClassPath=/Script/Blutility.EditorUtilityExpandableArea
   bIsVariable=true
   parent=Border_0
   widget=Expandable
   properties {
    bIsExpanded=true
    bOverride_Cursor=true
    visibility=SelfHitTestInvisible
    areaPadding=<ltrb 1 -1 1 1>
    borderBrush {
     bIsDynamicallyLoaded=false
     drawAs=RoundedBox
     imageType=NoImage
     mirroring=NoMirror
     resourceName=None
     resourceObject=null
     tiling=NoTile
     imageSize=<0 0>
     margin=<ltrb 0 0 0 0>
     outlineSettings {
      bUseBrushTransparency=false
      roundingType=FixedRadius
      width=1
      color {colorUseRule=UseColor_Specified, specifiedColor=<rgba 0.005 0.005 0.005 0.4>}
      cornerRadii=<3 3 0 0>
     }
     tintColor {colorUseRule=UseColor_Specified, specifiedColor=<rgba 0.3 0.3 0.3 1>}
     uVRegion {bIsValid=false, max=<0 0>, min=<0 0>}
    }
    headerPadding=<ltrb 6 4 4 2.7>
    style {
     rolloutAnimationSeconds=0.1
     collapsedImage {
      bIsDynamicallyLoaded=false
      drawAs=Image
      imageType=Vector
      mirroring=NoMirror
      resourceName=../../../Engine/Content/Slate/Starship/Common/caret-right.svg
      resourceObject=null
      tiling=NoTile
      imageSize=<16 16>
      margin=<ltrb 0 0 0 0>
      outlineSettings {
       bUseBrushTransparency=false
       roundingType=HalfHeightRadius
       width=0
       color {colorUseRule=UseColor_Specified, specifiedColor=<rgba 0 0 0 0>}
       cornerRadii=<0 0 0 0>
      }
      tintColor {colorUseRule=UseColor_Specified, specifiedColor=<rgba 0.9 0.9 0.9 1>}
      uVRegion {bIsValid=false, max=<0 0>, min=<0 0>}
     }
     expandedImage {
      bIsDynamicallyLoaded=false
      drawAs=Image
      imageType=Vector
      mirroring=NoMirror
      resourceName=../../../Engine/Content/Slate/Starship/Common/caret-down.svg
      resourceObject=null
      tiling=NoTile
      imageSize=<16 16>
      margin=<ltrb 0 0 0 0>
      outlineSettings {
       bUseBrushTransparency=false
       roundingType=HalfHeightRadius
       width=0
       color {colorUseRule=UseColor_Specified, specifiedColor=<rgba 0 0 0 0>}
       cornerRadii=<0 0 0 0>
      }
      tintColor {colorUseRule=UseColor_Specified, specifiedColor=<rgba 0.9 0.9 0.9 1>}
      uVRegion {bIsValid=false, max=<0 0>, min=<0 0>}
     }
    }
   }
   slot {class=BorderSlot, properties {padding=<ltrb 0 1 0 0>}}
  }
  Label {
   widgetClassPath=&TextBlock
   bIsVariable=true
   namedSlotHost=Expandable
   widget=Label
   properties {
    text=Default
    colorAndOpacity {colorUseRule=UseColor_Specified, specifiedColor=<rgba 1 1 1 0.75>}
    font {
     bForceMonospaced=false
     bMaterialIsStencil=false
     fontMaterial=null
     fontObject=&Roboto
     letterSpacing=0
     monospacedWidth=1
     size=8.25
     skewAmount=0
     typefaceFontName=Bold
     outlineSettings {
      bApplyOutlineToDropShadows=false
      bMiteredCorners=false
      bSeparateFillAlpha=false
      outlineMaterial=null
      outlineSize=0
      outlineColor=<rgba 0 0 0 1>
     }
    }
   }
  }
  VBox_Content {widgetClassPath=&VerticalBox, bIsVariable=true, namedSlotHost=Expandable, widget=VBox_Content}
 ]
}
