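# Editor utility widget for BookShelf, a tool that keeps named presets of colored, categorized asset bookmarks.
# Bookmarks rebuild into category widgets and support adding, editing, removal and drag-drop moves, saved to a SaveGame object after each change.
# A selectable DA_BookShelf data asset (the workflow) supplies shared presets, which merge with locally saved bookmarks.
# On initialization, a Python script sends launch telemetry for non-developer users.
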EditorUtilityWidgetBlueprint /Game/NeverCook/TechArt/BookShelf/EUW_BookShelf_Utility
graphs [
 Add Bookmark {nodes [
   Add Bookmark() {is=title K2Node_FunctionEntry}
   > n1 = Validate Active Path!(execute:Exec) -> (Valid:Exec, Invalid:Exec) {is=title K2Node_MacroInstance}
   PrintString(execute=@n1.Invalid, InString=Failed To Add Bookmark, TextColor=(R=1,G=0.2015,B=0.1154,A=1), Duration=5)
   Set bcd(
    execute=@n1.Valid
    bcd:Map of Names to STRUCT Book Shelf Book Data Structures=Get Local_BookCaseData!(self:BP Quick Widgets Save Game Object Reference=Get SaveGame() -> (SaveGame:BP Quick Widgets Save Game Object Reference)) -> (Local_BookCaseData:Map of Names to STRUCT Book Shelf Book Data Structures)
   ) -> (Output_Get:Map of Names to STRUCT Book Shelf Book Data Structures)
   n2 = Get bcd() -> (bcd:Map of Names to STRUCT Book Shelf Book Data Structures)
   n3 = Get ActivePreset!(self:BP Quick Widgets Save Game Object Reference=Get SaveGame()) -> (ActivePreset:Name=None)
   > n4 = Set bookmarks(bookmarks:Map of Names to STRUCT Book Shelf Asset Data Structures=Break STRUCT Book Shelf Book Data!(STRUCT_BookShelf_BookData:"STRUCT Book Shelf Book Data Structure (by ref)"=Find(TargetMap:Map of Names to STRUCT Book Shelf Book Data Structures=@n2, Key:"Name (by ref)"=@n3) -> (Value:STRUCT Book Shelf Book Data Structure)) -> (Bookmarks_27_#1:Map of Names to STRUCT Book Shelf Asset Data Structures) {is=title K2Node_BreakStruct}) -> (Output_Get:Map of Names to STRUCT Book Shelf Asset Data Structures)
   n5 = Get Text(self:Editable Text Object Reference=Get cfg_Name() -> (cfg_Name:Editor Utility Editable Text Object Reference))
   n6 = Get Text(self=Get cfg_Path() -> (cfg_Path:Editor Utility Editable Text Object Reference))
   n7 = Get Text(self=Get cfg_Category() -> (cfg_Category:Editor Utility Editable Text Object Reference))
   n8 = Get Color() -> (Color:Color Structure)
   > Add!(
    execute
    TargetMap:Map of Names to STRUCT Book Shelf Asset Data Structures=@n4
    Key:"Name (by ref)"=String To Name(InString="To String (Text)"(InText=@n5))
    Value:"STRUCT Book Shelf Asset Data Structure (by ref)"=Make STRUCT Book Shelf Asset Data!(
     AssetPath_3_#1:String="To String (Text)"(InText=@n6)
     AssetCategory_5_#2:String="To String (Text)"(InText=@n7)
     BoxColor_15_#3:Color Structure=LimitColorRange!(In_Color:"Color Structure (by ref)"=@n8) -> (Color:Color Structure)
     Arguments_10_#4:Array of Strings
     KeywordArguments_14_#5:Map of Strings to Strings=()
    ) -> (STRUCT_BookShelf_AssetData:STRUCT Book Shelf Asset Data Structure) {is=title K2Node_MakeStruct}
   ) -> (then) {is=title K2Node_CallFunction}
   > Add!(
    execute
    TargetMap:Map of Names to STRUCT Book Shelf Book Data Structures=Get bcd()
    Key=Get ActivePreset(self=Get SaveGame()) -> (ActivePreset=None)
    Value_Bookmarks_27_#1:Map of Names to STRUCT Book Shelf Asset Data Structures=Get bookmarks() -> (bookmarks:Map of Names to STRUCT Book Shelf Asset Data Structures)
   ) -> (then)
   > Set Local_BookCaseData!(
    execute
    Local_BookCaseData:Map of Names to STRUCT Book Shelf Book Data Structures=Get bcd()
    self:BP Quick Widgets Save Game Object Reference=Get SaveGame()
   ) -> (then, Output_Get:Map of Names to STRUCT Book Shelf Book Data Structures)
   > DoSaveGame()
   "To String (Text)"(InText=Get Text(self=Get cfg_Path()))
  ]}
 ChangeLoadedWorkflow {nodes [
   ChangeLoadedWorkflow() {is=title K2Node_FunctionEntry}
   > PrintString(InString=Updating Bookshelf Workflow, bPrintToScreen=false)
   n1 = Get PersistantSaveGame() -> (PersistantSaveGame:BP Quick Widgets Save Game Object Reference)
   > Set LoadedWorkflow!(
    execute
    LoadedWorkflow:DA Book Shelf Soft Object Reference=Make Soft Reference!(Input:DA Book Shelf Object Reference=Get BookShelfDataAsset() -> (BookShelfDataAsset:DA Book Shelf Object Reference)) -> (Output:DA Book Shelf Soft Object Reference)
    self:BP Quick Widgets Save Game Object Reference=@n1
   ) -> (then, Output_Get:DA Book Shelf Soft Object Reference)
   > SaveGameToSlot!(execute, SaveGameObject:Save Game Object Reference=@n1, SlotName:String=Get BaseSaveName() -> (BaseSaveName:String), UserIndex:int=0) -> (then, ReturnValue:bool)
   > DoSaveGame()
  ]}
 ConstructPresetData {nodes [
   ConstructPresetData() {is=title K2Node_FunctionEntry}
   > n1 = Is Valid(InputObject=Get EditingBookmark() -> (EditingBookmark:EUW Book Case Book Mark Object Reference))
   n2 = Set Text(execute=@n1.Is Not Valid, Text="LOCTEXT(\"Add Bookmark\")", self=Get TX_Configurebookmark() -> (TX_Configurebookmark:Text Object Reference))
   ClearChildren(execute=[@n18.then, @n2.then], self=Get BookmarksList() -> (BookmarksList:Editor Utility Scroll Box Object Reference))
   > Clear!(
    execute
    TargetMap:Map of Strings to EUW Book Case Category Object References=Get ActiveCategories() -> (ActiveCategories:Map of Strings to EUW Book Case Category Object References)
   ) -> (then) {is=title K2Node_CallFunction}
   n3 = GetActivePresetData!() -> (ActivePreset:Name=None, New:STRUCT Book Shelf Book Data Structure)
   > n4 = "For Each Loop (Map)"!(
    execute:Exec
    MapPin:Map of Names to STRUCT Book Shelf Asset Data Structures=Break STRUCT Book Shelf Book Data(STRUCT_BookShelf_BookData=@n3.New)
    BreakPin:Exec
   ) -> (then:Exec, KeyPin:Name, ValuePin:STRUCT Book Shelf Asset Data Structure, CompletedPin:Exec)
   > Set name(name:Name=@n4) -> (Output_Get:Name=None)
   > Set data(data:STRUCT Book Shelf Asset Data Structure=@n4.ValuePin) -> (Output_Get:STRUCT Book Shelf Asset Data Structure)
   n5 = Break STRUCT Book Shelf Asset Data!(STRUCT_BookShelf_AssetData:"STRUCT Book Shelf Asset Data Structure (by ref)"=@n4.ValuePin) -> (
    AssetPath_3_#1:String
    AssetCategory_5_#2:String
    BoxColor_15_#3:Color Structure
    Arguments_10_#4:Array of Strings
    KeywordArguments_14_#5:Map of Strings to Strings
   ) {is=title K2Node_BreakStruct}
   n6 = SelectString(
    A=Default
    B=@n5.AssetCategory_5_#2
    bPickA="Equal, Case Insensitive (String)"(A=@n5.AssetCategory_5_#2)
   )
   n7 = Find(TargetMap:Map of Strings to EUW Book Case Category Object References=Get ActiveCategories(), Key:"String (by ref)"=@n6) -> (Value:EUW Book Case Category Object Reference)
   > n8 = Branch(Condition=@n7.ReturnValue)
   > n9 = Set bookcase_category(bookcase_category:EUW Book Case Category Object Reference=@n7) -> (Output_Get:EUW Book Case Category Object Reference)
   n10 = Construct Object from Class(execute=[@n16.then, @n9.then], Class=/Game/NeverCook/TechArt/BookShelf/Dependencies/Widgets/EUW_BookCase_BookMark.EUW_BookCase_BookMark_C) -> (ReturnValue:EUW Book Case Book Mark Object Reference)
   > AddChild(
    self=Get VBox_Content!(self:EUW Book Case Category Object Reference=Get bookcase_category() -> (bookcase_category:EUW Book Case Category Object Reference)) -> (VBox_Content:Vertical Box Object Reference)
    Content=@n10
   )
   > n11 = Branch(Condition="Equal (Name)"(A=Get active_bookmark_name() -> (active_bookmark_name:Name=None), B=Get name() -> (name:Name=None)))
   > n12 = Set EditingBookmark(EditingBookmark:EUW Book Case Book Mark Object Reference=@n10) -> (Output_Get:EUW Book Case Book Mark Object Reference)
   > n13 = SetConfigMode!(execute, self:EUW Book Case Book Mark Object Reference=@n12, Active:bool=true) -> (then)
   F_Init_Bookmark!(
    execute=[@n13.then, @n11.else]
    self:EUW Book Case Book Mark Object Reference=@n10
    Bookmark:Name=Get name() -> (name=None)
    Data:STRUCT Book Shelf Asset Data Structure=Get data() -> (data:STRUCT Book Shelf Asset Data Structure)
    Parent:EUW Book Case Category Object Reference=Get bookcase_category()
   ) -> (then)
   > BindEvents!(
    execute
    self:EUW Book Case Book Mark Object Reference=@n10
    RemoveSelf:"Delegate (by ref)"=Create Event()
    BeginEditBookmark:"Delegate (by ref)"=Create Event()
    CommitEditBookmark:"Delegate (by ref)"=Create Event()
   ) -> (then)
   n14 = Construct Object from Class(execute=@n8.else, Class=/Game/NeverCook/TechArt/BookShelf/Dependencies/Widgets/EUW_BookCase_Category.EUW_BookCase_Category_C) -> (ReturnValue:EUW Book Case Category Object Reference)
   > Add!(
    execute
    TargetMap:Map of Strings to EUW Book Case Category Object References=Get ActiveCategories()
    Key:"String (by ref)"=@n6
    Value:"EUW Book Case Category Object Reference (by ref)"=@n14
   ) -> (then)
   > n15 = Set bookcase_category(bookcase_category=@n14)
   > AddChild(self=Get BookmarksList(), Content=@n15)
   > Bind Event to Bookmark Dropped(self:EUW Book Case Category Object Reference=@n15, Delegate=Create Event()) {is=title K2Node_AddDelegate}
   > n16 = SetCategoryHeader!(execute, self:EUW Book Case Category Object Reference=Get bookcase_category(), Name:String=@n6) -> (then)
   Set ListName(execute=@n4.CompletedPin, ListName:String="To String (Name)"(InName=@n3)) -> (Output_Get:String)
   > n17 = "For Each Loop (Map)"(MapPin:Map of Strings to EUW Book Case Category Object References=Get ActiveCategories()) -> (KeyPin:String, ValuePin:EUW Book Case Category Object Reference)
   > UI_AverageChildColors(self:EUW Book Case Category Object Reference=@n17.ValuePin)
   Set active_bookmark_name(
    execute=@n1.Is Valid
    active_bookmark_name:Name=Get Bookmark Name!(self:EUW Book Case Book Mark Object Reference=Get EditingBookmark()) -> (Bookmark Name:Name=None)
   ) -> (Output_Get:Name=None)
   > Set active_bookmark_data(active_bookmark_data:STRUCT Book Shelf Asset Data Structure=Get BookmarkData!(self:EUW Book Case Book Mark Object Reference=Get EditingBookmark()) -> (BookmarkData:STRUCT Book Shelf Asset Data Structure)) -> (Output_Get:STRUCT Book Shelf Asset Data Structure)
   > n18 = Set editing_bm(editing_bm:bool=true) -> (Output_Get:bool)
  ]}
 DoSaveGame {nodes [
   DoSaveGame() {is=title K2Node_FunctionEntry}
   > SortSaveGame_ActivePreset()
   > SaveGameToSlot(SaveGameObject=Get SaveGame(), SlotName=Get SaveGame_Name() -> (SaveGame_Name:String), UserIndex=0)
  ]}
 DoesWorkflowHaveSaveData {nodes [
   DoesWorkflowHaveSaveData() {is=title K2Node_FunctionEntry}
   n1 = Get BookShelfDataAsset()
   > n2 = Is Valid(InputObject=@n1)
   Return Node!(execute=@n2.Is Not Valid, ReturnValue:bool=false)
   > n3 = DoesSaveGameExist!(execute, SlotName:String=Append(A=BookShelf_, B=GetObjectName(Object=@n1)), UserIndex:int=0) -> (then, ReturnValue:bool)
   > Return Node(ReturnValue=@n3)
  ]}
 Edit Bookmark {nodes [
   Edit Bookmark() {is=title K2Node_FunctionEntry}
   > n1 = Validate Active Path()
   PrintString(execute=@n1.Invalid, InString=Failed To Edit Bookmark, TextColor=(R=1,G=0.2015,B=0.1154,A=1), Duration=5)
   Set bcd(execute=@n1.Valid, bcd=Get Local_BookCaseData(self=Get SaveGame()))
   n2 = Get bcd()
   n3 = Get ActivePreset(self=Get SaveGame()) -> (ActivePreset=None)
   > Set bookmarks(bookmarks=Break STRUCT Book Shelf Book Data(STRUCT_BookShelf_BookData=Find(TargetMap:Map of Names to STRUCT Book Shelf Book Data Structures=@n2, Key:"Name (by ref)"=@n3) -> (Value:STRUCT Book Shelf Book Data Structure)))
   > Is Valid(InputObject=Get EditingBookmark())
   n4 = Get Text(self=Get cfg_Name())
   > Set new_name(new_name:Name=String To Name(InString="To String (Text)"(InText=@n4))) -> (Output_Get:Name=None)
   n5 = Get EditingBookmark()
   > Set old_name(old_name:Name=Get Bookmark Name(self=@n5) -> (Bookmark Name=None)) -> (Output_Get:Name=None)
   > Set new_color(new_color:Color Structure=LimitColorRange(In_Color=Get Color())) -> (Output_Get:Color Structure)
   n6 = Get new_name() -> (new_name:Name=None)
   n7 = Get bookmarks()
   > Set was_bookmark_renamed(was_bookmark_renamed:bool=NOT Boolean(A=Contains!(TargetMap:Map of Names to STRUCT Book Shelf Asset Data Structures=@n7, Key:"Name (by ref)"=@n6) -> (ReturnValue))) -> (Output_Get:bool)
   n8 = Get Text(self=Get cfg_Path())
   n9 = Get Text(self=Get cfg_Category())
   n10 = Get!(self:EUW Book Case Book Mark Object Reference=@n5) -> (
    BookmarkData_AssetPath_3_#1:String
    BookmarkData_AssetCategory_5_#2:String
    BookmarkData_BoxColor_15_#3:Color Structure
    BookmarkData_Arguments_10_#4:Array of Strings
    BookmarkData_KeywordArguments_14_#5:Map of Strings to Strings
   )
   > Add(
    TargetMap:Map of Names to STRUCT Book Shelf Asset Data Structures=Get bookmarks()
    Key:"Name (by ref)"=Get new_name() -> (new_name=None)
    Value:"STRUCT Book Shelf Asset Data Structure (by ref)"=Make STRUCT Book Shelf Asset Data(
     AssetPath_3_#1="To String (Text)"(InText=@n8)
     AssetCategory_5_#2="To String (Text)"(InText=@n9)
     BoxColor_15_#3=Get new_color() -> (new_color:Color Structure)
     Arguments_10_#4=@n10.BookmarkData_Arguments_10_#4
     KeywordArguments_14_#5=@n10.BookmarkData_KeywordArguments_14_#5
    )
   )
   > n11 = Sequence()
   > Branch(Condition=Get was_bookmark_renamed() -> (was_bookmark_renamed:bool))
   > n12 = "For Each Loop (Map)"(MapPin:Map of Names to STRUCT Book Shelf Asset Data Structures=Get bookmarks()) -> (KeyPin:Name, ValuePin:STRUCT Book Shelf Asset Data Structure)
   > n13 = Branch(Condition="Equal (Name)"(A=Get old_name() -> (old_name:Name=None), B=@n12))
   > Add(
    TargetMap=Get new_bookmarks() -> (new_bookmarks:Map of Names to STRUCT Book Shelf Asset Data Structures)
    Key=Get new_name() -> (new_name=None)
    Value=@n12.ValuePin
   )
   Add(execute=@n13.else, TargetMap=Get new_bookmarks(), Key=@n12, Value=@n12.ValuePin)
   Set bookmarks(execute=@n12.CompletedPin, bookmarks=Get new_bookmarks())
   Add!(
    execute=@n11.then_1
    TargetMap:Map of Names to STRUCT Book Shelf Book Data Structures=Get bcd()
    Key=Get ActivePreset(self=Get SaveGame()) -> (ActivePreset=None)
    Value_Bookmarks_27_#1:Map of Names to STRUCT Book Shelf Asset Data Structures=Get bookmarks()
   ) -> (then)
   > Set Local_BookCaseData(Local_BookCaseData=Get bcd(), self=Get SaveGame())
   > Set Bookmark Name!(execute, Bookmark Name:Name=Get new_name() -> (new_name=None), self:EUW Book Case Book Mark Object Reference=Get EditingBookmark()) -> (then, Output_Get:Name=None)
   n14 = Get EditingBookmark()
   Is Valid(InputObject=@n14)
   > Event_EditBookmark!(execute, self, Item:EUW Book Case Book Mark Object Reference=@n14) -> (then)
  ]}
 EventGraph {nodes [
   n1 = "OnTextCommitted (Editable Text) (Title)"!() -> (OutputDelegate, then, Text:"Text (by ref)", CommitMethod:ETextCommit Enum=Default) {is=title K2Node_ComponentBoundEvent}
   > Branch(Condition="Equal (Enum)"(A:ETextCommit Enum=@n1.CommitMethod, B:ETextCommit Enum=OnEnter))
   > Update Active Preset Name!(execute, self, Name:String="To String (Text)"(InText=@n1.Text)) -> (then)
   > ConstructPresetData() {is=title K2Node_CallFunction}
   > FindValidSaveDataPresets()
   "On Opening (CB_WorkflowSelect)"() {is=title K2Node_ComponentBoundEvent}
   n2 = Get BookShelfDataAsset()
   > Branch(Condition=AND Boolean(A=Get isAltHeld() -> (isAltHeld:bool), B=IsValid(Object=@n2)))
   > OpenEditorForAssets!(
    execute
    self:Asset Editor Subsystem Object Reference=Get AssetEditorSubsystem()
    Assets:Array of Object References=Make Array("[0]":Object Reference=@n2) -> (Array:Array of Object References)
    OpenedMethod:EAssetTypeActivationOpenedMethod Enum=Edit
   ) -> (then, ReturnValue:bool)
   > Set isAltHeld(isAltHeld:bool=false) -> (Output_Get:bool)
   "On Clicked (BTN_WSLoadBookmarkConfig)"() {is=title K2Node_ComponentBoundEvent}
   > SetActiveWidgetIndex(self=Get ConfigSwitcher() -> (ConfigSwitcher:Widget Switcher Object Reference), Index=0)
   "On Clicked (BTN_WSLoadDataAssetConfig)"() {is=title K2Node_ComponentBoundEvent}
   > SetActiveWidgetIndex(self=Get ConfigSwitcher(), Index=1)
   "On Clicked (BTN_GetOpenFolder)"() {is=title K2Node_ComponentBoundEvent}
   > n3 = GetOpenFolderPath!(execute, self) -> (then, folder path:String)
   > Set Text(
    Text="To Text (String)"(InString=@n3)
    self:Text Box Object Reference=Get TB_BookCaseFolderPath() -> (TB_BookCaseFolderPath:Editor Utility Editable Text Box Object Reference)
   )
   n4 = "On Text Committed (TB_BookCaseAssetName)"!() -> (OutputDelegate, then, Text:"Text (by ref)", CommitMethod:ETextCommit Enum=Default) {is=title K2Node_ComponentBoundEvent}
   n5 = "To String (Text)"(InText=@n4.Text)
   > n6 = Branch(Condition=IsEmpty(InString=@n5))
   n7 = Branch(execute=@n6.else, Condition=StartsWith(SourceString=@n5, InPrefix=DA_))
   n8 = ToUpper!(SourceString:String=Left!(SourceString:String=@n5, Count:int=1) -> (ReturnValue:String)) -> (ReturnValue:String)
   n9 = RightChop!(SourceString:String=@n5, Count:int=1) -> (ReturnValue:String)
   Set Text(
    execute=@n7.else
    Text="To Text (String)"(InString=Append!(A=DA_, B=@n8, C:String=@n9) -> (ReturnValue))
    self=Get TB_BookCaseAssetName() -> (TB_BookCaseAssetName:Editor Utility Editable Text Box Object Reference)
   )
   "On Clicked (BN_SaveNewBookCase)"() {is=title K2Node_ComponentBoundEvent}
   > Set BookShelfDataAsset(BookShelfDataAsset:DA Book Shelf Object Reference) -> (Output_Get:DA Book Shelf Object Reference)
   > Delay!(execute, WorldContextObject:Object Reference, Duration=0.01) -> (then)
   > SaveNewBookCaseDataAsset()
   "On Clicked (BTN_AddBookmark)"() {is=title K2Node_ComponentBoundEvent}
   n10 = Get Text(self:Text Object Reference=Get TX_Configurebookmark())
   > n11 = Branch(Condition=Contains!(SearchIn:String="To String (Text)"(InText=@n10), Substring:String=Update, bUseCase:bool=true, bSearchFromEnd:bool) -> (ReturnValue))
   > n12 = Edit Bookmark() {is=title K2Node_CallFunction}
   ConstructPresetData(execute=[@n13.then, @n12.then])
   > DoSaveGame() {is=title K2Node_CallFunction}
   n13 = Add Bookmark(execute=@n11.else) {is=title K2Node_CallFunction}
   "On Clicked (BTN_GetSelectedAsset)"() {is=title K2Node_ComponentBoundEvent}
   > PrepCreateBookmark()
   > SetBookmarkConfigurationMode!(execute, self, enable:bool=false) -> (then)
   "On Clicked (BTN_ClearSaveData)"() {is=title K2Node_ComponentBoundEvent}
   > n14 = WarnBeforeResettingData!(execute, self) -> (then, Do:bool)
   > Branch(Condition=@n14)
   > n15 = Construct Object from Class(Class=/Game/NeverCook/TechArt/BookShelf/Dependencies/BP_QuickWidgets_SaveGame.BP_QuickWidgets_SaveGame_C) -> (ReturnValue:BP Quick Widgets Save Game Object Reference)
   > Set SaveGame(SaveGame:BP Quick Widgets Save Game Object Reference=@n15) -> (Output_Get:BP Quick Widgets Save Game Object Reference)
   > DoSaveGame()
   > ConstructPresetData()
   > FindValidSaveDataPresets()
   "On Property Changed (PV_Color)"!() -> (OutputDelegate, then, PropertyName:Name=None) {is=title K2Node_ComponentBoundEvent}
   n16 = Break Color!(Color:"Color Structure (by ref)"=Get Color()) -> (B:Byte, G:Byte, R:Byte, A:Byte) {is=title K2Node_BreakStruct}
   > Set!(execute, Color_B:Byte=@n16, Color_G:Byte=@n16.G, Color_R:Byte=@n16.R, Color_A:Byte=255) -> (then, Output_Get:Color Structure)
   n17 = "On Selection Changed (CB_WorkflowSelect)"!() -> (OutputDelegate, then, SelectedItem:String, SelectionType:ESelectInfo Enum) {is=title K2Node_ComponentBoundEvent}
   > Branch(Condition=AND Boolean(
     A="Not Equal (Enum)"!(A:ESelectInfo Enum=@n17.SelectionType, B:ESelectInfo Enum=OnNavigation) -> (ReturnValue:bool)
     B="Not Equal (Enum)"(A=@n17.SelectionType, B=Direct)
    ))
   > UpdatedWorkflowSelection!(execute, self, WorkflowName:String=@n17.SelectedItem) -> (then)
   "On Clicked (BN_SavePreset)"() {is=title K2Node_ComponentBoundEvent}
   > SaveNamedPreset()
   n18 = "On Selection Changed (CB_ProfilesSelect)"!() -> (OutputDelegate, then, SelectedItem:String, SelectionType:ESelectInfo Enum) {is=title K2Node_ComponentBoundEvent}
   > Branch(Condition="Not Equal (Enum)"(A=@n18.SelectionType, B=Direct))
   > n19 = Set ActivePreset!(execute, ActivePreset:Name=String To Name(InString=@n18.SelectedItem), self:BP Quick Widgets Save Game Object Reference=Get SaveGame()) -> (then, Output_Get:Name=None)
   DoSaveGame(execute=[@n19.then, @n28.then])
   > ConstructPresetData()
   > SetBookmarkConfigurationMode(enable=false)
   > Set EditingBookmark()
   > FindValidSaveDataPresets()
   ReloadDefaults() {is=title K2Node_CustomEvent}
   > LoadDefaultPresets()
   > ConstructPresetData()
   > FindValidSaveDataPresets()
   Refresh Data() {is=title K2Node_CustomEvent}
   > F_PopulateBookShelfData()
   > ConstructPresetData()
   > FindValidSaveDataPresets()
   n20 = Load Workflow!() -> (OutputDelegate, then, BookShelfDataAsset:DA Book Shelf Object Reference) {is=title K2Node_CustomEvent}
   > n21 = Set BookShelfDataAsset(BookShelfDataAsset=@n20.BookShelfDataAsset)
   n22 = "To String (Name)"(InName=Get BookShelf Name!(self:DA Book Shelf Object Reference=@n21) -> (BookShelf Name:Name=None))
   > n23 = SelectNewWorkflow!(execute, self, Option:String=@n22) -> (then, Changed:bool)
   > Branch(Condition=@n23)
   > Delay(Duration=0.05)
   > WorkflowComboBoxCleanup!(execute, self, ActiveLine:String=@n22) -> (then)
   > Refresh Data() {is=title K2Node_CallFunction}
   n24 = Load Alternative Save!() -> (OutputDelegate, then, AlternativeSaveName:String=EUW_QuickWidgets_Bookmarks) {is=title K2Node_CustomEvent}
   > Set BaseSaveName(BaseSaveName:String=@n24.AlternativeSaveName) -> (Output_Get:String)
   > LoadOrCreateSaveDataObject()
   > InitDataAssetDefaults()
   > Refresh Data()
   n25 = Load Preset!() -> (OutputDelegate, then, ActivePreset:Name=None, PresetIndex:int=-1) {is=title K2Node_CustomEvent}
   > n26 = Set ActivePreset(ActivePreset=@n25.ActivePreset, self=Get SaveGame()) -> (Output_Get=None)
   > n27 = SelectNewPreset!(execute, self, PresetName:Name=@n26, PresetIndex:int=@n25.PresetIndex) -> (then, SelectedPreset:String, Update Preset:bool)
   > Branch(Condition=@n27.Update Preset)
   > n28 = Set ActivePreset(ActivePreset=String To Name(InString=@n27), self=Get SaveGame()) -> (Output_Get=None)
   Event PreConstruct()
   > SetObject(self=Get PV_Color() -> (PV_Color:Single Property View Object Reference), NewObject=Self-Reference())
   > LoadOrCreateSaveDataObject()
   > InitDataAssetDefaults()
   > Refresh Data()
   Event Construct()
   > Set bIsExpanded!(
    execute
    bIsExpanded:bool=false
    self:Expandable Area Object Reference=Get EXP_Configuration() -> (EXP_Configuration:Editor Utility Expandable Area Object Reference)
   ) -> (then, Output_Get:bool)
   Event OnInitialized()
   > DoOnce MultiInput!(A In:Exec, Reset In:Exec) -> (A Out:Exec, Reset Out:Exec) {is=title K2Node_DoOnceMultiInput}
   > Execute Python Script(PythonScript="\r\ndevs = [\"alex.mannion\"]\r\ndata_key = \"LFN_BookShelf_Utility.launch\"\r\n\r\n## ----- ##\r\nimport os\r\nimport getpass\r\n\r\nusername = getpass.getuser()\t## Get the username \r\nis_dev = username in devs\r\n##os.environ['UNPIPE_TELEMETRY_OFF'] = is_dev\r\n\r\nif is_dev:\r\n    print(\"Developer Useage Determined\")    \r\nelse:\r\n    from unpipe.telemetry import BasicEvent\r\n    print(\"Logging Useage Data\")\r\n    BasicEvent(data_key).send()")
   > Bind Event to On Clicked(self:Button Object Reference=Get BN_RemovePreset() -> (BN_RemovePreset:Editor Utility Button Object Reference), Delegate=Create Event()) {is=title K2Node_AddDelegate}
   Event OnFocusLost!() -> (OutputDelegate, then, InFocusEvent:Focus Event Structure) {is=title K2Node_Event}
   > Set isAltHeld(isAltHeld=false)
  ]}
 Event_BookmarkMoved {nodes [
   Event_BookmarkMoved!() -> (then, Operation:BP Book Case Drag Drop Object Reference) {is=title K2Node_FunctionEntry}
   > n1 = Sequence!(execute) -> (then_0, then_1, then_2:Exec)
   n2 = Get Operation() -> (Operation:BP Book Case Drag Drop Object Reference)
   Set loc_draggedBookmarkName(
    execute=@n1.then_1
    loc_draggedBookmarkName:Name=Get Bookmark Name(self=Get Bookmark!(self:BP Book Case Drag Drop Object Reference=@n2) -> (Bookmark:EUW Book Case Book Mark Object Reference)) -> (Bookmark Name=None)
   ) -> (Output_Get:Name=None)
   n3 = Get Operation()
   n4 = Get Operation()
   > n5 = Branch(Condition="Equal (Object)"(
     A=Get NeighborBookmark!(self:BP Book Case Drag Drop Object Reference=@n3) -> (NeighborBookmark:EUW Book Case Book Mark Object Reference)
     B=Get Bookmark(self=@n4)
    ))
   n6 = Get NeighborBookmark(self=Get Operation())
   n7 = Is Valid(exec=@n5.else, InputObject=@n6)
   > n8 = Set loc_neighborName(loc_neighborName:Name=Get Bookmark Name(self=@n6) -> (Bookmark Name=None)) -> (Output_Get:Name=None)
   n9 = "For Each Loop (Map)"(
    execute=[@n8.then, @n7.Is Not Valid]
    MapPin=Break STRUCT Book Shelf Book Data(STRUCT_BookShelf_BookData=GetActivePresetData() -> (ActivePreset=None).New)
   )
   > n10 = Branch(Condition="Equal (Name)"(A=Get loc_draggedBookmarkName() -> (loc_draggedBookmarkName:Name=None), B=@n9))
   n11 = Branch(execute=@n10.else, Condition="Equal (Name)"(A=Get loc_neighborName() -> (loc_neighborName:Name=None), B=None1209388120431))
   n12 = Break STRUCT Book Shelf Asset Data!(STRUCT_BookShelf_AssetData=@n9.ValuePin) -> (AssetCategory_5_#2)
   n13 = Get loc_new_category() -> (loc_new_category:String)
   n14 = Get loc_foundBM() -> (loc_foundBM:bool)
   > n15 = Branch(Condition=AND Boolean(A="Equal Exactly (String)"(A=@n12, B=@n13), B=NOT Boolean(A=@n14)))
   n16 = Get Bookmark(self=Get Operation())
   > Add!(
    execute
    TargetMap:Map of Names to STRUCT Book Shelf Asset Data Structures=Get local_bookmarks() -> (local_bookmarks:Map of Names to STRUCT Book Shelf Asset Data Structures)
    Key=Get Bookmark Name(self=@n16) -> (Bookmark Name=None)
    Value:"STRUCT Book Shelf Asset Data Structure (by ref)"=Get BookmarkData(self=@n16)
   ) -> (then)
   > n17 = Set loc_foundBM(loc_foundBM:bool=true) -> (Output_Get:bool)
   Add(execute=[@n17.then, @n15.else], TargetMap=Get local_bookmarks(), Key=@n9, Value=@n9.ValuePin)
   Add(execute=@n11.else, TargetMap=Get local_bookmarks(), Key=@n9, Value=@n9.ValuePin)
   > Branch(Condition="Equal (Name)"(A=@n9, B=Get loc_neighborName() -> (loc_neighborName=None)))
   n18 = Get Bookmark(self=Get Operation())
   > Add(TargetMap=Get local_bookmarks(), Key=Get Bookmark Name(self=@n18) -> (Bookmark Name=None), Value=Get BookmarkData(self=@n18))
   n19 = Get SaveGame()
   Add!(
    execute=@n9.CompletedPin
    TargetMap:Map of Names to STRUCT Book Shelf Book Data Structures=Get Local_BookCaseData(self=@n19)
    Key=Get ActivePreset(self=@n19) -> (ActivePreset=None)
    Value_Bookmarks_27_#1:Map of Names to STRUCT Book Shelf Asset Data Structures=Get local_bookmarks()
   ) -> (then)
   ConstructPresetData(execute=@n1.then_2)
   > DoSaveGame()
   n20 = Get Operation()
   n21 = Get Text(self=Get Label!(self:EUW Book Case Category Object Reference=Get NewCategory!(self:BP Book Case Drag Drop Object Reference=@n20) -> (NewCategory:EUW Book Case Category Object Reference)) -> (Label:Text Object Reference))
   "SetText (Editable Text)"!(execute=@n1.then_0, self:Editable Text Object Reference=Get cfg_Category(), InText:Text=@n21) -> (then)
   n22 = "To String (Text)"(InText=@n21)
   > Set loc_new_category(loc_new_category:String=@n22) -> (Output_Get:String)
   n23 = Get Operation()
   > Set members in STRUCT Book Shelf Asset Data!(
    execute:Exec
    StructRef:"STRUCT Book Shelf Asset Data Structure (by ref)"=Get BookmarkData(self=Get Bookmark(self=@n23))
    AssetCategory_5_#2:String=@n22
   ) -> (then:Exec, StructOut:"STRUCT Book Shelf Asset Data Structure (by ref)") {is=title K2Node_SetFieldsInStruct}
   > Edit Bookmark()
  ]}
 Event_ConfigureAsset {nodes [
   n1 = Event_ConfigureAsset!() -> (then, Item:EUW Book Case Book Mark Object Reference) {is=title K2Node_FunctionEntry}
   > Set EditingBookmark(EditingBookmark=@n1)
   n2 = Get(self=@n1)
   > Set Color(Color:Color Structure=@n2.BookmarkData_BoxColor_15_#3) -> (Output_Get:Color Structure)
   > Set Text(
    Text="To Text (Name)"!(InName:Name=Get Bookmark Name(self=@n1) -> (Bookmark Name=None)) -> (ReturnValue:Text)
    self:Editable Text Object Reference=Get cfg_Name()
   )
   > Set Text(Text=To Text!(Value:String=@n2) -> (ReturnValue:Text) {is=title K2Node_GenericToText}, self=Get cfg_Path())
   > Set Text(Text=To Text(Value=@n2.BookmarkData_AssetCategory_5_#2), self=Get cfg_Category())
  ]}
 Event_EditBookmark {nodes [
   n1 = Event_EditBookmark!() -> (then, Item:EUW Book Case Book Mark Object Reference) {is=title K2Node_FunctionEntry}
   > n2 = Branch(Condition="Equal (Object)"(A=@n1, B=Get EditingBookmark()))
   > Set EditingBookmark()
   > SetBookmarkConfigurationMode(enable=false)
   > Set Text(Text="LOCTEXT(\"Add Bookmark\")", self:Text Object Reference=Get TX_Configurebookmark())
   n3 = Set EditingBookmark(execute=@n2.else, EditingBookmark=@n1)
   > SetBookmarkConfigurationMode(enable=false)
   > SetConfigMode(self=@n3, Active=true)
   n4 = Get(self=@n3)
   > Set Color(Color=@n4.BookmarkData_BoxColor_15_#3)
   > Set Text(Text="To Text (Name)"(InName=Get Bookmark Name(self=@n3) -> (Bookmark Name=None)), self:Editable Text Object Reference=Get cfg_Name())
   > Set Text(Text=To Text(Value=@n4), self=Get cfg_Path())
   > Set Text(Text=To Text(Value=@n4.BookmarkData_AssetCategory_5_#2), self=Get cfg_Category())
   > SetIsExpanded!(execute, self:Expandable Area Object Reference=Get EXP_Configuration(), IsExpanded:bool=true) -> (then)
   > Set Text(Text="LOCTEXT(\"Update Bookmark\")", self:Text Object Reference=Get TX_Configurebookmark())
  ]}
 Event_RemoveBookmark_Clean {nodes [
   n1 = Event_RemoveBookmark_Clean!() -> (then, Item:EUW Book Case Book Mark Object Reference) {is=title K2Node_FunctionEntry}
   n2 = GetActivePresetData() -> (ActivePreset=None)
   n3 = Get SaveGame()
   > Set bookmarks(bookmarks=Find!(TargetMap=Get Local_BookCaseData(self=@n3), Key=@n2) -> (Value_Bookmarks_27_#1:Map of Names to STRUCT Book Shelf Asset Data Structures, ReturnValue))
   > n4 = Sequence!(execute) -> (then_0, then_1)
   n5 = Get ParentCategory!(self:EUW Book Case Book Mark Object Reference=@n1) -> (ParentCategory:EUW Book Case Category Object Reference)
   n6 = Get VBox_Content(self=@n5)
   > RemoveChild(self=@n6, Content=@n1)
   > n7 = Branch(Condition=HasAnyChildren!(self:Panel Widget Object Reference=@n6) -> (ReturnValue:bool))
   RemoveChild(execute=@n7.else, self=Get BookmarksList(), Content=@n5)
   n8 = Get Bookmark Name(self=@n1) -> (Bookmark Name=None)
   Branch(
    execute=@n4.then_1
    Condition=Contains!(
     TargetMap:Map of Names to STRUCT Book Shelf Asset Data Structures=Break STRUCT Book Shelf Book Data(STRUCT_BookShelf_BookData=@n2.New)
     Key:"Name (by ref)"=@n8
    ) -> (ReturnValue)
   )
   > Remove!(execute, TargetMap:Map of Names to STRUCT Book Shelf Asset Data Structures=Get bookmarks(), Key:"Name (by ref)"=@n8) -> (then, ReturnValue:bool)
   > Add(
    TargetMap=Get Local_BookCaseData(self=Get SaveGame())
    Key=GetActivePresetData() -> (ActivePreset=None)
    Value_Bookmarks_27_#1=Get bookmarks()
   )
   > DoSaveGame()
  ]}
 F_AppendNewBookShelfData {nodes [
   F_AppendNewBookShelfData() {is=title K2Node_FunctionEntry}
   n1 = Get BookShelfDataAsset()
   n2 = Get Presets!(self:DA Book Shelf Object Reference=@n1) -> (Presets:Map of Names to STRUCT Book Shelf Book Data Structures)
   > PrintText(InText=Format Text!(
     Format="LOCTEXT(\"Loading {pre} presets for BookShelf Data Asset: {bookshelf}\")"
     pre:int=Length!(TargetMap:Map of Names to STRUCT Book Shelf Book Data Structures=@n2) -> (ReturnValue:int=0) {is=title K2Node_CallFunction}
     bookshelf:Name=Get BookShelf Name(self=@n1) -> (BookShelf Name=None)
    ) -> (Result))
   > n3 = "For Each Loop (Map)"(MapPin:Map of Names to STRUCT Book Shelf Book Data Structures=Get Presets(self=Get BookShelfDataAsset())) -> (ValuePin:STRUCT Book Shelf Book Data Structure)
   n4 = Get Local_BookCaseData(self=Get SaveGame())
   > n5 = Branch(Condition=Contains(TargetMap:Map of Names to STRUCT Book Shelf Book Data Structures=@n4, Key=@n3))
   n6 = Get SaveGame()
   > Set out_bms(out_bms:Map of Names to STRUCT Book Shelf Asset Data Structures=Find(TargetMap=Get Local_BookCaseData(self=@n6), Key=@n3)) -> (Output_Get:Map of Names to STRUCT Book Shelf Asset Data Structures)
   n7 = Break STRUCT Book Shelf Book Data(STRUCT_BookShelf_BookData=@n3.ValuePin)
   > n8 = "For Each Loop (Map)"(MapPin:Map of Names to STRUCT Book Shelf Asset Data Structures=@n7) -> (ValuePin:STRUCT Book Shelf Asset Data Structure)
   > n9 = Branch(Condition=Contains(
     TargetMap:Map of Names to STRUCT Book Shelf Asset Data Structures=Get out_bms() -> (out_bms:Map of Names to STRUCT Book Shelf Asset Data Structures)
     Key=@n8
    ))
   Add!(
    execute=@n9.else
    TargetMap:Map of Names to STRUCT Book Shelf Asset Data Structures=Get out_bms()
    Key=@n8
    Value:"STRUCT Book Shelf Asset Data Structure (by ref)"=@n8.ValuePin
   ) -> (then)
   Add!(
    execute=@n8.CompletedPin
    TargetMap:Map of Names to STRUCT Book Shelf Book Data Structures=Get Local_BookCaseData(self=Get SaveGame())
    Key=@n3
    Value_Bookmarks_27_#1:Map of Names to STRUCT Book Shelf Asset Data Structures=Get out_bms()
   ) -> (then)
   Add(execute=@n5.else, TargetMap=@n4, Key=@n3, Value_Bookmarks_27_#1=@n7)
   DoSaveGame(execute=@n3.CompletedPin)
  ]}
 F_PopulateBookShelfData {nodes [
   F_PopulateBookShelfData() {is=title K2Node_FunctionEntry}
   n1 = Get BookShelfDataAsset()
   > Branch(Condition="Not Equal (Object)"(A:Object Reference=@n1, B:Object Reference) -> (ReturnValue:bool=~) {is=title K2Node_PromotableOperator})
   n2 = Get BookShelfDataAsset()
   n3 = Get Presets(self=@n2)
   > PrintText(InText=Format Text(
     Format="LOCTEXT(\"Loading {pre} presets for BookShelf Data Asset: {bookshelf}\")"
     pre=Length(TargetMap=@n3) -> (ReturnValue=0)
     bookshelf=Get BookShelf Name(self=@n2) -> (BookShelf Name=None)
    ))
   > n4 = "For Each Loop (Map)"(MapPin:Map of Names to STRUCT Book Shelf Book Data Structures=Get Presets(self=@n1)) -> (ValuePin:STRUCT Book Shelf Book Data Structure)
   > Set loaded_bookmarks(loaded_bookmarks:Map of Names to STRUCT Book Shelf Asset Data Structures=Break STRUCT Book Shelf Book Data(STRUCT_BookShelf_BookData=@n4.ValuePin)) -> (Output_Get:Map of Names to STRUCT Book Shelf Asset Data Structures)
   n5 = Get SaveGame()
   > n6 = "For Each Loop (Map)"(MapPin:Map of Names to STRUCT Book Shelf Asset Data Structures=Find(TargetMap=Get Local_BookCaseData(self=@n5), Key=@n4)) -> (ValuePin:STRUCT Book Shelf Asset Data Structure)
   > n7 = Branch(Condition=Contains(TargetMap=Get loaded_bookmarks() -> (loaded_bookmarks:Map of Names to STRUCT Book Shelf Asset Data Structures), Key=@n6))
   Add!(
    execute=@n7.else
    TargetMap:Map of Names to STRUCT Book Shelf Asset Data Structures=Get loaded_bookmarks()
    Key=@n6
    Value:"STRUCT Book Shelf Asset Data Structure (by ref)"=@n6.ValuePin
   ) -> (then)
   Add!(
    execute=@n6.CompletedPin
    TargetMap:Map of Names to STRUCT Book Shelf Book Data Structures=Get Local_BookCaseData(self=Get SaveGame())
    Key=@n4
    Value_Bookmarks_27_#1:Map of Names to STRUCT Book Shelf Asset Data Structures=Get loaded_bookmarks()
   ) -> (then)
   DoSaveGame(execute=@n4.CompletedPin)
   > n8 = GetPathNameForLoadedAsset(LoadedAsset=Get BookShelfDataAsset())
   n9 = Split(SourceString=@n8, InStr=., SearchDir=FromEnd)
   > Set Text(
    Text="To Text (String)"(InString=Split(SourceString=@n9, InStr=/, SearchDir=FromEnd))
    self:Text Box Object Reference=Get TB_BookCaseFolderPath()
   )
   > Set Text(Text="To Text (String)"(InString=@n9.RightS), self=Get TB_BookCaseAssetName())
  ]}
 FindValidSaveDataPresets {nodes [
   FindValidSaveDataPresets() {is=title K2Node_FunctionEntry}
   > n1 = Keys!(execute, TargetMap:Map of Names to STRUCT Book Shelf Book Data Structures=Get Local_BookCaseData(self=Get SaveGame())) -> (then, Keys:Array of Names)
   > n2 = For Each Loop(Array:Array of Names=@n1) -> (Array Element:Name)
   n3 = Get CB_ProfilesSelect() -> (CB_ProfilesSelect:Editor Utility Combo Box String Object Reference)
   ClearOptions(execute=@n2.Completed, self:"Combo Box (String) Object Reference"=@n3)
   > n4 = For Each Loop(Array:Array of Strings=Get presets() -> (presets:Array of Strings)) -> (Array Element:String)
   RefreshOptions(execute=@n4.Completed, self:"Combo Box (String) Object Reference"=@n3)
   n5 = GetActivePresetData() -> (ActivePreset=None)
   > SetSelectedOption!(
    execute
    self:"Combo Box (String) Object Reference"=@n3
    Option:String=Append!(A="​Preset - ", B="To String (Name)"(InName=@n5)) -> (ReturnValue)
   ) -> (then)
   AddOption!(execute=@n4.LoopBody, self:"Combo Box (String) Object Reference"=@n3, Option:String=@n4) -> (then)
   n6 = "To String (Name)"(InName=@n2)
   PrintString(execute=@n2.LoopBody, InString=@n6)
   n7 = "To String (Name)"(InName=GetActivePresetData() -> (ActivePreset=None))
   > n8 = Branch(Condition="Equal Exactly (String)"(A=@n7, B=@n6))
   > Add Unique(TargetArray=Get presets(), NewItem=Append(A="​Preset - ", B=@n7))
   Add Unique(execute=@n8.else, TargetArray=Get presets(), NewItem=@n6)
   n9 = GetProjectContentDirectory!() -> (ReturnValue:String)
   Append(A=Split(SourceString=LeftChop(SourceString=@n9, Count=1), InStr=/, SearchDir=FromEnd), B=/Saved/SaveGames)
  ]}
 Get Plugin Path From Package Path {nodes [
   n1 = Get Plugin Path From Package Path!() -> (then, PluginPath:String) {is=title K2Node_FunctionEntry}
   > n2 = Branch(Condition=StartsWith(SourceString=@n1, InPrefix=/All/Game/))
   > Return Node!(execute, PackagePath:String=Append(A=/Game/, B=Split(SourceString=@n1, InStr=/Game/, SearchDir=FromStart).RightS))
   n3 = Branch(execute=@n2.else, Condition=NOT Boolean(A=StartsWith(SourceString=@n1, InPrefix=/Game/)))
   > n4 = For Each Loop(Array=ParseIntoArray(SourceString=Replace(SourceString=@n1, From="\\", To=/), Delimiter=/))
   Return Node(execute=@n4.Completed, PackagePath=Get Path() -> (Path:String))
   > n5 = GetPluginBaseDir!(execute, PluginName:String=@n4) -> (then, OutBaseDir:String, ReturnValue:bool)
   > n6 = Branch(Condition=@n5.ReturnValue)
   > n7 = Set Start(Start:bool=true) -> (Output_Get:bool)
   Branch(execute=[@n6.else, @n7.then], Condition=Get Start() -> (Start:bool))
   > Set Path(Path:String=Append!(A=Get Path(), B=/, C:String=@n4) -> (ReturnValue)) -> (Output_Get:String)
   Return Node(execute=@n3.else, PackagePath=@n1)
  ]}
 GetActivePresetData {nodes [
   GetActivePresetData() {is=title K2Node_FunctionEntry}
   n1 = Get SaveGame()
   n2 = Get ActivePreset(self=@n1) -> (ActivePreset=None)
   > Return Node!(
    execute
    ActivePreset:Name=@n2
    New:STRUCT Book Shelf Book Data Structure=Find!(TargetMap=Get Local_BookCaseData(self=@n1), Key=@n2) -> (Value:STRUCT Book Shelf Book Data Structure, ReturnValue)
   )
  ]}
 GetButtonModeEnabled {nodes [
   n1 = GetButtonModeEnabled!() -> (then, Button:Button Object Reference) {is=title K2Node_FunctionEntry}
   n2 = Get WidgetStyle!(self:Button Object Reference=@n1) -> (WidgetStyle:Button Style Structure)
   n3 = "To Color (SlateColor)"!(Value:"Slate Color Structure (by ref)"=Break Slate Brush!(SlateBrush:"Slate Brush Structure (by ref)"=Break Button Style!(ButtonStyle:"Button Style Structure (by ref)"=@n2) -> (Normal:Slate Brush Structure) {is=title K2Node_BreakStruct}) -> (TintColor:Slate Color Structure) {is=title K2Node_BreakStruct}) -> (ReturnValue_R:float=~, ReturnValue_G:float=~, ReturnValue_B:float=~, ReturnValue_A:float=~)
   > Return Node!(execute, Enabled:bool=OR Boolean(A="float > float"(A=@n3.ReturnValue_G, B=@n3), B="float > float"(A=@n3.ReturnValue_G, B=@n3.ReturnValue_B)))
  ]}
 GetOpenFolderPath {nodes [
   GetOpenFolderPath() {is=title K2Node_FunctionEntry}
   n1 = GetSelectedFolderPaths!() -> (ReturnValue:Array of Strings)
   > n2 = Branch(Condition=Is Not Empty!(TargetArray=@n1) -> (ReturnValue:bool))
   > n3 = Set path(path:String="Get (a copy)"(Array=@n1)) -> (Output_Get:String)
   Return Node!(execute=[@n4.then, @n3.then], folder path:String=Get path() -> (path:String))
   n4 = Set path(execute=@n2.else, path=GetCurrentContentBrowserPath!() -> (OutPath:String, ReturnValue:bool))
  ]}
 GetSelectedPreset {nodes [
   GetSelectedPreset() {is=title K2Node_FunctionEntry}
   n1 = Get CB_ProfilesSelect()
   > Return Node!(
    execute
    Option:String=GetSelectedOption!(self:"Combo Box (String) Object Reference"=@n1) -> (ReturnValue:String)
    Index:int=GetSelectedIndex!(self:"Combo Box (String) Object Reference"=@n1) -> (ReturnValue:int=0)
   )
  ]}
 GetSelectedWorkflow {nodes [
   GetSelectedWorkflow() {is=title K2Node_FunctionEntry}
   n1 = Get CB_WorkflowSelect() -> (CB_WorkflowSelect:"Searchable Combo Box (String) Object Reference")
   > Return Node(
    Option=GetSelectedOption(self:"Searchable Combo Box (String) Object Reference"=@n1)
    Index=GetSelectedIndex(self:"Searchable Combo Box (String) Object Reference"=@n1) -> (ReturnValue=0)
   )
  ]}
 InitDataAssetDefaults {nodes [
   InitDataAssetDefaults() {is=title K2Node_FunctionEntry}
   > Set active_name(active_name:String=Get Local User Data Name() -> (Local User Data Name:String)) -> (Output_Get:String)
   n1 = Get CB_WorkflowSelect()
   > ClearOptions(self:"Searchable Combo Box (String) Object Reference"=@n1)
   n2 = Get Local User Data Name()
   > AddOption(self:"Searchable Combo Box (String) Object Reference"=@n1, Option=@n2)
   > SetSelectedOption(self:"Searchable Combo Box (String) Object Reference"=@n1, Option=@n2)
   > n3 = GetAssetsByClass!(
    execute
    self:Asset Registry Interface=GetAssetRegistry()
    ClassPathName:Top Level Asset Path Structure="Get Class Path Name (Top Level Asset Path)"!(Class:Object Class Reference=/Game/NeverCook/TechArt/BookShelf/Dependencies/DA_BookShelf.DA_BookShelf_C) -> (ReturnValue:Top Level Asset Path Structure) {is=title K2Node_CallFunction}
    bSearchSubClasses:bool=true
   ) -> (then, OutAssetData:Array of Asset Data Structures, ReturnValue:bool)
   > n4 = For Each Loop(Array:Array of Asset Data Structures=@n3) -> (Array Element:Asset Data Structure)
   WorkflowComboBoxCleanup(execute=@n4.Completed, ActiveLine=Get active_name() -> (active_name:String))
   n5 = LoadAsset_Blocking(
    execute=@n4.LoopBody
    Asset=To Soft Object Reference(SoftObjectPath=ToSoftObjectPath!(InAssetData:"Asset Data Structure (by ref)"=@n4) -> (ReturnValue:Soft Object Path Structure))
   )
   > n6 = Cast To DA_BookShelf!(execute, Object=@n5) -> (then, CastFailed, AsDA Book Shelf:DA Book Shelf Object Reference) {is=title K2Node_DynamicCast}
   n7 = Get BookShelf Name(self=@n6) -> (BookShelf Name=None)
   > Add!(
    execute
    TargetMap:Map of Strings to DA Book Shelf Soft Object References=Get DA_Lookup() -> (DA_Lookup:Map of Strings to DA Book Shelf Soft Object References)
    Key:"String (by ref)"="To String (Name)"(InName=@n7)
    Value:"DA Book Shelf Soft Object Reference (by ref)"=Make Soft Reference(Input=@n6)
   ) -> (then)
   n8 = "To String (Name)"(InName=@n7)
   > AddOption(self=@n1, Option=@n8)
   > Branch(Condition="Equal (Object)"(A=@n6, B=Get BookShelfDataAsset()))
   > SetSelectedOption(self=@n1, Option=@n8)
   > Set active_name(active_name=@n8)
  ]}
 IsUppercase {nodes [
   n1 = IsUppercase!() -> (then, str:String) {is=title K2Node_FunctionEntry}
   > Return Node!(execute, out:bool="Equal Exactly (String)"(A=ToUpper(SourceString=@n1), B=@n1))
  ]}
 LimitColorRange {nodes [
   n1 = LimitColorRange!() -> (then, In_Color:"Color Structure (by ref)") {is=title K2Node_FunctionEntry}
   n2 = Break Color(Color=@n1)
   n3 = "To Float (Byte)"!(InByte:Byte=@n2) -> (ReturnValue:double)
   n4 = "To Byte (Integer)"!(InInt:int=Truncate(A=MapRangeClamped(Value=@n3, InRangeA=0, InRangeB=255, OutRangeA=0.001, OutRangeB=200))) -> (ReturnValue:Byte=0)
   n5 = Get!() -> (In_Color_B:Byte, In_Color_G:Byte, In_Color_R:Byte, In_Color_A:Byte)
   n6 = "Byte < Byte"(A:Byte=@n5, B:Byte=200) -> (ReturnValue:bool=~) {is=title K2Node_PromotableOperator}
   n7 = "To Float (Byte)"(InByte=@n2.R)
   n8 = "To Byte (Integer)"(InInt=Truncate(A=MapRangeClamped(Value=@n7, InRangeA=0, InRangeB=255, OutRangeA=0.001, OutRangeB=200))) -> (ReturnValue=0)
   n9 = "Byte < Byte"(A=@n5.In_Color_R, B=200) -> (ReturnValue=~)
   n10 = "To Float (Byte)"(InByte=@n2.G)
   n11 = "To Byte (Integer)"(InInt=Truncate(A=MapRangeClamped(Value=@n10, InRangeA=0, InRangeB=255, OutRangeA=0.001, OutRangeB=200))) -> (ReturnValue=0)
   n12 = "Byte < Byte"(A=@n5.In_Color_G, B=200) -> (ReturnValue=~)
   > Return Node!(
    execute
    Color:Color Structure=Make Color!(
     B:Byte=Select(Option 0:Byte=@n4, Option 1:Byte=@n5, Index=@n6) -> (ReturnValue:Byte)
     G:Byte=Select(Option 0=@n11, Option 1=@n5.In_Color_G, Index=@n12)
     R:Byte=Select(Option 0=@n8, Option 1=@n5.In_Color_R, Index=@n9)
     A:Byte=255
    ) -> (Color:Color Structure) {is=title K2Node_MakeStruct}
   )
  ]}
 LoadDefaultPresets {nodes [
   LoadDefaultPresets() {is=title K2Node_FunctionEntry}
   > n1 = "For Each Loop (Map)"(MapPin:Map of Names to STRUCT Book Shelf Book Data Structures=Get Presets(self=Get BookShelfDataAsset())) -> (ValuePin:STRUCT Book Shelf Book Data Structure)
   n2 = Get SaveGame()
   > n3 = Branch(Condition=Contains(TargetMap:Map of Names to STRUCT Book Shelf Book Data Structures=Get Local_BookCaseData(self=@n2), Key=@n1))
   Add(
    execute=@n3.else
    TargetMap:Map of Names to STRUCT Book Shelf Book Data Structures=Get Local_BookCaseData(self=Get SaveGame())
    Key:"Name (by ref)"=@n1
    Value:"STRUCT Book Shelf Book Data Structure (by ref)"=@n1.ValuePin
   )
   DoSaveGame(execute=@n1.CompletedPin)
  ]}
 LoadOrCreateSaveDataObject {nodes [
   LoadOrCreateSaveDataObject() {is=title K2Node_FunctionEntry}
   > n1 = LoadGameFromSlot!(execute, SlotName:String=Get BaseSaveName(), UserIndex:int=0) -> (then, ReturnValue:Save Game Object Reference)
   > n2 = Cast To BP_QuickWidgets_SaveGame!(execute, Object=@n1) -> (then, CastFailed, AsBP Quick Widgets Save Game:BP Quick Widgets Save Game Object Reference) {is=title K2Node_DynamicCast}
   > n3 = Set SaveGame(SaveGame=@n2)
   > n4 = Set PersistantSaveGame(PersistantSaveGame:BP Quick Widgets Save Game Object Reference=@n3) -> (Output_Get:BP Quick Widgets Save Game Object Reference)
   n5 = Get LoadedWorkflow!(self:BP Quick Widgets Save Game Object Reference=@n4) -> (LoadedWorkflow:DA Book Shelf Soft Object Reference)
   > n6 = Branch(Condition=IsValidSoftObjectReference(SoftObjectReference=@n5))
   > n7 = LoadAsset_Blocking(Asset=@n5)
   > n8 = Cast To DA_BookShelf(Object=@n7)
   > Set BookShelfDataAsset(BookShelfDataAsset=@n8)
   n9 = Append!(A=BookShelf_, B=GetObjectName(Object=Get BookShelfDataAsset())) -> (ReturnValue)
   > n10 = LoadGameFromSlot(SlotName=@n9, UserIndex=0)
   > n11 = Cast To BP_QuickWidgets_SaveGame(Object=@n10)
   > n12 = Set SaveGame(SaveGame=@n11)
   Set SaveGame_Name(execute=[@n12.then, @n15.then], SaveGame_Name:String=@n9) -> (Output_Get:String)
   > DoSaveGame()
   n13 = Get SaveGame()
   > Set ListName(ListName="To String (Name)"(InName=Get ActivePreset(self=@n13) -> (ActivePreset=None)))
   n14 = CreateSaveGameObject!(
    execute=@n11.CastFailed
    SaveGameClass:Save Game Class Reference=&BP_QuickWidgets_SaveGame_C
   ) -> (then, ReturnValue:BP Quick Widgets Save Game Object Reference)
   > n15 = Set SaveGame(SaveGame=@n14)
   Set SaveGame_Name(execute=[@n8.CastFailed, @n6.else], SaveGame_Name=Get BaseSaveName())
   n16 = CreateSaveGameObject(execute=@n2.CastFailed, SaveGameClass=&BP_QuickWidgets_SaveGame_C)
   > Set SaveGame(SaveGame=@n16)
   > DoSaveGame()
   > Set SaveGame_Name(SaveGame_Name=Get BaseSaveName())
  ]}
 OnPreviewKeyDown {nodes [
   n1 = OnPreviewKeyDown!() -> (then, MyGeometry:Geometry Structure, InKeyEvent:Key Event Structure) {is=title K2Node_FunctionEntry}
   > Set isAltHeld(isAltHeld=Is Alt Down!(Input:"Input Event Structure (by ref)"=GetInputEventFromKeyEvent!(Event:"Key Event Structure (by ref)"=@n1.InKeyEvent) -> (ReturnValue:Input Event Structure)) -> (ReturnValue:bool))
   > Return Node!(execute, ReturnValue:Event Reply Structure=Unhandled())
  ]}
 OnPreviewMouseButtonDown {nodes [
   n1 = OnPreviewMouseButtonDown!() -> (then, MyGeometry:Geometry Structure, MouseEvent:"Pointer Event Structure (by ref)") {is=title K2Node_FunctionEntry}
   > Set isAltHeld(isAltHeld=Is Alt Down(Input=GetInputEventFromPointerEvent!(Event:"Pointer Event Structure (by ref)"=@n1.MouseEvent) -> (ReturnValue:Input Event Structure)))
   > Return Node(ReturnValue=Unhandled())
  ]}
 PrepCreateBookmark {nodes [
   PrepCreateBookmark() {is=title K2Node_FunctionEntry}
   > n1 = GetSelectedAssets()
   > n2 = Branch(Condition=Is Not Empty(TargetArray:Array of Object References=@n1))
   n3 = "Get (a copy)"(Array:Array of Object References=@n1) -> (Output:Object Reference)
   > Set Text(
    Text="To Text (String)"(InString=GetSoftObjectPath!(Object=@n3) -> (ReturnValue_PathString:String))
    self:Editable Text Object Reference=Get cfg_Path()
   )
   n4 = GetObjectName(Object=@n3)
   n5 = Replace(SourceString=Replace(SourceString=ToPascalSpace!(in:String=@n4) -> (out:String), From=_, To=" "), From="  ", To=" ")
   > Branch(Condition="integer > integer"(A=Len(S=@n5), B=1))
   > Set Text(Text="To Text (String)"(InString=@n5), self=Get cfg_Name())
   n6 = GetSelectedFolderPaths()
   Branch(execute=@n2.else, Condition=Is Not Empty(TargetArray:Array of Strings=@n6))
   > n7 = Get Plugin Path From Package Path!(execute, self, PluginPath:String="Get (a copy)"(Array:Array of Strings=@n6) -> (Output:String)) -> (then, PackagePath:String) {is=title K2Node_CallFunction}
   > Set Text(Text="To Text (String)"(InString=@n7), self=Get cfg_Path())
   > Set Text(Text="To Text (String)"(InString=@n7), self=Get cfg_Name())
  ]}
 ProvideUserInfoForNewBookShelf {nodes [
   ProvideUserInfoForNewBookShelf() {is=title K2Node_FunctionEntry}
   > n1 = "For Each Loop (Map)"!(execute, MapPin=Get Presets(self=Get BookShelfDataAsset()), BreakPin) -> (then, KeyPin, ValuePin_Bookmarks_27_#1:Map of Names to STRUCT Book Shelf Asset Data Structures, CompletedPin)
   n2 = Length(TargetMap:Map of Names to STRUCT Book Shelf Asset Data Structures=@n1.ValuePin_Bookmarks_27_#1) -> (ReturnValue=0)
   > Add(
    TargetArray=Get old_text() -> (old_text:Array of Strings)
    NewItem=Append!(A="To String (Name)"(InName=@n1), B=" : ", C:String="To String (Integer)"(InInt=@n2), D:String=" bookmarks found.") -> (ReturnValue)
   ) {is=title K2Node_CallArrayFunction}
   n3 = "For Each Loop (Map)"(execute=@n1.CompletedPin, MapPin=Get Local_BookCaseData(self=Get SaveGame()))
   n4 = Length(TargetMap=@n3.ValuePin_Bookmarks_27_#1) -> (ReturnValue=0)
   > Add(
    TargetArray=Get new_text() -> (new_text:Array of Strings)
    NewItem=Append(A="To String (Name)"(InName=@n3), B=" : ", C="To String (Integer)"(InInt=@n4), D=" bookmarks found.")
   )
   n5 = Get new_text()
   n6 = Get old_text()
   n7 = Get BookShelfDataAsset()
   n8 = Show Message Dialog(
    execute=@n3.CompletedPin
    Title=Format Text!(Format="LOCTEXT(\"Update Existing BookShelf?\")") -> (Result)
    Message=Format Text!(
     Format="LOCTEXT(\"Existing Data Asset found for {name}.\\r\\n\\r\\n> Data Asset Presets:\\r\\n{presets}\\r\\n\\r\\n> Your Presets:\\r\\n{new}\\r\\n\\r\\nWould you like to update the {DA} Data Asset?\")"
     name:String=GetObjectName(Object=@n7)
     presets:String=JoinStringArray(SourceArray=@n6, Separator="\r\n")
     new:String=JoinStringArray(SourceArray=@n5, Separator="\r\n")
     DA:Name=Get BookShelf Name(self=@n7) -> (BookShelf Name=None)
    ) -> (Result)
    MessageType=OkCancel
    DefaultValue=No
    MessageCategory=Info
   )
   > n9 = Switch on EAppReturnType(Selection=@n8)
   Return Node!(execute=[@n9.Cancel, @n9.No], continue:bool=false)
   Return Node(execute=[@n9.Continue, @n9.Ok, @n9.Yes], continue=true)
   Break STRUCT Book Shelf Book Data()
   GetActivePresetData() -> (ActivePreset=None) {is=title K2Node_CallFunction}
  ]}
 RemovePreset {nodes [
   RemovePreset() {is=title K2Node_FunctionEntry}
   > PrintString(InString="~~ Beginning Remove Preset ~~")
   n1 = Get Local_BookCaseData(self=Get SaveGame())
   > n2 = Branch(Condition="Equal (Integer)"(A=Length(TargetMap:Map of Names to STRUCT Book Shelf Book Data Structures=@n1) -> (ReturnValue=0), B=1))
   > ConstructPresetData()
   > DoSaveGame()
   n3 = Get Local_BookCaseData(self=Get SaveGame())
   n4 = Remove(execute=@n2.else, TargetMap:Map of Names to STRUCT Book Shelf Book Data Structures=@n3, Key=GetActivePresetData() -> (ActivePreset=None))
   > n5 = Branch(Condition=@n4)
   > n6 = Keys(TargetMap=@n3)
   > Set ActivePreset(ActivePreset="Get (a copy)"(Array:Array of Names=@n6) -> (Output:Name), self=Get SaveGame()) -> (Output_Get=None)
   > ConstructPresetData()
   > FindValidSaveDataPresets() {is=title K2Node_CallFunction}
   > DoSaveGame()
   PrintString(execute=@n5.else, InString=Error: Failed to remove preset)
  ]}
 SaveNamedPreset {nodes [
   SaveNamedPreset() {is=title K2Node_FunctionEntry}
   > n1 = Sequence()
   n2 = Get Text(self:Text Box Object Reference=Get TB_SavePreset() -> (TB_SavePreset:Editor Utility Editable Text Box Object Reference))
   > n3 = Set Preset(Preset:Name=String To Name(InString="To String (Text)"(InText=@n2))) -> (Output_Get:Name=None)
   n4 = Get SaveGame()
   > n5 = Branch(Condition=Contains(TargetMap=Get Local_BookCaseData(self=@n4), Key=@n3))
   > Set ActivePreset(ActivePreset=Get Preset() -> (Preset:Name=None), self=Get SaveGame()) -> (Output_Get=None)
   > DoSaveGame()
   > FindValidSaveDataPresets()
   n6 = For Each Loop(Exec=@n5.else, Array:Array of Widget Object References=GetAllChildren(self=Get BookmarksList())) -> (Array Element:Widget Object Reference)
   Add(
    execute=@n6.Completed
    TargetMap=Get Local_BookCaseData(self=Get SaveGame())
    Key=Get Preset() -> (Preset=None)
    Value=Make STRUCT Book Shelf Book Data!(Bookmarks_27_#1:Map of Names to STRUCT Book Shelf Asset Data Structures=Get bookmarks()) -> (STRUCT_BookShelf_BookData:STRUCT Book Shelf Book Data Structure) {is=title K2Node_MakeStruct}
   ) {is=title K2Node_CallFunction}
   > Set ActivePreset(ActivePreset=Get Preset() -> (Preset=None), self=Get SaveGame()) -> (Output_Get=None)
   > DoSaveGame()
   > FindValidSaveDataPresets()
   n7 = Cast To EUW_BookCase_Category!(execute=@n6.LoopBody, Object=@n6) -> (then, CastFailed, AsEUW Book Case Category:EUW Book Case Category Object Reference) {is=title K2Node_DynamicCast}
   > n8 = For Each Loop(Array=GetAllChildren(self=Get VBox_Content(self=@n7)))
   > n9 = Cast To EUW_BookCase_BookMark!(execute, Object=@n8) -> (then, CastFailed, AsEUW Book Case Book Mark:EUW Book Case Book Mark Object Reference) {is=title K2Node_DynamicCast}
   > Add(
    TargetMap:Map of Names to STRUCT Book Shelf Asset Data Structures=Get bookmarks()
    Key=Get Bookmark Name(self=@n9) -> (Bookmark Name=None)
    Value:"STRUCT Book Shelf Asset Data Structure (by ref)"=Get BookmarkData(self=@n9)
   )
   ConstructPresetData(execute=@n1.then_1)
  ]}
 SaveNewBookCaseDataAsset {nodes [
   SaveNewBookCaseDataAsset() {is=title K2Node_FunctionEntry}
   > Set BookShelfDataAsset()
   n1 = "To String (Text)"(InText=Get Text(self=Get TB_BookCaseFolderPath()))
   n2 = "To String (Text)"(InText=Get Text(self=Get TB_BookCaseAssetName()))
   n3 = Append!(A=@n1, B=/, C=@n2) -> (ReturnValue)
   > n4 = DoesAssetExist(AssetPath=@n3)
   > n5 = Branch(Condition=@n4)
   > n6 = LoadAsset(AssetPath=@n3)
   > n7 = Cast To DA_BookShelf(Object=@n6)
   > Set BookShelfDataAsset(BookShelfDataAsset=@n7)
   > n8 = ProvideUserInfoForNewBookShelf!(execute, self) -> (then, continue:bool) {is=title K2Node_CallFunction}
   > Branch(Condition=@n8)
   > n9 = Set made_new(made_new:bool=false) -> (Output_Get:bool)
   Set Presets!(
    execute=[@n20.then, @n9.then]
    Presets:Map of Names to STRUCT Book Shelf Book Data Structures=Get Local_BookCaseData(self=Get SaveGame())
    self:DA Book Shelf Object Reference=Get BookShelfDataAsset()
   ) -> (then, Output_Get:Map of Names to STRUCT Book Shelf Book Data Structures)
   > SaveLoadedAsset(AssetToSave=Get BookShelfDataAsset(), bOnlyIfIsDirty=true)
   > Branch(Condition=Get made_new() -> (made_new:bool))
   n10 = Get BookShelfDataAsset()
   > UpdatedWorkflowSelection(WorkflowName="To String (Name)"(InName=Get BookShelf Name(self=@n10) -> (BookShelf Name=None)))
   Show Message Dialog(
    execute=@n7.CastFailed
    Title=MakeLiteralText(Value="LOCTEXT(\"Found Asset is Incompatible!\")")
    Message=Format Text!(
     Format="LOCTEXT(\"Asset: {name}\\r\\nAt location: {loc}\\r\\nIs incompatible with type DA_BookShelf.\\r\\n\\r\\nYou are unable to overwrite this asset, please try entering a different name or location to create a new Data Asset.\")"
     name=@n2
     loc:String=@n1
    ) -> (Result)
    MessageType=Ok
    DefaultValue=Ok
    MessageCategory=Error
   )
   n11 = MakeValidFileName!(InString:String=@n2, InReplacementChar:String) -> (ReturnValue:String)
   n12 = Branch(
    execute=@n5.else
    Condition=OR Boolean!(A=IsEmpty(InString=@n2), B=IsEmpty(InString=@n1), C:bool="Not Equal Exactly (String)"(A=@n2, B=@n11)) -> (ReturnValue)
   )
   > PrintString(InString=Invalid path and/or name for new BookShelf Data Asset, TextColor=(R=1,G=0.1051,B=0,A=1), Duration=8)
   n13 = Execute Python Script!(
    execute=@n12.else
    PythonScript="import unreal\r\n\r\nmy_data_asset_class = unreal.EditorAssetLibrary.load_blueprint_class(data_class_path)\r\n\r\n## Get asset tools\r\nasset_tools = unreal.AssetToolsHelpers.get_asset_tools()\r\n\r\n## Create the asset\r\nnew_asset = asset_tools.create_asset(asset_name, package_path, my_data_asset_class, unreal.DataAssetFactory())\r\n\r\nprint(f\"Created DataAsset: {new_asset}\")"
    asset_name:String=@n2
    package_path:String=@n1
    data_class_path:String=MakeLiteralString(Value=/Script/Engine.Blueprint'/Game/NeverCook/TechArt/BookShelf/Dependencies/DA_BookShelf')
   ) -> (then, ReturnValue, new_asset:Primary Data Asset Object Reference) {inputs [asset_name, package_path, data_class_path], outputs [new_asset]}
   > n14 = Set dataasset(dataasset:Primary Data Asset Object Reference=@n13.new_asset) -> (Output_Get:Primary Data Asset Object Reference)
   > n15 = Cast To DA_BookShelf(Object=@n14)
   > n16 = Set BookShelfDataAsset(BookShelfDataAsset=@n15)
   n17 = Get TB_BookCaseAssetName()
   n18 = RightChop(SourceString="To String (Text)"(InText=Get Text(self=@n17)), Count=3)
   n19 = String To Name(InString=ToPascalSpace(in=Replace(SourceString=@n18, From=_, To=" ")))
   > Set BookShelf Name!(execute, BookShelf Name:Name=@n19, self:DA Book Shelf Object Reference=@n16) -> (then, Output_Get:Name=None)
   > n20 = Add(
    TargetMap:Map of Strings to DA Book Shelf Soft Object References=Get DA_Lookup()
    Key:"String (by ref)"="To String (Name)"(InName=@n19)
    Value:"DA Book Shelf Soft Object Reference (by ref)"=Make Soft Reference(Input=@n16)
   )
  ]}
 SelectNewPreset {nodes [
   n1 = SelectNewPreset!() -> (then, PresetName:Name=None, PresetIndex:int=-1) {is=title K2Node_FunctionEntry}
   n2 = "integer < integer"(A=@n1.PresetIndex)
   > n3 = Branch(Condition=AND Boolean(A="Equal (Name)"(A=@n1), B=@n2))
   n4 = Branch(execute=@n3.else, Condition=@n2)
   n5 = "To String (Name)"(InName=@n1)
   n6 = Get CB_ProfilesSelect()
   > n7 = Branch(Condition="Equal (Integer)"(A=FindOptionIndex!(self:"Combo Box (String) Object Reference"=@n6, Option:String=@n5) -> (ReturnValue:int=0), B=-1))
   n8 = Set local_option(execute=@n7.else, local_option:String=@n5) -> (Output_Get:String)
   n9 = Get CB_ProfilesSelect()
   Branch(
    execute=[@n13.then, @n8.then]
    Condition="Not Equal Exactly (String)"(A=Get local_option() -> (local_option:String), B=GetSelectedOption(self:"Combo Box (String) Object Reference"=@n9))
   )
   n10 = Get local_option()
   > SetSelectedOption(self:"Combo Box (String) Object Reference"=@n6, Option=@n10)
   > Return Node!(execute, SelectedPreset:String=@n10, Update Preset:bool=true)
   n11 = GetOptionAtIndex!(self:"Combo Box (String) Object Reference"=@n6, Index:int=@n1.PresetIndex) -> (ReturnValue:String)
   n12 = Branch(execute=@n4.else, Condition=IsEmpty(InString=@n11))
   n13 = Set local_option(execute=@n12.else, local_option=@n11)
  ]}
 SelectNewWorkflow {nodes [
   n1 = SelectNewWorkflow!() -> (then, Option:String) {is=title K2Node_FunctionEntry}
   n2 = Get CB_WorkflowSelect()
   > n3 = Branch(Condition="Equal (Integer)"(A=FindOptionIndex(self:"Searchable Combo Box (String) Object Reference"=@n2, Option=@n1) -> (ReturnValue=0), B=-1))
   > Return Node!(execute, Changed:bool=false)
   SetSelectedOption(execute=@n3.else, self:"Searchable Combo Box (String) Object Reference"=@n2, Option=@n1)
   > Return Node(Changed=true)
  ]}
 SetBookmarkConfigurationMode {nodes [
   n1 = SetBookmarkConfigurationMode!() -> (then, enable:bool=~) {is=title K2Node_FunctionEntry}
   > Set Text(
    Text=SelectText(A="LOCTEXT(\"Update Bookmark\")", B="LOCTEXT(\"Add Bookmark\")", bPickA=@n1)
    self:Text Object Reference=Get TX_Configurebookmark()
   )
   > n2 = Values!(execute, TargetMap:Map of Strings to EUW Book Case Category Object References=Get ActiveCategories()) -> (then, Values:Array of EUW Book Case Category Object References)
   > n3 = For Each Loop(Array:Array of EUW Book Case Category Object References=@n2) -> (Array Element:EUW Book Case Category Object Reference)
   > SetConfigurationMode!(execute, self:EUW Book Case Category Object Reference=@n3, enable:bool=@n1) -> (then)
  ]}
 SortSaveGame_ActivePreset {nodes [
   SortSaveGame_ActivePreset() {is=title K2Node_FunctionEntry}
   n1 = Break STRUCT Book Shelf Book Data(STRUCT_BookShelf_BookData=GetActivePresetData() -> (ActivePreset=None).New)
   > n2 = Keys(TargetMap:Map of Names to STRUCT Book Shelf Asset Data Structures=@n1)
   > Set loc_names(loc_names:Array of Names=@n2) -> (Output_Get:Array of Names)
   > n3 = Values(TargetMap:Map of Names to STRUCT Book Shelf Asset Data Structures=@n1) -> (Values:Array of STRUCT Book Shelf Asset Data Structures)
   > n4 = Set loc_values(loc_values:Array of STRUCT Book Shelf Asset Data Structures=@n3) -> (Output_Get:Array of STRUCT Book Shelf Asset Data Structures)
   > n5 = For Each Loop(Array:Array of STRUCT Book Shelf Asset Data Structures=@n4) -> (Array Element:STRUCT Book Shelf Asset Data Structure)
   n6 = For Each Loop(Exec=@n5.Completed, Array:Array of Strings=Get loc_categories_ordered() -> (loc_categories_ordered:Array of Strings)) -> (Array Element:String)
   Add!(
    execute=@n6.Completed
    TargetMap:Map of Names to STRUCT Book Shelf Book Data Structures=Get Local_BookCaseData(self=Get SaveGame())
    Key:"Name (by ref)"=GetActivePresetData() -> (ActivePreset=None)
    Value_Bookmarks_27_#1:Map of Names to STRUCT Book Shelf Asset Data Structures=Get loc_newBookmarks() -> (loc_newBookmarks:Map of Names to STRUCT Book Shelf Asset Data Structures)
   ) -> (then)
   n7 = For Each Loop(
    Exec=@n6.LoopBody
    Array:Array of STRUCT Book Shelf Asset Data Structures=Get loc_values() -> (loc_values:Array of STRUCT Book Shelf Asset Data Structures)
   ) -> (Array Element:STRUCT Book Shelf Asset Data Structure)
   > Branch(Condition="Equal Exactly (String)"(A=Break STRUCT Book Shelf Asset Data(STRUCT_BookShelf_AssetData=@n7), B=@n6))
   > Add!(
    execute
    TargetMap:Map of Names to STRUCT Book Shelf Asset Data Structures=Get loc_newBookmarks()
    Key="Get (a copy)"(Array=Get loc_names() -> (loc_names:Array of Names), Dimension 1=@n7.Array Index)
    Value:"STRUCT Book Shelf Asset Data Structure (by ref)"=@n7
   ) -> (then)
   Add Unique(execute=@n5.LoopBody, TargetArray=Get loc_categories_ordered(), NewItem=Break STRUCT Book Shelf Asset Data(STRUCT_BookShelf_AssetData=@n5))
   Get Local_BookCaseData(self=Get SaveGame())
  ]}
 ToPascalSpace {nodes [
   n1 = ToPascalSpace!() -> (then, in:String) {is=title K2Node_FunctionEntry}
   > Set OriginalArray(OriginalArray:Array of Strings=GetCharacterArrayFromString!(SourceString:String=@n1) -> (ReturnValue:Array of Strings)) -> (Output_Get:Array of Strings)
   > n2 = For Each Loop(Array:Array of Strings=Get OriginalArray() -> (OriginalArray:Array of Strings)) -> (Array Element:String)
   > n3 = Set str(str:String=@n2) -> (Output_Get:String)
   > n4 = Branch(Condition=IsUppercase!(str:String=@n3) -> (out:bool) {is=title K2Node_CallFunction})
   n5 = Get OutArray() -> (OutArray:Array of Strings)
   n6 = Last Index!(TargetArray=@n5) -> (ReturnValue=0)
   > n7 = Branch(Condition=IsUppercase(str="Get (a copy)"(Array:Array of Strings=@n5, Dimension 1=@n6) -> (Output:String)))
   n8 = Get str() -> (str:String)
   > Add(TargetArray=Get OutArray(), NewItem=@n8) {is=title K2Node_CallArrayFunction}
   Add(execute=@n7.else, TargetArray=Get OutArray(), NewItem=Append!(A=" ", B=@n8) -> (ReturnValue))
   Add(execute=@n4.else, TargetArray=Get OutArray(), NewItem=Get str())
   n9 = Set OutputString(execute=@n2.Completed, OutputString:String=JoinStringArray(SourceArray=Get OutArray())) -> (Output_Get:String)
   > Return Node!(execute, out:String=@n9)
  ]}
 Update Active Preset Name {nodes [
   n1 = Update Active Preset Name!() -> (then, Name:String) {is=title K2Node_FunctionEntry}
   > n2 = Branch(Condition=IsEmpty(InString=@n1))
   n3 = Set ListName(execute=@n2.else, ListName=@n1)
   n4 = String To Name(InString=@n1)
   n5 = GetActivePresetData() -> (ActivePreset=None)
   n6 = Get Local_BookCaseData(self=Get SaveGame())
   > Add(
    TargetMap:Map of Names to STRUCT Book Shelf Book Data Structures=@n6
    Key=@n4
    Value:"STRUCT Book Shelf Book Data Structure (by ref)"=@n5.New
   ) {is=title K2Node_CallFunction}
   > Remove(TargetMap=@n6, Key=@n5)
   > Set ActivePreset(ActivePreset=@n4, self=Get SaveGame()) -> (Output_Get=None)
   > DoSaveGame()
   String To Name(InString=@n3)
  ]}
 UpdatedWorkflowSelection {nodes [
   n1 = UpdatedWorkflowSelection!() -> (then, WorkflowName:String) {is=title K2Node_FunctionEntry}
   > n2 = Branch(Condition=StartsWith(SourceString=@n1, InPrefix="​Workflow", SearchCase=CaseSensitive))
   n3 = LoadAsset_Blocking(
    execute=@n2.else
    Asset=Find(TargetMap:Map of Strings to DA Book Shelf Soft Object References=Get DA_Lookup(), Key:"String (by ref)"=@n1) -> (Value:DA Book Shelf Soft Object Reference)
   )
   > n4 = Cast To DA_BookShelf(Object=@n3)
   > n5 = Set BookShelfDataAsset(BookShelfDataAsset=@n4)
   ChangeLoadedWorkflow(execute=[@n6.then, @n5.then]) {is=title K2Node_CallFunction}
   > LoadOrCreateSaveDataObject() {is=title K2Node_CallFunction}
   > Refresh Data()
   > WorkflowComboBoxCleanup(ActiveLine=@n1)
   n6 = Set BookShelfDataAsset(execute=@n4.CastFailed)
  ]}
 ValidateActivePath {nodes [
   Inputs()
   n1 = Get cfg_Path()
   > n2 = Get Plugin Path From Package Path(PluginPath="To String (Text)"(InText=Get Text(self:Editable Text Object Reference=@n1)))
   > n3 = Branch(Condition=IsEmpty(InString=@n2))
   > n4 = Set Text(Text="LOCTEXT(\"Invalid Path\")", self:Editable Text Object Reference=Get cfg_Path())
   Outputs!(Valid:Exec=@n3.else, Invalid:Exec=@n4.then)
   Get UnrealEditorSubsystem()
  ]}
 WarnBeforeResettingData {nodes [
   WarnBeforeResettingData() {is=title K2Node_FunctionEntry}
   > n1 = Show Message Dialog(
    Title=MakeLiteralText(Value="LOCTEXT(\"Warning - This Action Is Destructive!\")")
    Message=MakeLiteralText(Value="LOCTEXT(\"Proceeding will result in both your active and saved Bookmark data being reset.\\r\\n\\r\\nPlease select \\'Ok\\' if you would like to continue resetting your save data anyways. \\'Cancel\\' will abort this action.\")")
    MessageType=OkCancel
    DefaultValue=Cancel
    MessageCategory=Warning
   )
   > n2 = Switch on EAppReturnType(Selection=@n1)
   Return Node!(execute=@n2.Cancel, Do:bool=false)
   Return Node(execute=@n2.Ok, Do=true)
  ]}
 WorkflowComboBoxCleanup {nodes [
   n1 = WorkflowComboBoxCleanup!() -> (then, ActiveLine:String) {is=title K2Node_FunctionEntry}
   > ClearOptions(self=Get CB_WorkflowSelect())
   > n2 = Keys(TargetMap:Map of Strings to DA Book Shelf Soft Object References=Get DA_Lookup()) -> (Keys:Array of Strings)
   > n3 = Set Categories(Categories:Array of Strings=@n2) -> (Output_Get:Array of Strings)
   > SortStringArray!(execute, TargetArray:Array of Strings=@n3, bStableSort:bool=false, SortOrder:EArraySortOrder Enum=Ascending) -> (then)
   > Insert!(execute, TargetArray=Get Categories() -> (Categories:Array of Strings), NewItem=Get Local User Data Name(), Index:int=0) -> (then) {is=title K2Node_CallArrayFunction}
   > n4 = For Each Loop(Array=Get Categories())
   > n5 = Branch(Condition="Equal Exactly (String)"(A=@n4, B=@n1))
   n6 = Append(A="​Workflow - ", B=@n1)
   n7 = Get CB_WorkflowSelect()
   > AddOption(self=@n7, Option=@n6)
   > SetSelectedOption(self=@n7, Option=@n6)
   AddOption(execute=@n5.else, self=@n7, Option=@n4)
  ]}
]
variables [
 ActiveCategories {type="map<EUW_BookCase_Category_C>", default {}}
 BaseSaveName {type=string, default=EUW_QuickWidgets_Bookmarks}
 BookShelfDataAsset {type=DA_BookShelf_C, default=null}
 Color {type=Color, default=<rgba 0.2902 0.2902 0.298 1>}
 DA_Lookup {type="map<DA_BookShelf_C>", default {}}
 EditingBookmark {type=EUW_BookCase_BookMark_C, default=null}
 isAltHeld {type=boolean}
 ListName {type=string, default=Bookmarks}
 Local User Data Name {type=string, default="<< ​Local User​ >>"}
 PersistantSaveGame {type=BP_QuickWidgets_SaveGame_C, default=null}
 SaveGame {type=BP_QuickWidgets_SaveGame_C, default=null}
 SaveGame_Name {type=string, default=EUW_QuickWidgets_Bookmarks}
]
widgets {
 info {namedSlotCount=2, rootWidgetClass=&Border, widgetCount=62}
 widgets [
  Border_0 {widgetClassPath=&Border, widget=Border_0, properties {brushColor=<rgba 0.01 0.01 0.01 1>, padding=<ltrb 0 0 0 0>}}
  VerticalBox_2 {widgetClassPath=&VerticalBox, parent=Border_0, widget=VerticalBox_2, slot {class=BorderSlot, properties {padding=<ltrb 0 0 0 0>}}}
  Border_7 {
   widgetClassPath=&Border
   parent=VerticalBox_2
   widget=Border_7
   properties {
    background {
     bIsDynamicallyLoaded=false
     drawAs=RoundedBox
     imageType=NoImage
     mirroring=NoMirror
     resourceName=None
     resourceObject=null
     tiling=NoTile
     imageSize=<32 32>
     margin=<ltrb 0 0 0 0>
     outlineSettings {
      bUseBrushTransparency=false
      roundingType=FixedRadius
      width=0
      color {colorUseRule=UseColor_Specified, specifiedColor=<rgba 0 0 0 0>}
      cornerRadii=<0 0 0 0>
     }
     tintColor {colorUseRule=UseColor_Specified, specifiedColor=<rgba 1 1 1 1>}
     uVRegion {bIsValid=false, max=<0 0>, min=<0 0>}
    }
    brushColor=<rgba 0.007 0.007 0.007 1>
    padding=<ltrb 4 4 4 2>
   }
   slot {class=VerticalBoxSlot}
  }
  VerticalBox_86 {
   widgetClassPath=&VerticalBox
   parent=Border_7
   widget=VerticalBox_86
   slot {class=BorderSlot, properties {padding=<ltrb 4 4 4 2>}}
  }
  Border_253 {
   widgetClassPath=&Border
   parent=VerticalBox_86
   widget=Border_253
   properties {
    verticalAlignment=VAlign_Top
    background {
     bIsDynamicallyLoaded=false
     drawAs=RoundedBox
     imageType=NoImage
     mirroring=NoMirror
     resourceName=None
     resourceObject=null
     tiling=NoTile
     imageSize=<32 32>
     margin=<ltrb 0 0 0 0>
     outlineSettings {
      bUseBrushTransparency=false
      roundingType=FixedRadius
      width=0
      color {colorUseRule=UseColor_Specified, specifiedColor=<rgba 0 0 0 0>}
      cornerRadii=<4 4 0 0>
     }
     tintColor {colorUseRule=UseColor_Specified, specifiedColor=<rgba 1 1 1 1>}
     uVRegion {bIsValid=false, max=<0 0>, min=<0 0>}
    }
    brushColor=<rgba 0.004 0.004 0.004 1>
    padding=<ltrb 0 0 0 0>
   }
   slot {class=VerticalBoxSlot, properties {padding=<ltrb -4 -2 -4 0>}}
  }
  HB--Workflow {
   widgetClassPath=&HorizontalBox
   parent=VerticalBox_86
   widget=HB--Workflow
   slot {class=VerticalBoxSlot, properties {padding=<ltrb 4 4 4 2>}}
  }
  CB_WorkflowSelect {
   widgetClassPath=/Script/FortniteEditor.SearchableStringComboBox
   bIsVariable=true
   parent=HB--Workflow
   widget=CB_WorkflowSelect
   properties {
    bOverride_Cursor=true
    selectedOption=Item A
    toolTipText="Change which workflow is active.\r\nLocal User Workflow changes persist between sessions.\r\nOther workflows require saving over the associated Data Asset to save changes (other than bookmark additions)."
    defaultOptions [Item A, Item B, Item C]
    font {
     bForceMonospaced=false
     bMaterialIsStencil=false
     fontMaterial=null
     fontObject=&Roboto
     letterSpacing=0
     monospacedWidth=1
     size=8.25
     skewAmount=0
     typefaceFontName=Regular
     outlineSettings {
      bApplyOutlineToDropShadows=false
      bMiteredCorners=false
      bSeparateFillAlpha=false
      outlineMaterial=null
      outlineSize=0
      outlineColor=<rgba 0 0 0 1>
     }
    }
    foregroundColor {colorUseRule=UseColor_Specified, specifiedColor=<rgba 0.5271 0.5271 0.5271 1>}
    itemStyle {
     bUseParentRowBrush=false
     activeBrush {
      bIsDynamicallyLoaded=false
      drawAs=Image
      imageType=NoImage
      mirroring=NoMirror
      resourceName=None
      resourceObject=null
      tiling=NoTile
      imageSize=<0 0>
      margin=<ltrb 0 0 0 0>
      outlineSettings {
       bUseBrushTransparency=false
       roundingType=HalfHeightRadius
       width=0
       color {colorUseRule=UseColor_Specified, specifiedColor=<rgba 0 0 0 0>}
       cornerRadii=<0 0 0 0>
      }
      tintColor {colorUseRule=UseColor_Specified, specifiedColor=<rgba 0 0.162 0.7454 1>}
      uVRegion {bIsValid=false, max=<0 0>, min=<0 0>}
     }
     activeHighlightedBrush {
      bIsDynamicallyLoaded=false
      drawAs=Image
      imageType=NoImage
      mirroring=NoMirror
      resourceName=None
      resourceObject=null
      tiling=NoTile
      imageSize=<0 0>
      margin=<ltrb 0 0 0 0>
      outlineSettings {
       bUseBrushTransparency=false
       roundingType=HalfHeightRadius
       width=0
       color {colorUseRule=UseColor_Specified, specifiedColor=<rgba 0 0 0 0>}
       cornerRadii=<0 0 0 0>
      }
      tintColor {colorUseRule=UseColor_Specified, specifiedColor=<rgba 0.0252 0.0319 0.0423 1>}
      uVRegion {bIsValid=false, max=<0 0>, min=<0 0>}
     }
     activeHoveredBrush {
      bIsDynamicallyLoaded=false
      drawAs=Image
      imageType=NoImage
      mirroring=NoMirror
      resourceName=None
      resourceObject=null
      tiling=NoTile
      imageSize=<0 0>
      margin=<ltrb 0 0 0 0>
      outlineSettings {
       bUseBrushTransparency=false
       roundingType=HalfHeightRadius
       width=0
       color {colorUseRule=UseColor_Specified, specifiedColor=<rgba 0 0 0 0>}
       cornerRadii=<0 0 0 0>
      }
      tintColor {colorUseRule=UseColor_Specified, specifiedColor=<rgba 0 0.162 0.7454 1>}
      uVRegion {bIsValid=false, max=<0 0>, min=<0 0>}
     }
     dropIndicator_Above {
      bIsDynamicallyLoaded=false
      drawAs=Box
      imageType=FullColor
      mirroring=NoMirror
      resourceName=../../../Engine/Content/Slate/Common/DropZoneIndicator_Above.png
      resourceObject=null
      tiling=NoTile
      imageSize=<0 0>
      margin=<ltrb 0.625 0.625 0 0>
      outlineSettings {
       bUseBrushTransparency=false
       roundingType=HalfHeightRadius
       width=0
       color {colorUseRule=UseColor_Specified, specifiedColor=<rgba 0 0 0 0>}
       cornerRadii=<0 0 0 0>
      }
      tintColor {colorUseRule=UseColor_Specified, specifiedColor=<rgba 0.0091 0.3515 0.9647 1>}
      uVRegion {bIsValid=false, max=<0 0>, min=<0 0>}
     }
     dropIndicator_Below {
      bIsDynamicallyLoaded=false
      drawAs=Box
      imageType=FullColor
      mirroring=NoMirror
      resourceName=../../../Engine/Content/Slate/Common/DropZoneIndicator_Below.png
      resourceObject=null
      tiling=NoTile
      imageSize=<0 0>
      margin=<ltrb 0.625 0 0 0.625>
      outlineSettings {
       bUseBrushTransparency=false
       roundingType=HalfHeightRadius
       width=0
       color {colorUseRule=UseColor_Specified, specifiedColor=<rgba 0 0 0 0>}
       cornerRadii=<0 0 0 0>
      }
      tintColor {colorUseRule=UseColor_Specified, specifiedColor=<rgba 0.0091 0.3515 0.9647 1>}
      uVRegion {bIsValid=false, max=<0 0>, min=<0 0>}
     }
     dropIndicator_Onto {
      bIsDynamicallyLoaded=false
      drawAs=Box
      imageType=FullColor
      mirroring=NoMirror
      resourceName=../../../Engine/Content/Slate/Common/DropZoneIndicator_Onto.png
      resourceObject=null
      tiling=NoTile
      imageSize=<0 0>
      margin=<ltrb 0.25 0.25 0.25 0.25>
      outlineSettings {
       bUseBrushTransparency=false
       roundingType=HalfHeightRadius
       width=0
       color {colorUseRule=UseColor_Specified, specifiedColor=<rgba 0 0 0 0>}
       cornerRadii=<0 0 0 0>
      }
      tintColor {colorUseRule=UseColor_Specified, specifiedColor=<rgba 0.0091 0.3515 0.9647 1>}
      uVRegion {bIsValid=false, max=<0 0>, min=<0 0>}
     }
     evenRowBackgroundBrush {
      bIsDynamicallyLoaded=false
      drawAs=Image
      imageType=NoImage
      mirroring=NoMirror
      resourceName=None
      resourceObject=null
      tiling=NoTile
      imageSize=<0 0>
      margin=<ltrb 0 0 0 0>
      outlineSettings {
       bUseBrushTransparency=false
       roundingType=HalfHeightRadius
       width=0
       color {colorUseRule=UseColor_Specified, specifiedColor=<rgba 0 0 0 0>}
       cornerRadii=<0 0 0 0>
      }
      tintColor {colorUseRule=UseColor_Specified, specifiedColor=<rgba 0.0103 0.0103 0.0103 1>}
      uVRegion {bIsValid=false, max=<0 0>, min=<0 0>}
     }
     evenRowBackgroundHoveredBrush {
      bIsDynamicallyLoaded=false
      drawAs=Image
      imageType=NoImage
      mirroring=NoMirror
      resourceName=None
      resourceObject=null
      tiling=NoTile
      imageSize=<0 0>
      margin=<ltrb 0 0 0 0>
      outlineSettings {
       bUseBrushTransparency=false
       roundingType=HalfHeightRadius
       width=0
       color {colorUseRule=UseColor_Specified, specifiedColor=<rgba 0 0 0 0>}
       cornerRadii=<0 0 0 0>
      }
      tintColor {colorUseRule=UseColor_Specified, specifiedColor=<rgba 0.0176 0.0176 0.0176 1>}
      uVRegion {bIsValid=false, max=<0 0>, min=<0 0>}
     }
     inactiveBrush {
      bIsDynamicallyLoaded=false
      drawAs=Image
      imageType=NoImage
      mirroring=NoMirror
      resourceName=None
      resourceObject=null
      tiling=NoTile
      imageSize=<0 0>
      margin=<ltrb 0 0 0 0>
      outlineSettings {
       bUseBrushTransparency=false
       roundingType=HalfHeightRadius
       width=0
       color {colorUseRule=UseColor_Specified, specifiedColor=<rgba 0 0 0 0>}
       cornerRadii=<0 0 0 0>
      }
      tintColor {colorUseRule=UseColor_Specified, specifiedColor=<rgba 0.0513 0.0953 0.159 1>}
      uVRegion {bIsValid=false, max=<0 0>, min=<0 0>}
     }
     inactiveHighlightedBrush {
      bIsDynamicallyLoaded=false
      drawAs=Image
      imageType=NoImage
      mirroring=NoMirror
      resourceName=None
      resourceObject=null
      tiling=NoTile
      imageSize=<0 0>
      margin=<ltrb 0 0 0 0>
      outlineSettings {
       bUseBrushTransparency=false
       roundingType=HalfHeightRadius
       width=0
       color {colorUseRule=UseColor_Specified, specifiedColor=<rgba 0 0 0 0>}
       cornerRadii=<0 0 0 0>
      }
      tintColor {colorUseRule=UseColor_Specified, specifiedColor=<rgba 0.0252 0.0319 0.0423 1>}
      uVRegion {bIsValid=false, max=<0 0>, min=<0 0>}
     }
     inactiveHoveredBrush {
      bIsDynamicallyLoaded=false
      drawAs=Image
      imageType=NoImage
      mirroring=NoMirror
      resourceName=None
      resourceObject=null
      tiling=NoTile
      imageSize=<0 0>
      margin=<ltrb 0 0 0 0>
      outlineSettings {
       bUseBrushTransparency=false
       roundingType=HalfHeightRadius
       width=0
       color {colorUseRule=UseColor_Specified, specifiedColor=<rgba 0 0 0 0>}
       cornerRadii=<0 0 0 0>
      }
      tintColor {colorUseRule=UseColor_Specified, specifiedColor=<rgba 0.0513 0.0953 0.159 1>}
      uVRegion {bIsValid=false, max=<0 0>, min=<0 0>}
     }
     oddRowBackgroundBrush {
      bIsDynamicallyLoaded=false
      drawAs=Image
      imageType=NoImage
      mirroring=NoMirror
      resourceName=None
      resourceObject=null
      tiling=NoTile
      imageSize=<0 0>
      margin=<ltrb 0 0 0 0>
      outlineSettings {
       bUseBrushTransparency=false
       roundingType=HalfHeightRadius
       width=0
       color {colorUseRule=UseColor_Specified, specifiedColor=<rgba 0 0 0 0>}
       cornerRadii=<0 0 0 0>
      }
      tintColor {colorUseRule=UseColor_Specified, specifiedColor=<rgba 0.0103 0.0103 0.0103 1>}
      uVRegion {bIsValid=false, max=<0 0>, min=<0 0>}
     }
     oddRowBackgroundHoveredBrush {
      bIsDynamicallyLoaded=false
      drawAs=Image
      imageType=NoImage
      mirroring=NoMirror
      resourceName=None
      resourceObject=null
      tiling=NoTile
      imageSize=<0 0>
      margin=<ltrb 0 0 0 0>
      outlineSettings {
       bUseBrushTransparency=false
       roundingType=HalfHeightRadius
       width=0
       color {colorUseRule=UseColor_Specified, specifiedColor=<rgba 0 0 0 0>}
       cornerRadii=<0 0 0 0>
      }
      tintColor {colorUseRule=UseColor_Specified, specifiedColor=<rgba 0.0176 0.0176 0.0176 1>}
      uVRegion {bIsValid=false, max=<0 0>, min=<0 0>}
     }
     parentRowBackgroundBrush {
      bIsDynamicallyLoaded=false
      drawAs=Image
      imageType=NoImage
      mirroring=NoMirror
      resourceName=None
      resourceObject=null
      tiling=NoTile
      imageSize=<32 32>
      margin=<ltrb 0 0 0 0>
      outlineSettings {
       bUseBrushTransparency=false
       roundingType=HalfHeightRadius
       width=0
       color {colorUseRule=UseColor_Specified, specifiedColor=<rgba 0 0 0 0>}
       cornerRadii=<0 0 0 0>
      }
      tintColor {colorUseRule=UseColor_Specified, specifiedColor=<rgba 1 1 1 1>}
      uVRegion {bIsValid=false, max=<0 0>, min=<0 0>}
     }
     parentRowBackgroundHoveredBrush {
      bIsDynamicallyLoaded=false
      drawAs=Image
      imageType=NoImage
      mirroring=NoMirror
      resourceName=None
      resourceObject=null
      tiling=NoTile
      imageSize=<32 32>
      margin=<ltrb 0 0 0 0>
      outlineSettings {
       bUseBrushTransparency=false
       roundingType=HalfHeightRadius
       width=0
       color {colorUseRule=UseColor_Specified, specifiedColor=<rgba 0 0 0 0>}
       cornerRadii=<0 0 0 0>
      }
      tintColor {colorUseRule=UseColor_Specified, specifiedColor=<rgba 1 1 1 1>}
      uVRegion {bIsValid=false, max=<0 0>, min=<0 0>}
     }
     selectedTextColor {colorUseRule=UseColor_Specified, specifiedColor=<rgba 1 1 1 1>}
     selectorFocusedBrush {
      bIsDynamicallyLoaded=false
      drawAs=Border
      imageType=FullColor
      mirroring=NoMirror
      resourceName=../../../Engine/Content/Slate/Common/Selector.png
      resourceObject=null
      tiling=Both
      imageSize=<0 0>
      margin=<ltrb 0.25 0.25 0.25 0.25>
      outlineSettings {
       bUseBrushTransparency=false
       roundingType=HalfHeightRadius
       width=0
       color {colorUseRule=UseColor_Specified, specifiedColor=<rgba 0 0 0 0>}
       cornerRadii=<0 0 0 0>
      }
      tintColor {colorUseRule=UseColor_Specified, specifiedColor=<rgba 0.701 0.225 0.003 1>}
      uVRegion {bIsValid=false, max=<0 0>, min=<0 0>}
     }
     textColor {colorUseRule=UseColor_Specified, specifiedColor=<rgba 0.5271 0.5271 0.5271 1>}
    }
    widgetStyle {
     comboButtonStyle {
      downArrowAlign=VAlign_Center
      buttonStyle {
       clickedSlateSound {resourceObject=null}
       disabled {
        bIsDynamicallyLoaded=false
        drawAs=RoundedBox
        imageType=NoImage
        mirroring=NoMirror
        resourceName=None
        resourceObject=null
        tiling=NoTile
        imageSize=<0 0>
        margin=<ltrb 0 0 0 0>
        outlineSettings {
         bUseBrushTransparency=false
         roundingType=FixedRadius
         width=1
         color {colorUseRule=UseColor_ColorTable, specifiedColor=<rgba 0 0 0 0>}
         cornerRadii=<4 4 4 4>
        }
        tintColor {colorUseRule=UseColor_Specified, specifiedColor=<rgba 0.0395 0.0395 0.0395 1>}
        uVRegion {bIsValid=false, max=<0 0>, min=<0 0>}
       }
       disabledForeground {colorUseRule=UseColor_ColorTable, specifiedColor=<rgba 0 0 0 0>}
       hovered {
        bIsDynamicallyLoaded=false
        drawAs=RoundedBox
        imageType=NoImage
        mirroring=NoMirror
        resourceName=None
        resourceObject=null
        tiling=NoTile
        imageSize=<0 0>
        margin=<ltrb 0 0 0 0>
        outlineSettings {
         bUseBrushTransparency=false
         roundingType=FixedRadius
         width=1
         color {colorUseRule=UseColor_ColorTable, specifiedColor=<rgba 0.06 0.06 0.06 1>}
         cornerRadii=<4 4 4 4>
        }
        tintColor {colorUseRule=UseColor_Specified, specifiedColor=<rgba 0.0048 0.0048 0.0048 1>}
        uVRegion {bIsValid=false, max=<0 0>, min=<0 0>}
       }
       hoveredForeground {colorUseRule=UseColor_ColorTable, specifiedColor=<rgba 0 0 0 0>}
       hoveredSlateSound {resourceObject=null}
       normal {
        bIsDynamicallyLoaded=false
        drawAs=RoundedBox
        imageType=NoImage
        mirroring=NoMirror
        resourceName=None
        resourceObject=null
        tiling=NoTile
        imageSize=<0 0>
        margin=<ltrb 0 0 0 0>
        outlineSettings {
         bUseBrushTransparency=false
         roundingType=FixedRadius
         width=1
         color {colorUseRule=UseColor_ColorTable, specifiedColor=<rgba 0.04 0.04 0.04 1>}
         cornerRadii=<4 4 4 4>
        }
        tintColor {colorUseRule=UseColor_Specified, specifiedColor=<rgba 0.0048 0.0048 0.0048 1>}
        uVRegion {bIsValid=false, max=<0 0>, min=<0 0>}
       }
       normalForeground {colorUseRule=UseColor_ColorTable, specifiedColor=<rgba 0 0 0 0>}
       normalPadding=<ltrb 8 1 8 1>
       pressed {
        bIsDynamicallyLoaded=false
        drawAs=RoundedBox
        imageType=NoImage
        mirroring=NoMirror
        resourceName=None
        resourceObject=null
        tiling=NoTile
        imageSize=<0 0>
        margin=<ltrb 0 0 0 0>
        outlineSettings {
         bUseBrushTransparency=false
         roundingType=FixedRadius
         width=1
         color {colorUseRule=UseColor_ColorTable, specifiedColor=<rgba 0 0 0 0>}
         cornerRadii=<4 4 4 4>
        }
        tintColor {colorUseRule=UseColor_Specified, specifiedColor=<rgba 0.0048 0.0048 0.0048 1>}
        uVRegion {bIsValid=false, max=<0 0>, min=<0 0>}
       }
       pressedForeground {colorUseRule=UseColor_ColorTable, specifiedColor=<rgba 0 0 0 0>}
       pressedPadding=<ltrb 8 2 8 0>
       pressedSlateSound {resourceObject=null}
      }
      contentPadding=<ltrb 0 0 0 0>
      downArrowImage {
       bIsDynamicallyLoaded=false
       drawAs=Image
       imageType=Vector
       mirroring=NoMirror
       resourceName=../../../Engine/Content/Slate/Starship/CoreWidgets/ComboBox/wide-chevron-down.svg
       resourceObject=null
       tiling=NoTile
       imageSize=<8 9>
       margin=<ltrb 0 0 0 0>
       outlineSettings {
        bUseBrushTransparency=false
        roundingType=HalfHeightRadius
        width=0
        color {colorUseRule=UseColor_Specified, specifiedColor=<rgba 0 0 0 0>}
        cornerRadii=<0 0 0 0>
       }
       tintColor {colorUseRule=UseColor_Specified, specifiedColor=<rgba 1 1 1 0.3565>}
       uVRegion {bIsValid=false, max=<0 0>, min=<0 0>}
      }
      downArrowPadding=<ltrb 0 2 0 2>
      menuBorderBrush {
       bIsDynamicallyLoaded=false
       drawAs=RoundedBox
       imageType=NoImage
       mirroring=NoMirror
       resourceName=None
       resourceObject=null
       tiling=NoTile
       imageSize=<0 0>
       margin=<ltrb 0 0 0 0>
       outlineSettings {
        bUseBrushTransparency=false
        roundingType=FixedRadius
        width=1
        color {colorUseRule=UseColor_ColorTable, specifiedColor=<rgba 0 0 0 0>}
        cornerRadii=<0 0 0 0>
       }
       tintColor {colorUseRule=UseColor_Specified, specifiedColor=<rgba 0.0395 0.0395 0.0395 1>}
       uVRegion {bIsValid=false, max=<0 0>, min=<0 0>}
      }
      menuBorderPadding=<ltrb 0 0 0 0>
      shadowColorAndOpacity=<rgba 0 0 0 1>
      shadowOffset=<0 0>
     }
     contentPadding=<ltrb 3 1 2 1>
     menuRowPadding=<ltrb 3 1.5 3 1.5>
     pressedSlateSound {resourceObject=null}
     selectionChangeSlateSound {resourceObject=null}
    }
   }
   slot {class=HorizontalBoxSlot, properties {size {sizeRule=Fill, value=2}}}
  }
  HB--Preset {
   widgetClassPath=&HorizontalBox
   parent=VerticalBox_86
   widget=HB--Preset
   slot {class=VerticalBoxSlot, properties {padding=<ltrb 4 2 4 4>}}
  }
  CB_ProfilesSelect {
   widgetClassPath=/Script/Blutility.EditorUtilityComboBoxString
   bIsVariable=true
   parent=HB--Preset
   widget=CB_ProfilesSelect
   properties {
    selectedOption=Item A
    toolTipText=Change which preset is displayed - removing/adding bookmarks is saved automatically
    defaultOptions [Item A, Item B, Item C]
    font {
     bForceMonospaced=false
     bMaterialIsStencil=false
     fontMaterial=null
     fontObject=&Roboto
     letterSpacing=0
     monospacedWidth=1
     size=8.25
     skewAmount=0
     typefaceFontName=Regular
     outlineSettings {
      bApplyOutlineToDropShadows=false
      bMiteredCorners=false
      bSeparateFillAlpha=false
      outlineMaterial=null
      outlineSize=0
      outlineColor=<rgba 0 0 0 1>
     }
    }
    foregroundColor {colorUseRule=UseColor_Specified, specifiedColor=<rgba 0.5271 0.5271 0.5271 1>}
    widgetStyle {
     comboButtonStyle {
      downArrowAlign=VAlign_Center
      buttonStyle {
       clickedSlateSound {resourceObject=null}
       disabled {
        bIsDynamicallyLoaded=false
        drawAs=RoundedBox
        imageType=NoImage
        mirroring=NoMirror
        resourceName=None
        resourceObject=null
        tiling=NoTile
        imageSize=<0 0>
        margin=<ltrb 0 0 0 0>
        outlineSettings {
         bUseBrushTransparency=false
         roundingType=FixedRadius
         width=1
         color {colorUseRule=UseColor_ColorTable, specifiedColor=<rgba 0 0 0 0>}
         cornerRadii=<4 4 4 4>
        }
        tintColor {colorUseRule=UseColor_Specified, specifiedColor=<rgba 0.0395 0.0395 0.0395 1>}
        uVRegion {bIsValid=false, max=<0 0>, min=<0 0>}
       }
       disabledForeground {colorUseRule=UseColor_ColorTable, specifiedColor=<rgba 0 0 0 0>}
       hovered {
        bIsDynamicallyLoaded=false
        drawAs=RoundedBox
        imageType=NoImage
        mirroring=NoMirror
        resourceName=None
        resourceObject=null
        tiling=NoTile
        imageSize=<0 0>
        margin=<ltrb 0 0 0 0>
        outlineSettings {
         bUseBrushTransparency=false
         roundingType=FixedRadius
         width=1
         color {colorUseRule=UseColor_ColorTable, specifiedColor=<rgba 0 0 0 0>}
         cornerRadii=<4 4 4 4>
        }
        tintColor {colorUseRule=UseColor_Specified, specifiedColor=<rgba 0.0048 0.0048 0.0048 1>}
        uVRegion {bIsValid=false, max=<0 0>, min=<0 0>}
       }
       hoveredForeground {colorUseRule=UseColor_ColorTable, specifiedColor=<rgba 0 0 0 0>}
       hoveredSlateSound {resourceObject=null}
       normal {
        bIsDynamicallyLoaded=false
        drawAs=RoundedBox
        imageType=NoImage
        mirroring=NoMirror
        resourceName=None
        resourceObject=null
        tiling=NoTile
        imageSize=<0 0>
        margin=<ltrb 0 0 0 0>
        outlineSettings {
         bUseBrushTransparency=false
         roundingType=FixedRadius
         width=1
         color {colorUseRule=UseColor_ColorTable, specifiedColor=<rgba 0 0 0 0>}
         cornerRadii=<4 4 4 4>
        }
        tintColor {colorUseRule=UseColor_Specified, specifiedColor=<rgba 0.0048 0.0048 0.0048 1>}
        uVRegion {bIsValid=false, max=<0 0>, min=<0 0>}
       }
       normalForeground {colorUseRule=UseColor_ColorTable, specifiedColor=<rgba 0 0 0 0>}
       normalPadding=<ltrb 8 1 8 1>
       pressed {
        bIsDynamicallyLoaded=false
        drawAs=RoundedBox
        imageType=NoImage
        mirroring=NoMirror
        resourceName=None
        resourceObject=null
        tiling=NoTile
        imageSize=<0 0>
        margin=<ltrb 0 0 0 0>
        outlineSettings {
         bUseBrushTransparency=false
         roundingType=FixedRadius
         width=1
         color {colorUseRule=UseColor_ColorTable, specifiedColor=<rgba 0 0 0 0>}
         cornerRadii=<4 4 4 4>
        }
        tintColor {colorUseRule=UseColor_Specified, specifiedColor=<rgba 0.0048 0.0048 0.0048 1>}
        uVRegion {bIsValid=false, max=<0 0>, min=<0 0>}
       }
       pressedForeground {colorUseRule=UseColor_ColorTable, specifiedColor=<rgba 0 0 0 0>}
       pressedPadding=<ltrb 8 2 8 0>
       pressedSlateSound {resourceObject=null}
      }
      contentPadding=<ltrb 0 0 0 0>
      downArrowImage {
       bIsDynamicallyLoaded=false
       drawAs=Image
       imageType=Vector
       mirroring=NoMirror
       resourceName=../../../Engine/Content/Slate/Starship/CoreWidgets/ComboBox/wide-chevron-down.svg
       resourceObject=null
       tiling=NoTile
       imageSize=<8 9>
       margin=<ltrb 0 0 0 0>
       outlineSettings {
        bUseBrushTransparency=false
        roundingType=HalfHeightRadius
        width=0
        color {colorUseRule=UseColor_Specified, specifiedColor=<rgba 0 0 0 0>}
        cornerRadii=<0 0 0 0>
       }
       tintColor {colorUseRule=UseColor_Specified, specifiedColor=<rgba 1 1 1 0.3565>}
       uVRegion {bIsValid=false, max=<0 0>, min=<0 0>}
      }
      downArrowPadding=<ltrb 0 2 0 2>
      menuBorderBrush {
       bIsDynamicallyLoaded=false
       drawAs=RoundedBox
       imageType=NoImage
       mirroring=NoMirror
       resourceName=None
       resourceObject=null
       tiling=NoTile
       imageSize=<0 0>
       margin=<ltrb 0 0 0 0>
       outlineSettings {
        bUseBrushTransparency=false
        roundingType=FixedRadius
        width=1
        color {colorUseRule=UseColor_ColorTable, specifiedColor=<rgba 0 0 0 0>}
        cornerRadii=<0 0 0 0>
       }
       tintColor {colorUseRule=UseColor_Specified, specifiedColor=<rgba 0.0395 0.0395 0.0395 1>}
       uVRegion {bIsValid=false, max=<0 0>, min=<0 0>}
      }
      menuBorderPadding=<ltrb 0 0 0 0>
      shadowColorAndOpacity=<rgba 0 0 0 1>
      shadowOffset=<0 0>
     }
     contentPadding=<ltrb 0 0 0 0>
     menuRowPadding=<ltrb 3 1 3 1>
     pressedSlateSound {resourceObject=null}
     selectionChangeSlateSound {resourceObject=null}
    }
   }
   slot {class=HorizontalBoxSlot, properties {padding=<ltrb 0 0 4 0>, size {sizeRule=Fill, value=1.25}}}
  }
  BN_SavePreset {
   widgetClassPath=&EditorUtilityButton
   parent=HB--Preset
   widget=BN_SavePreset
   properties {
    toolTipText=Create a new preset from the current.
    widgetStyle {
     clickedSlateSound {resourceObject=null}
     disabled {
      bIsDynamicallyLoaded=false
      drawAs=RoundedBox
      imageType=NoImage
      mirroring=NoMirror
      resourceName=None
      resourceObject=null
      tiling=NoTile
      imageSize=<0 0>
      margin=<ltrb 0 0 0 0>
      outlineSettings {
       bUseBrushTransparency=false
       roundingType=FixedRadius
       width=1
       color {colorUseRule=UseColor_ColorTable, specifiedColor=<rgba 0 0 0 0>}
       cornerRadii=<4 4 4 4>
      }
      tintColor {colorUseRule=UseColor_Specified, specifiedColor=<rgba 0.0395 0.0395 0.0395 1>}
      uVRegion {bIsValid=false, max=<0 0>, min=<0 0>}
     }
     disabledForeground {colorUseRule=UseColor_ColorTable, specifiedColor=<rgba 0 0 0 0>}
     hovered {
      bIsDynamicallyLoaded=false
      drawAs=RoundedBox
      imageType=NoImage
      mirroring=NoMirror
      resourceName=None
      resourceObject=null
      tiling=NoTile
      imageSize=<32 32>
      margin=<ltrb 0 0 0 0>
      outlineSettings {
       bUseBrushTransparency=false
       roundingType=FixedRadius
       width=1
       color {colorUseRule=UseColor_Specified, specifiedColor=<rgba 0.2083 0.2083 0.2083 1>}
       cornerRadii=<4 4 4 4>
      }
      tintColor {colorUseRule=UseColor_Specified, specifiedColor=<rgba 0.03 0.03 0.03 1>}
      uVRegion {bIsValid=false, max=<0 0>, min=<0 0>}
     }
     hoveredForeground {colorUseRule=UseColor_ColorTable, specifiedColor=<rgba 0 0 0 0>}
     hoveredSlateSound {resourceObject=null}
     normal {
      bIsDynamicallyLoaded=false
      drawAs=RoundedBox
      imageType=NoImage
      mirroring=NoMirror
      resourceName=None
      resourceObject=null
      tiling=NoTile
      imageSize=<32 32>
      margin=<ltrb 0 0 0 0>
      outlineSettings {
       bUseBrushTransparency=false
       roundingType=FixedRadius
       width=0.5
       color {colorUseRule=UseColor_Specified, specifiedColor=<rgba 0.0573 0.0573 0.0573 1>}
       cornerRadii=<2 2 2 2>
      }
      tintColor {colorUseRule=UseColor_Specified, specifiedColor=<rgba 0.0075 0.0075 0.0075 1>}
      uVRegion {bIsValid=false, max=<0 0>, min=<0 0>}
     }
     normalForeground {colorUseRule=UseColor_ColorTable, specifiedColor=<rgba 0 0 0 0>}
     normalPadding=<ltrb 2 0.5 2 0.5>
     pressed {
      bIsDynamicallyLoaded=false
      drawAs=RoundedBox
      imageType=NoImage
      mirroring=NoMirror
      resourceName=None
      resourceObject=null
      tiling=NoTile
      imageSize=<32 32>
      margin=<ltrb 0 0 0 0>
      outlineSettings {
       bUseBrushTransparency=false
       roundingType=FixedRadius
       width=1
       color {colorUseRule=UseColor_Specified, specifiedColor=<rgba 0.1198 0.0597 0.0168 1>}
       cornerRadii=<4 4 4 4>
      }
      tintColor {colorUseRule=UseColor_Specified, specifiedColor=<rgba 0.0365 0.0308 0.0258 1>}
      uVRegion {bIsValid=false, max=<0 0>, min=<0 0>}
     }
     pressedForeground {colorUseRule=UseColor_ColorTable, specifiedColor=<rgba 0 0 0 0>}
     pressedPadding=<ltrb 3 1 3 0.75>
     pressedSlateSound {resourceObject=null}
    }
   }
   slot {class=HorizontalBoxSlot, properties {horizontalAlignment=HAlign_Right, padding=<ltrb 8 0 0 0>, size {sizeRule=Automatic, value=0.5}}}
  }
  TextBlock_4 {
   widgetClassPath=&TextBlock
   parent=BN_SavePreset
   widget=TextBlock_4
   properties {
    clipping=ClipToBounds
    justification=Center
    text=New
    font {
     bForceMonospaced=false
     bMaterialIsStencil=false
     fontMaterial=null
     fontObject=&Roboto
     letterSpacing=0
     monospacedWidth=1
     size=9
     skewAmount=0
     typefaceFontName=Bold
     outlineSettings {
      bApplyOutlineToDropShadows=false
      bMiteredCorners=false
      bSeparateFillAlpha=false
      outlineMaterial=null
      outlineSize=0
      outlineColor=<rgba 0 0 0 1>
     }
    }
   }
   slot {class=ButtonSlot, properties {padding=<ltrb 4 1 4 1>}}
  }
  TB_SavePreset {
   widgetClassPath=&EditorUtilityEditableTextBox
   bIsVariable=true
   parent=HB--Preset
   widget=TB_SavePreset
   properties {
    hintText=Preset Name
    toolTipText=Create a new preset from the current by the name entered in this box.
    widgetStyle {
     backgroundColor {colorUseRule=UseColor_Specified, specifiedColor=<rgba 1 1 1 1>}
     backgroundImageFocused {
      bIsDynamicallyLoaded=false
      drawAs=RoundedBox
      imageType=NoImage
      mirroring=NoMirror
      resourceName=None
      resourceObject=null
      tiling=NoTile
      imageSize=<0 0>
      margin=<ltrb 0 0 0 0>
      outlineSettings {
       bUseBrushTransparency=false
       roundingType=FixedRadius
       width=1
       color {colorUseRule=UseColor_ColorTable, specifiedColor=<rgba 0 0 0 0>}
       cornerRadii=<4 4 4 4>
      }
      tintColor {colorUseRule=UseColor_Specified, specifiedColor=<rgba 0.0048 0.0048 0.0048 1>}
      uVRegion {bIsValid=false, max=<0 0>, min=<0 0>}
     }
     backgroundImageHovered {
      bIsDynamicallyLoaded=false
      drawAs=RoundedBox
      imageType=NoImage
      mirroring=NoMirror
      resourceName=None
      resourceObject=null
      tiling=NoTile
      imageSize=<0 0>
      margin=<ltrb 0 0 0 0>
      outlineSettings {
       bUseBrushTransparency=false
       roundingType=FixedRadius
       width=1
       color {colorUseRule=UseColor_ColorTable, specifiedColor=<rgba 0 0 0 0>}
       cornerRadii=<4 4 4 4>
      }
      tintColor {colorUseRule=UseColor_Specified, specifiedColor=<rgba 0.0048 0.0048 0.0048 1>}
      uVRegion {bIsValid=false, max=<0 0>, min=<0 0>}
     }
     backgroundImageNormal {
      bIsDynamicallyLoaded=false
      drawAs=RoundedBox
      imageType=NoImage
      mirroring=NoMirror
      resourceName=None
      resourceObject=null
      tiling=NoTile
      imageSize=<32 32>
      margin=<ltrb 0 0 0 0>
      outlineSettings {
       bUseBrushTransparency=false
       roundingType=FixedRadius
       width=0.5
       color {colorUseRule=UseColor_Specified, specifiedColor=<rgba 0.0573 0.0573 0.0573 1>}
       cornerRadii=<2 2 2 2>
      }
      tintColor {colorUseRule=UseColor_Specified, specifiedColor=<rgba 0.0075 0.0075 0.0075 1>}
      uVRegion {bIsValid=false, max=<0 0>, min=<0 0>}
     }
     backgroundImageReadOnly {
      bIsDynamicallyLoaded=false
      drawAs=RoundedBox
      imageType=NoImage
      mirroring=NoMirror
      resourceName=None
      resourceObject=null
      tiling=NoTile
      imageSize=<0 0>
      margin=<ltrb 0 0 0 0>
      outlineSettings {
       bUseBrushTransparency=false
       roundingType=FixedRadius
       width=1
       color {colorUseRule=UseColor_ColorTable, specifiedColor=<rgba 0 0 0 0>}
       cornerRadii=<4 4 4 4>
      }
      tintColor {colorUseRule=UseColor_Specified, specifiedColor=<rgba 0.0284 0.0284 0.0284 1>}
      uVRegion {bIsValid=false, max=<0 0>, min=<0 0>}
     }
     focusedForegroundColor {colorUseRule=UseColor_Specified, specifiedColor=<rgba 1 1 1 1>}
     foregroundColor {colorUseRule=UseColor_Specified, specifiedColor=<rgba 0.5271 0.5271 0.5271 1>}
     hScrollBarPadding=<ltrb 0 0 0 0>
     padding=<ltrb 4 1 4 1>
     readOnlyForegroundColor {colorUseRule=UseColor_Foreground, specifiedColor=<rgba 1 0 1 1>}
     scrollBarStyle {
      thickness=8
      draggedThumbImage {
       bIsDynamicallyLoaded=false
       drawAs=RoundedBox
       imageType=NoImage
       mirroring=NoMirror
       resourceName=None
       resourceObject=null
       tiling=NoTile
       imageSize=<0 0>
       margin=<ltrb 0 0 0 0>
       outlineSettings {
        bUseBrushTransparency=false
        roundingType=FixedRadius
        width=0
        color {colorUseRule=UseColor_Specified, specifiedColor=<rgba 0 0 0 0>}
        cornerRadii=<4 4 4 4>
       }
       tintColor {colorUseRule=UseColor_Specified, specifiedColor=<rgba 0.2159 0.2159 0.2159 1>}
       uVRegion {bIsValid=false, max=<0 0>, min=<0 0>}
      }
      horizontalBackgroundImage {
       bIsDynamicallyLoaded=false
       drawAs=NoDrawType
       imageType=NoImage
       mirroring=NoMirror
       resourceName=None
       resourceObject=null
       tiling=NoTile
       imageSize=<0 0>
       margin=<ltrb 0 0 0 0>
       outlineSettings {
        bUseBrushTransparency=false
        roundingType=HalfHeightRadius
        width=0
        color {colorUseRule=UseColor_Specified, specifiedColor=<rgba 0 0 0 0>}
        cornerRadii=<0 0 0 0>
       }
       tintColor {colorUseRule=UseColor_Specified, specifiedColor=<rgba 1 1 1 1>}
       uVRegion {bIsValid=false, max=<0 0>, min=<0 0>}
      }
      horizontalBottomSlotImage {
       bIsDynamicallyLoaded=false
       drawAs=NoDrawType
       imageType=NoImage
       mirroring=NoMirror
       resourceName=None
       resourceObject=null
       tiling=NoTile
       imageSize=<0 0>
       margin=<ltrb 0 0 0 0>
       outlineSettings {
        bUseBrushTransparency=false
        roundingType=HalfHeightRadius
        width=0
        color {colorUseRule=UseColor_Specified, specifiedColor=<rgba 0 0 0 0>}
        cornerRadii=<0 0 0 0>
       }
       tintColor {colorUseRule=UseColor_Specified, specifiedColor=<rgba 1 1 1 1>}
       uVRegion {bIsValid=false, max=<0 0>, min=<0 0>}
      }
      horizontalTopSlotImage {
       bIsDynamicallyLoaded=false
       drawAs=NoDrawType
       imageType=NoImage
       mirroring=NoMirror
       resourceName=None
       resourceObject=null
       tiling=NoTile
       imageSize=<0 0>
       margin=<ltrb 0 0 0 0>
       outlineSettings {
        bUseBrushTransparency=false
        roundingType=HalfHeightRadius
        width=0
        color {colorUseRule=UseColor_Specified, specifiedColor=<rgba 0 0 0 0>}
        cornerRadii=<0 0 0 0>
       }
       tintColor {colorUseRule=UseColor_Specified, specifiedColor=<rgba 1 1 1 1>}
       uVRegion {bIsValid=false, max=<0 0>, min=<0 0>}
      }
      hoveredThumbImage {
       bIsDynamicallyLoaded=false
       drawAs=RoundedBox
       imageType=NoImage
       mirroring=NoMirror
       resourceName=None
       resourceObject=null
       tiling=NoTile
       imageSize=<0 0>
       margin=<ltrb 0 0 0 0>
       outlineSettings {
        bUseBrushTransparency=false
        roundingType=FixedRadius
        width=0
        color {colorUseRule=UseColor_Specified, specifiedColor=<rgba 0 0 0 0>}
        cornerRadii=<4 4 4 4>
       }
       tintColor {colorUseRule=UseColor_Specified, specifiedColor=<rgba 0.2159 0.2159 0.2159 1>}
       uVRegion {bIsValid=false, max=<0 0>, min=<0 0>}
      }
      normalThumbImage {
       bIsDynamicallyLoaded=false
       drawAs=RoundedBox
       imageType=NoImage
       mirroring=NoMirror
       resourceName=None
       resourceObject=null
       tiling=NoTile
       imageSize=<0 0>
       margin=<ltrb 0 0 0 0>
       outlineSettings {
        bUseBrushTransparency=false
        roundingType=FixedRadius
        width=0
        color {colorUseRule=UseColor_Specified, specifiedColor=<rgba 0 0 0 0>}
        cornerRadii=<4 4 4 4>
       }
       tintColor {colorUseRule=UseColor_Specified, specifiedColor=<rgba 0.0953 0.0953 0.0953 1>}
       uVRegion {bIsValid=false, max=<0 0>, min=<0 0>}
      }
      verticalBackgroundImage {
       bIsDynamicallyLoaded=false
       drawAs=NoDrawType
       imageType=NoImage
       mirroring=NoMirror
       resourceName=None
       resourceObject=null
       tiling=NoTile
       imageSize=<0 0>
       margin=<ltrb 0 0 0 0>
       outlineSettings {
        bUseBrushTransparency=false
        roundingType=HalfHeightRadius
        width=0
        color {colorUseRule=UseColor_Specified, specifiedColor=<rgba 0 0 0 0>}
        cornerRadii=<0 0 0 0>
       }
       tintColor {colorUseRule=UseColor_Specified, specifiedColor=<rgba 1 1 1 1>}
       uVRegion {bIsValid=false, max=<0 0>, min=<0 0>}
      }
      verticalBottomSlotImage {
       bIsDynamicallyLoaded=false
       drawAs=NoDrawType
       imageType=NoImage
       mirroring=NoMirror
       resourceName=None
       resourceObject=null
       tiling=NoTile
       imageSize=<0 0>
       margin=<ltrb 0 0 0 0>
       outlineSettings {
        bUseBrushTransparency=false
        roundingType=HalfHeightRadius
        width=0
        color {colorUseRule=UseColor_Specified, specifiedColor=<rgba 0 0 0 0>}
        cornerRadii=<0 0 0 0>
       }
       tintColor {colorUseRule=UseColor_Specified, specifiedColor=<rgba 1 1 1 1>}
       uVRegion {bIsValid=false, max=<0 0>, min=<0 0>}
      }
      verticalTopSlotImage {
       bIsDynamicallyLoaded=false
       drawAs=NoDrawType
       imageType=NoImage
       mirroring=NoMirror
       resourceName=None
       resourceObject=null
       tiling=NoTile
       imageSize=<0 0>
       margin=<ltrb 0 0 0 0>
       outlineSettings {
        bUseBrushTransparency=false
        roundingType=HalfHeightRadius
        width=0
        color {colorUseRule=UseColor_Specified, specifiedColor=<rgba 0 0 0 0>}
        cornerRadii=<0 0 0 0>
       }
       tintColor {colorUseRule=UseColor_Specified, specifiedColor=<rgba 1 1 1 1>}
       uVRegion {bIsValid=false, max=<0 0>, min=<0 0>}
      }
     }
     textStyle {
      overflowPolicy=Clip
      transformPolicy=None
      colorAndOpacity {colorUseRule=UseColor_Foreground, specifiedColor=<rgba 1 0 1 1>}
      font {
       bForceMonospaced=false
       bMaterialIsStencil=false
       fontMaterial=null
       fontObject=&Roboto
       letterSpacing=0
       monospacedWidth=1
       size=7.5
       skewAmount=0
       typefaceFontName=Regular
       outlineSettings {
        bApplyOutlineToDropShadows=false
        bMiteredCorners=false
        bSeparateFillAlpha=false
        outlineMaterial=null
        outlineSize=0
        outlineColor=<rgba 0 0 0 1>
       }
      }
      highlightColor {colorUseRule=UseColor_ColorTable, specifiedColor=<rgba 0 0 0 0>}
      highlightShape {
       bIsDynamicallyLoaded=false
       drawAs=Image
       imageType=NoImage
       mirroring=NoMirror
       resourceName=None
       resourceObject=null
       tiling=NoTile
       imageSize=<0 0>
       margin=<ltrb 0 0 0 0>
       outlineSettings {
        bUseBrushTransparency=false
        roundingType=HalfHeightRadius
        width=0
        color {colorUseRule=UseColor_Specified, specifiedColor=<rgba 0 0 0 0>}
        cornerRadii=<0 0 0 0>
       }
       tintColor {colorUseRule=UseColor_ColorTable, specifiedColor=<rgba 0 0 0 0>}
       uVRegion {bIsValid=false, max=<0 0>, min=<0 0>}
      }
      selectedBackgroundColor {colorUseRule=UseColor_ColorTable, specifiedColor=<rgba 0 0 0 0>}
      shadowColorAndOpacity=<rgba 0 0 0 1>
      shadowOffset=<0 0>
      strikeBrush {
       bIsDynamicallyLoaded=false
       drawAs=Image
       imageType=NoImage
       mirroring=NoMirror
       resourceName=None
       resourceObject=null
       tiling=NoTile
       imageSize=<32 32>
       margin=<ltrb 0 0 0 0>
       outlineSettings {
        bUseBrushTransparency=false
        roundingType=HalfHeightRadius
        width=0
        color {colorUseRule=UseColor_Specified, specifiedColor=<rgba 0 0 0 0>}
        cornerRadii=<0 0 0 0>
       }
       tintColor {colorUseRule=UseColor_Specified, specifiedColor=<rgba 1 1 1 1>}
       uVRegion {bIsValid=false, max=<0 0>, min=<0 0>}
      }
      underlineBrush {
       bIsDynamicallyLoaded=false
       drawAs=Image
       imageType=NoImage
       mirroring=NoMirror
       resourceName=None
       resourceObject=null
       tiling=NoTile
       imageSize=<32 32>
       margin=<ltrb 0 0 0 0>
       outlineSettings {
        bUseBrushTransparency=false
        roundingType=HalfHeightRadius
        width=0
        color {colorUseRule=UseColor_Specified, specifiedColor=<rgba 0 0 0 0>}
        cornerRadii=<0 0 0 0>
       }
       tintColor {colorUseRule=UseColor_Specified, specifiedColor=<rgba 1 1 1 1>}
       uVRegion {bIsValid=false, max=<0 0>, min=<0 0>}
      }
     }
     vScrollBarPadding=<ltrb 0 0 0 0>
    }
   }
   slot {class=HorizontalBoxSlot, properties {padding=<ltrb 0 0 4 0>, size {sizeRule=Fill, value=1}}}
  }
  BN_RemovePreset {
   widgetClassPath=&EditorUtilityButton
   parent=HB--Preset
   widget=BN_RemovePreset
   properties {
    toolTipText=Remove the current preset
    backgroundColor=<rgba 0.7708 0.1967 0.1445 0.5043>
    widgetStyle {
     clickedSlateSound {resourceObject=null}
     disabled {
      bIsDynamicallyLoaded=false
      drawAs=RoundedBox
      imageType=NoImage
      mirroring=NoMirror
      resourceName=None
      resourceObject=null
      tiling=NoTile
      imageSize=<0 0>
      margin=<ltrb 0 0 0 0>
      outlineSettings {
       bUseBrushTransparency=false
       roundingType=FixedRadius
       width=1
       color {colorUseRule=UseColor_ColorTable, specifiedColor=<rgba 0 0 0 0>}
       cornerRadii=<4 4 4 4>
      }
      tintColor {colorUseRule=UseColor_Specified, specifiedColor=<rgba 0.0395 0.0395 0.0395 1>}
      uVRegion {bIsValid=false, max=<0 0>, min=<0 0>}
     }
     disabledForeground {colorUseRule=UseColor_ColorTable, specifiedColor=<rgba 0 0 0 0>}
     hovered {
      bIsDynamicallyLoaded=false
      drawAs=RoundedBox
      imageType=FullColor
      mirroring=NoMirror
      resourceName=None
      resourceObject=/Game/Athena/UI/InGame/Creative/PropertyEditor/Icon_Picker/Textures/64x64/T_UI_IconLibrary_X_64
      tiling=NoTile
      imageSize=<16 15>
      margin=<ltrb 0 0 0 0>
      outlineSettings {
       bUseBrushTransparency=false
       roundingType=FixedRadius
       width=0
       color {colorUseRule=UseColor_ColorTable, specifiedColor=<rgba 0 0 0 0>}
       cornerRadii=<0 0 0 0>
      }
      tintColor {colorUseRule=UseColor_Specified, specifiedColor=<rgba 0.7448 0.6051 0.6051 1>}
      uVRegion {bIsValid=false, max=<0 0>, min=<0 0>}
     }
     hoveredForeground {colorUseRule=UseColor_ColorTable, specifiedColor=<rgba 0 0 0 0>}
     hoveredSlateSound {resourceObject=null}
     normal {
      bIsDynamicallyLoaded=false
      drawAs=RoundedBox
      imageType=FullColor
      mirroring=NoMirror
      resourceName=None
      resourceObject=&T_UI_IconLibrary_X_64
      tiling=NoTile
      imageSize=<16 15>
      margin=<ltrb 0 0 0 0>
      outlineSettings {
       bUseBrushTransparency=false
       roundingType=FixedRadius
       width=0
       color {colorUseRule=UseColor_ColorTable, specifiedColor=<rgba 0 0 0 0>}
       cornerRadii=<0 0 0 0>
      }
      tintColor {colorUseRule=UseColor_Specified, specifiedColor=<rgba 0.776 0.776 0.776 1>}
      uVRegion {bIsValid=false, max=<0 0>, min=<0 0>}
     }
     normalForeground {colorUseRule=UseColor_ColorTable, specifiedColor=<rgba 0 0 0 0>}
     normalPadding=<ltrb 12 1.5 12 1.5>
     pressed {
      bIsDynamicallyLoaded=false
      drawAs=RoundedBox
      imageType=FullColor
      mirroring=NoMirror
      resourceName=None
      resourceObject=&T_UI_IconLibrary_X_64
      tiling=NoTile
      imageSize=<16 15>
      margin=<ltrb 0 0 0 0>
      outlineSettings {
       bUseBrushTransparency=false
       roundingType=FixedRadius
       width=0
       color {colorUseRule=UseColor_ColorTable, specifiedColor=<rgba 0 0 0 0>}
       cornerRadii=<0 0 0 0>
      }
      tintColor {colorUseRule=UseColor_Specified, specifiedColor=<rgba 0.0174 0.0178 0.026 1>}
      uVRegion {bIsValid=false, max=<0 0>, min=<0 0>}
     }
     pressedForeground {colorUseRule=UseColor_ColorTable, specifiedColor=<rgba 0 0 0 0>}
     pressedPadding=<ltrb 12 2.5 12 0.5>
     pressedSlateSound {resourceObject=null}
    }
   }
   slot {class=HorizontalBoxSlot, properties {horizontalAlignment=HAlign_Right, verticalAlignment=VAlign_Center, padding=<ltrb 0 0 4 0>}}
  }
  Border_1 {
   widgetClassPath=&Border
   parent=VerticalBox_2
   widget=Border_1
   properties {
    verticalAlignment=VAlign_Top
    visibility=Collapsed
    background {
     bIsDynamicallyLoaded=false
     drawAs=Image
     imageType=NoImage
     mirroring=NoMirror
     resourceName=None
     resourceObject=null
     tiling=NoTile
     imageSize=<32 32>
     margin=<ltrb 0 0 0 0>
     outlineSettings {
      bUseBrushTransparency=false
      roundingType=HalfHeightRadius
      width=0
      color {colorUseRule=UseColor_Specified, specifiedColor=<rgba 0 0 0 0>}
      cornerRadii=<0 0 0 0>
     }
     tintColor {colorUseRule=UseColor_Specified, specifiedColor=<rgba 0 0 0 1>}
     uVRegion {bIsValid=false, max=<0 0>, min=<0 0>}
    }
    padding=<ltrb 0 0 0 0>
   }
   slot {class=VerticalBoxSlot, properties {padding=<ltrb 0 2 0 0>}}
  }
  Title {
   widgetClassPath=/Script/Blutility.EditorUtilityEditableText
   bIsVariable=true
   parent=Border_1
   widget=Title
   properties {
    justification=Center
    text=Bookmarks
    toolTipText=The name of the current bookmarks preset
    widgetStyle {
     backgroundImageComposing {
      bIsDynamicallyLoaded=false
      drawAs=Border
      imageType=FullColor
      mirroring=NoMirror
      resourceName=../../../Engine/Content/Slate/Old/HyperlinkDotted.png
      resourceObject=null
      tiling=Both
      imageSize=<0 0>
      margin=<ltrb 0 0 0 0.1875>
      outlineSettings {
       bUseBrushTransparency=false
       roundingType=HalfHeightRadius
       width=0
       color {colorUseRule=UseColor_Specified, specifiedColor=<rgba 0 0 0 0>}
       cornerRadii=<0 0 0 0>
      }
      tintColor {colorUseRule=UseColor_Specified, specifiedColor=<rgba 1 1 1 1>}
      uVRegion {bIsValid=false, max=<0 0>, min=<0 0>}
     }
     backgroundImageSelected {
      bIsDynamicallyLoaded=false
      drawAs=Image
      imageType=NoImage
      mirroring=NoMirror
      resourceName=None
      resourceObject=null
      tiling=NoTile
      imageSize=<0 0>
      margin=<ltrb 0 0 0 0>
      outlineSettings {
       bUseBrushTransparency=false
       roundingType=HalfHeightRadius
       width=0
       color {colorUseRule=UseColor_Specified, specifiedColor=<rgba 0 0 0 0>}
       cornerRadii=<0 0 0 0>
      }
      tintColor {colorUseRule=UseColor_Specified, specifiedColor=<rgba 0 0.162 0.7454 1>}
      uVRegion {bIsValid=false, max=<0 0>, min=<0 0>}
     }
     caretImage {
      bIsDynamicallyLoaded=false
      drawAs=Image
      imageType=NoImage
      mirroring=NoMirror
      resourceName=None
      resourceObject=null
      tiling=NoTile
      imageSize=<0 0>
      margin=<ltrb 0 0 0 0>
      outlineSettings {
       bUseBrushTransparency=false
       roundingType=HalfHeightRadius
       width=0
       color {colorUseRule=UseColor_Specified, specifiedColor=<rgba 0 0 0 0>}
       cornerRadii=<0 0 0 0>
      }
      tintColor {colorUseRule=UseColor_Specified, specifiedColor=<rgba 1 1 1 1>}
      uVRegion {bIsValid=false, max=<0 0>, min=<0 0>}
     }
     colorAndOpacity {colorUseRule=UseColor_Foreground, specifiedColor=<rgba 1 1 1 1>}
     font {
      bForceMonospaced=false
      bMaterialIsStencil=false
      fontMaterial=null
      fontObject=/LauncherSocial/BrutalType
      letterSpacing=0
      monospacedWidth=1
      size=15
      skewAmount=0
      typefaceFontName=Bold
      outlineSettings {
       bApplyOutlineToDropShadows=false
       bMiteredCorners=false
       bSeparateFillAlpha=false
       outlineMaterial=null
       outlineSize=0
       outlineColor=<rgba 0 0 0 1>
      }
     }
     hintColorAndOpacity {colorUseRule=UseColor_Foreground, specifiedColor=<rgba 1 0 1 1>}
     selectedBackgroundColor {colorUseRule=UseColor_Foreground, specifiedColor=<rgba 1 0 1 1>}
    }
   }
   slot {class=BorderSlot, properties {verticalAlignment=VAlign_Top, padding=<ltrb 0 0 0 0>}}
  }
  BookmarksList {
   widgetClassPath=&EditorUtilityScrollBox
   bIsVariable=true
   parent=VerticalBox_2
   widget=BookmarksList
   properties {
    scrollbarThickness=<4 4>
    widgetBarStyle {
     thickness=8
     draggedThumbImage {
      bIsDynamicallyLoaded=false
      drawAs=RoundedBox
      imageType=NoImage
      mirroring=NoMirror
      resourceName=None
      resourceObject=null
      tiling=NoTile
      imageSize=<0 0>
      margin=<ltrb 0 0 0 0>
      outlineSettings {
       bUseBrushTransparency=false
       roundingType=FixedRadius
       width=0
       color {colorUseRule=UseColor_Specified, specifiedColor=<rgba 0 0 0 0>}
       cornerRadii=<4 4 4 4>
      }
      tintColor {colorUseRule=UseColor_Specified, specifiedColor=<rgba 0.2159 0.2159 0.2159 1>}
      uVRegion {bIsValid=false, max=<0 0>, min=<0 0>}
     }
     horizontalBackgroundImage {
      bIsDynamicallyLoaded=false
      drawAs=NoDrawType
      imageType=NoImage
      mirroring=NoMirror
      resourceName=None
      resourceObject=null
      tiling=NoTile
      imageSize=<0 0>
      margin=<ltrb 0 0 0 0>
      outlineSettings {
       bUseBrushTransparency=false
       roundingType=HalfHeightRadius
       width=0
       color {colorUseRule=UseColor_Specified, specifiedColor=<rgba 0 0 0 0>}
       cornerRadii=<0 0 0 0>
      }
      tintColor {colorUseRule=UseColor_Specified, specifiedColor=<rgba 1 1 1 1>}
      uVRegion {bIsValid=false, max=<0 0>, min=<0 0>}
     }
     horizontalBottomSlotImage {
      bIsDynamicallyLoaded=false
      drawAs=NoDrawType
      imageType=NoImage
      mirroring=NoMirror
      resourceName=None
      resourceObject=null
      tiling=NoTile
      imageSize=<0 0>
      margin=<ltrb 0 0 0 0>
      outlineSettings {
       bUseBrushTransparency=false
       roundingType=HalfHeightRadius
       width=0
       color {colorUseRule=UseColor_Specified, specifiedColor=<rgba 0 0 0 0>}
       cornerRadii=<0 0 0 0>
      }
      tintColor {colorUseRule=UseColor_Specified, specifiedColor=<rgba 1 1 1 1>}
      uVRegion {bIsValid=false, max=<0 0>, min=<0 0>}
     }
     horizontalTopSlotImage {
      bIsDynamicallyLoaded=false
      drawAs=NoDrawType
      imageType=NoImage
      mirroring=NoMirror
      resourceName=None
      resourceObject=null
      tiling=NoTile
      imageSize=<0 0>
      margin=<ltrb 0 0 0 0>
      outlineSettings {
       bUseBrushTransparency=false
       roundingType=HalfHeightRadius
       width=0
       color {colorUseRule=UseColor_Specified, specifiedColor=<rgba 0 0 0 0>}
       cornerRadii=<0 0 0 0>
      }
      tintColor {colorUseRule=UseColor_Specified, specifiedColor=<rgba 1 1 1 1>}
      uVRegion {bIsValid=false, max=<0 0>, min=<0 0>}
     }
     hoveredThumbImage {
      bIsDynamicallyLoaded=false
      drawAs=RoundedBox
      imageType=NoImage
      mirroring=NoMirror
      resourceName=None
      resourceObject=null
      tiling=NoTile
      imageSize=<0 0>
      margin=<ltrb 0 0 0 0>
      outlineSettings {
       bUseBrushTransparency=false
       roundingType=FixedRadius
       width=0
       color {colorUseRule=UseColor_Specified, specifiedColor=<rgba 0 0 0 0>}
       cornerRadii=<4 4 4 4>
      }
      tintColor {colorUseRule=UseColor_Specified, specifiedColor=<rgba 0.2159 0.2159 0.2159 1>}
      uVRegion {bIsValid=false, max=<0 0>, min=<0 0>}
     }
     normalThumbImage {
      bIsDynamicallyLoaded=false
      drawAs=RoundedBox
      imageType=NoImage
      mirroring=NoMirror
      resourceName=None
      resourceObject=null
      tiling=NoTile
      imageSize=<0 0>
      margin=<ltrb 0 0 0 0>
      outlineSettings {
       bUseBrushTransparency=false
       roundingType=FixedRadius
       width=0
       color {colorUseRule=UseColor_Specified, specifiedColor=<rgba 0 0 0 0>}
       cornerRadii=<4 4 4 4>
      }
      tintColor {colorUseRule=UseColor_Specified, specifiedColor=<rgba 0.02 0.02 0.02 1>}
      uVRegion {bIsValid=false, max=<0 0>, min=<0 0>}
     }
     verticalBackgroundImage {
      bIsDynamicallyLoaded=false
      drawAs=NoDrawType
      imageType=NoImage
      mirroring=NoMirror
      resourceName=None
      resourceObject=null
      tiling=NoTile
      imageSize=<0 0>
      margin=<ltrb 0 0 0 0>
      outlineSettings {
       bUseBrushTransparency=false
       roundingType=HalfHeightRadius
       width=0
       color {colorUseRule=UseColor_Specified, specifiedColor=<rgba 0 0 0 0>}
       cornerRadii=<0 0 0 0>
      }
      tintColor {colorUseRule=UseColor_Specified, specifiedColor=<rgba 1 1 1 1>}
      uVRegion {bIsValid=false, max=<0 0>, min=<0 0>}
     }
     verticalBottomSlotImage {
      bIsDynamicallyLoaded=false
      drawAs=NoDrawType
      imageType=NoImage
      mirroring=NoMirror
      resourceName=None
      resourceObject=null
      tiling=NoTile
      imageSize=<0 0>
      margin=<ltrb 0 0 0 0>
      outlineSettings {
       bUseBrushTransparency=false
       roundingType=HalfHeightRadius
       width=0
       color {colorUseRule=UseColor_Specified, specifiedColor=<rgba 0 0 0 0>}
       cornerRadii=<0 0 0 0>
      }
      tintColor {colorUseRule=UseColor_Specified, specifiedColor=<rgba 1 1 1 1>}
      uVRegion {bIsValid=false, max=<0 0>, min=<0 0>}
     }
     verticalTopSlotImage {
      bIsDynamicallyLoaded=false
      drawAs=NoDrawType
      imageType=NoImage
      mirroring=NoMirror
      resourceName=None
      resourceObject=null
      tiling=NoTile
      imageSize=<0 0>
      margin=<ltrb 0 0 0 0>
      outlineSettings {
       bUseBrushTransparency=false
       roundingType=HalfHeightRadius
       width=0
       color {colorUseRule=UseColor_Specified, specifiedColor=<rgba 0 0 0 0>}
       cornerRadii=<0 0 0 0>
      }
      tintColor {colorUseRule=UseColor_Specified, specifiedColor=<rgba 1 1 1 1>}
      uVRegion {bIsValid=false, max=<0 0>, min=<0 0>}
     }
    }
   }
   slot {class=VerticalBoxSlot, properties {size {sizeRule=Fill, value=1}}}
  }
  EXP_Configuration {
   widgetClassPath=/Script/Blutility.EditorUtilityExpandableArea
   bIsVariable=true
   parent=VerticalBox_2
   widget=EXP_Configuration
   properties {
    bIsExpanded=true
    areaPadding=<ltrb 0 -0.5 0 1>
    borderBrush {
     bIsDynamicallyLoaded=false
     drawAs=RoundedBox
     imageType=NoImage
     mirroring=NoMirror
     resourceName=None
     resourceObject=null
     tiling=NoTile
     imageSize=<0 0>
     margin=<ltrb 0 0 0 0>
     outlineSettings {
      bUseBrushTransparency=false
      roundingType=FixedRadius
      width=1
      color {colorUseRule=UseColor_Specified, specifiedColor=<rgba 0 0 0 1>}
      cornerRadii=<2 2 0 0>
     }
     tintColor {colorUseRule=UseColor_Specified, specifiedColor=<rgba 0.005 0.005 0.005 1>}
     uVRegion {bIsValid=false, max=<0 0>, min=<0 0>}
    }
    headerPadding=<ltrb 2 3 1 3>
   }
   slot {class=VerticalBoxSlot}
  }
  TextBlock_177 {
   widgetClassPath=&TextBlock
   namedSlotHost=EXP_Configuration
   widget=TextBlock_177
   properties {
    justification=Center
    text=Configuration
    toolTipText="All configuration options, like adding bookmarks and changing presets."
    font {
     bForceMonospaced=false
     bMaterialIsStencil=false
     fontMaterial=null
     fontObject=&Roboto
     letterSpacing=0
     monospacedWidth=1
     size=9
     skewAmount=0
     typefaceFontName=Bold
     outlineSettings {
      bApplyOutlineToDropShadows=false
      bMiteredCorners=false
      bSeparateFillAlpha=false
      outlineMaterial=null
      outlineSize=0
      outlineColor=<rgba 0 0 0 1>
     }
    }
    margin=<ltrb -20 0 0 0>
   }
  }
  Border_2 {
   widgetClassPath=&Border
   namedSlotHost=EXP_Configuration
   widget=Border_2
   properties {
    background {
     bIsDynamicallyLoaded=false
     drawAs=Image
     imageType=NoImage
     mirroring=NoMirror
     resourceName=None
     resourceObject=null
     tiling=NoTile
     imageSize=<32 32>
     margin=<ltrb 0 0 0 0>
     outlineSettings {
      bUseBrushTransparency=false
      roundingType=HalfHeightRadius
      width=0
      color {colorUseRule=UseColor_Specified, specifiedColor=<rgba 0 0 0 0>}
      cornerRadii=<0 0 0 0>
     }
     tintColor {colorUseRule=UseColor_Specified, specifiedColor=<rgba 0 0 0 1>}
     uVRegion {bIsValid=false, max=<0 0>, min=<0 0>}
    }
    padding=<ltrb 4 4 4 4>
   }
  }
  ConfigSwitcher {
   widgetClassPath=&WidgetSwitcher
   bIsVariable=true
   parent=Border_2
   widget=ConfigSwitcher
   slot {class=BorderSlot, properties {padding=<ltrb 4 4 4 4>}}
  }
  VerticalBox_0 {widgetClassPath=&VerticalBox, parent=ConfigSwitcher, widget=VerticalBox_0, slot {class=WidgetSwitcherSlot}}
  Border_3 {
   widgetClassPath=&Border
   parent=VerticalBox_0
   widget=Border_3
   properties {
    background {
     bIsDynamicallyLoaded=false
     drawAs=RoundedBox
     imageType=NoImage
     mirroring=NoMirror
     resourceName=None
     resourceObject=null
     tiling=NoTile
     imageSize=<32 32>
     margin=<ltrb 0 0 0 0>
     outlineSettings {
      bUseBrushTransparency=false
      roundingType=FixedRadius
      width=0
      color {colorUseRule=UseColor_Specified, specifiedColor=<rgba 0 0 0 0>}
      cornerRadii=<3 3 3 3>
     }
     tintColor {colorUseRule=UseColor_Specified, specifiedColor=<rgba 1 1 1 1>}
     uVRegion {bIsValid=false, max=<0 0>, min=<0 0>}
    }
    brushColor=<rgba 0.01 0.01 0.01 1>
   }
   slot {class=VerticalBoxSlot}
  }
  VerticalBox_118 {widgetClassPath=&VerticalBox, parent=Border_3, widget=VerticalBox_118, slot {class=BorderSlot}}
  BTN_GetSelectedAsset {
   widgetClassPath=&EditorUtilityButton
   parent=VerticalBox_118
   widget=BTN_GetSelectedAsset
   properties {
    toolTipText=Loads selected asset info to save new bookmark
    backgroundColor=<rgba 1 1 1 0.8>
    widgetStyle {
     clickedSlateSound {resourceObject=null}
     disabled {
      bIsDynamicallyLoaded=false
      drawAs=RoundedBox
      imageType=NoImage
      mirroring=NoMirror
      resourceName=None
      resourceObject=null
      tiling=NoTile
      imageSize=<0 0>
      margin=<ltrb 0 0 0 0>
      outlineSettings {
       bUseBrushTransparency=false
       roundingType=FixedRadius
       width=1
       color {colorUseRule=UseColor_ColorTable, specifiedColor=<rgba 0 0 0 0>}
       cornerRadii=<4 4 4 4>
      }
      tintColor {colorUseRule=UseColor_Specified, specifiedColor=<rgba 0.0395 0.0395 0.0395 1>}
      uVRegion {bIsValid=false, max=<0 0>, min=<0 0>}
     }
     disabledForeground {colorUseRule=UseColor_ColorTable, specifiedColor=<rgba 0 0 0 0>}
     hovered {
      bIsDynamicallyLoaded=false
      drawAs=RoundedBox
      imageType=NoImage
      mirroring=NoMirror
      resourceName=None
      resourceObject=null
      tiling=NoTile
      imageSize=<32 32>
      margin=<ltrb 0 0 0 0>
      outlineSettings {
       bUseBrushTransparency=false
       roundingType=FixedRadius
       width=1
       color {colorUseRule=UseColor_Specified, specifiedColor=<rgba 0.2083 0.2083 0.2083 1>}
       cornerRadii=<4 4 4 4>
      }
      tintColor {colorUseRule=UseColor_Specified, specifiedColor=<rgba 0.03 0.03 0.03 1>}
      uVRegion {bIsValid=false, max=<0 0>, min=<0 0>}
     }
     hoveredForeground {colorUseRule=UseColor_ColorTable, specifiedColor=<rgba 0 0 0 0>}
     hoveredSlateSound {resourceObject=null}
     normal {
      bIsDynamicallyLoaded=false
      drawAs=RoundedBox
      imageType=NoImage
      mirroring=NoMirror
      resourceName=None
      resourceObject=null
      tiling=NoTile
      imageSize=<32 32>
      margin=<ltrb 0 0 0 0>
      outlineSettings {
       bUseBrushTransparency=false
       roundingType=FixedRadius
       width=0.5
       color {colorUseRule=UseColor_Specified, specifiedColor=<rgba 0.0573 0.0573 0.0573 1>}
       cornerRadii=<2 2 2 2>
      }
      tintColor {colorUseRule=UseColor_Specified, specifiedColor=<rgba 0.0075 0.0075 0.0075 1>}
      uVRegion {bIsValid=false, max=<0 0>, min=<0 0>}
     }
     normalForeground {colorUseRule=UseColor_ColorTable, specifiedColor=<rgba 0 0 0 0>}
     normalPadding=<ltrb 4 0.7 4 0.7>
     pressed {
      bIsDynamicallyLoaded=false
      drawAs=RoundedBox
      imageType=NoImage
      mirroring=NoMirror
      resourceName=None
      resourceObject=null
      tiling=NoTile
      imageSize=<32 32>
      margin=<ltrb 0 0 0 0>
      outlineSettings {
       bUseBrushTransparency=false
       roundingType=FixedRadius
       width=1
       color {colorUseRule=UseColor_Specified, specifiedColor=<rgba 0.1198 0.0597 0.0168 1>}
       cornerRadii=<4 4 4 4>
      }
      tintColor {colorUseRule=UseColor_Specified, specifiedColor=<rgba 0.0365 0.0308 0.0258 1>}
      uVRegion {bIsValid=false, max=<0 0>, min=<0 0>}
     }
     pressedForeground {colorUseRule=UseColor_ColorTable, specifiedColor=<rgba 0 0 0 0>}
     pressedPadding=<ltrb 5.2 1 5 0.7>
     pressedSlateSound {resourceObject=null}
    }
   }
   slot {class=VerticalBoxSlot, properties {verticalAlignment=VAlign_Center, padding=<ltrb 1 0 0 0>}}
  }
  TextBlock_1 {
   widgetClassPath=&TextBlock
   parent=BTN_GetSelectedAsset
   widget=TextBlock_1
   properties {
    justification=Center
    text=Bookmark Selected Asset
    font {
     bForceMonospaced=false
     bMaterialIsStencil=false
     fontMaterial=null
     fontObject=&Roboto
     letterSpacing=0
     monospacedWidth=1
     size=8.25
     skewAmount=0
     typefaceFontName=Bold
     outlineSettings {
      bApplyOutlineToDropShadows=false
      bMiteredCorners=false
      bSeparateFillAlpha=false
      outlineMaterial=null
      outlineSize=0
      outlineColor=<rgba 0 0 0 1>
     }
    }
   }
   slot {class=ButtonSlot, properties {padding=<ltrb 1 2 1 2>}}
  }
  --Name {
   widgetClassPath=&HorizontalBox
   parent=VerticalBox_118
   widget=--Name
   slot {class=VerticalBoxSlot, properties {verticalAlignment=VAlign_Top, padding=<ltrb 0 2 0 1>}}
  }
  TextBlock_13 {
   widgetClassPath=&TextBlock
   parent=--Name
   widget=TextBlock_13
   properties {
    justification=Center
    text="Bookmark Name : "
    font {
     bForceMonospaced=false
     bMaterialIsStencil=false
     fontMaterial=null
     fontObject=&Roboto
     letterSpacing=0
     monospacedWidth=1
     size=9
     skewAmount=0
     typefaceFontName=Bold
     outlineSettings {
      bApplyOutlineToDropShadows=false
      bMiteredCorners=false
      bSeparateFillAlpha=false
      outlineMaterial=null
      outlineSize=0
      outlineColor=<rgba 0 0 0 1>
     }
    }
   }
   slot {class=HorizontalBoxSlot, properties {horizontalAlignment=HAlign_Center, verticalAlignment=VAlign_Center, padding=<ltrb 4 2 1 2>}}
  }
  cfg_Name {
   widgetClassPath=&EditorUtilityEditableText
   bIsVariable=true
   parent=--Name
   widget=cfg_Name
   properties {
    hintText=The name the bookmark will appear as
    toolTipText=The name of the bookmark - doesn't have to be the asset name
    widgetStyle {
     backgroundImageComposing {
      bIsDynamicallyLoaded=false
      drawAs=Border
      imageType=FullColor
      mirroring=NoMirror
      resourceName=../../../Engine/Content/Slate/Old/HyperlinkDotted.png
      resourceObject=null
      tiling=Both
      imageSize=<0 0>
      margin=<ltrb 0 0 0 0.1875>
      outlineSettings {
       bUseBrushTransparency=false
       roundingType=HalfHeightRadius
       width=0
       color {colorUseRule=UseColor_Specified, specifiedColor=<rgba 0 0 0 0>}
       cornerRadii=<0 0 0 0>
      }
      tintColor {colorUseRule=UseColor_Specified, specifiedColor=<rgba 1 1 1 1>}
      uVRegion {bIsValid=false, max=<0 0>, min=<0 0>}
     }
     backgroundImageSelected {
      bIsDynamicallyLoaded=false
      drawAs=Image
      imageType=NoImage
      mirroring=NoMirror
      resourceName=None
      resourceObject=null
      tiling=NoTile
      imageSize=<0 0>
      margin=<ltrb 0 0 0 0>
      outlineSettings {
       bUseBrushTransparency=false
       roundingType=HalfHeightRadius
       width=0
       color {colorUseRule=UseColor_Specified, specifiedColor=<rgba 0 0 0 0>}
       cornerRadii=<0 0 0 0>
      }
      tintColor {colorUseRule=UseColor_Specified, specifiedColor=<rgba 0 0.162 0.7454 1>}
      uVRegion {bIsValid=false, max=<0 0>, min=<0 0>}
     }
     caretImage {
      bIsDynamicallyLoaded=false
      drawAs=Image
      imageType=NoImage
      mirroring=NoMirror
      resourceName=None
      resourceObject=null
      tiling=NoTile
      imageSize=<0 0>
      margin=<ltrb 0 0 0 0>
      outlineSettings {
       bUseBrushTransparency=false
       roundingType=HalfHeightRadius
       width=0
       color {colorUseRule=UseColor_Specified, specifiedColor=<rgba 0 0 0 0>}
       cornerRadii=<0 0 0 0>
      }
      tintColor {colorUseRule=UseColor_Specified, specifiedColor=<rgba 1 1 1 1>}
      uVRegion {bIsValid=false, max=<0 0>, min=<0 0>}
     }
     colorAndOpacity {colorUseRule=UseColor_Foreground, specifiedColor=<rgba 1 1 1 1>}
     font {
      bForceMonospaced=false
      bMaterialIsStencil=false
      fontMaterial=null
      fontObject=&Roboto
      letterSpacing=-70
      monospacedWidth=1
      size=8.25
      skewAmount=0
      typefaceFontName=Regular
      outlineSettings {
       bApplyOutlineToDropShadows=false
       bMiteredCorners=false
       bSeparateFillAlpha=false
       outlineMaterial=null
       outlineSize=0
       outlineColor=<rgba 0 0 0 1>
      }
     }
     hintColorAndOpacity {colorUseRule=UseColor_Foreground, specifiedColor=<rgba 1 0 1 1>}
     selectedBackgroundColor {colorUseRule=UseColor_Foreground, specifiedColor=<rgba 1 0 1 1>}
    }
   }
   slot {class=HorizontalBoxSlot, properties {verticalAlignment=VAlign_Center, padding=<ltrb 5 2 2 2>, size {sizeRule=Fill, value=0.5}}}
  }
  --PathHBox {
   widgetClassPath=&HorizontalBox
   parent=VerticalBox_118
   widget=--PathHBox
   properties {visibility=Visible}
   slot {class=VerticalBoxSlot, properties {verticalAlignment=VAlign_Center, padding=<ltrb 0 1 0 1>}}
  }
  TextBlock_2 {
   widgetClassPath=&TextBlock
   parent=--PathHBox
   widget=TextBlock_2
   properties {
    justification=Center
    text="Path : "
    font {
     bForceMonospaced=false
     bMaterialIsStencil=false
     fontMaterial=null
     fontObject=&Roboto
     letterSpacing=0
     monospacedWidth=1
     size=9
     skewAmount=0
     typefaceFontName=Bold
     outlineSettings {
      bApplyOutlineToDropShadows=false
      bMiteredCorners=false
      bSeparateFillAlpha=false
      outlineMaterial=null
      outlineSize=0
      outlineColor=<rgba 0 0 0 1>
     }
    }
   }
   slot {class=HorizontalBoxSlot, properties {horizontalAlignment=HAlign_Center, verticalAlignment=VAlign_Center, padding=<ltrb 4 2 1 2>}}
  }
  EditorUtilityScrollBox_0 {
   widgetClassPath=&EditorUtilityScrollBox
   parent=--PathHBox
   widget=EditorUtilityScrollBox_0
   properties {
    orientation=Orient_Horizontal
    scrollbarPadding=<ltrb 0 0 0 0>
    scrollbarThickness=<0 0>
    widgetBarStyle {
     thickness=0
     draggedThumbImage {
      bIsDynamicallyLoaded=false
      drawAs=RoundedBox
      imageType=NoImage
      mirroring=NoMirror
      resourceName=None
      resourceObject=null
      tiling=NoTile
      imageSize=<0 0>
      margin=<ltrb 0 0 0 0>
      outlineSettings {
       bUseBrushTransparency=false
       roundingType=FixedRadius
       width=0
       color {colorUseRule=UseColor_Specified, specifiedColor=<rgba 0 0 0 0>}
       cornerRadii=<4 4 4 4>
      }
      tintColor {colorUseRule=UseColor_Specified, specifiedColor=<rgba 0.2159 0.2159 0.2159 1>}
      uVRegion {bIsValid=false, max=<0 0>, min=<0 0>}
     }
     horizontalBackgroundImage {
      bIsDynamicallyLoaded=false
      drawAs=NoDrawType
      imageType=NoImage
      mirroring=NoMirror
      resourceName=None
      resourceObject=null
      tiling=NoTile
      imageSize=<0 0>
      margin=<ltrb 0 0 0 0>
      outlineSettings {
       bUseBrushTransparency=false
       roundingType=HalfHeightRadius
       width=0
       color {colorUseRule=UseColor_Specified, specifiedColor=<rgba 0 0 0 0>}
       cornerRadii=<0 0 0 0>
      }
      tintColor {colorUseRule=UseColor_Specified, specifiedColor=<rgba 1 1 1 1>}
      uVRegion {bIsValid=false, max=<0 0>, min=<0 0>}
     }
     horizontalBottomSlotImage {
      bIsDynamicallyLoaded=false
      drawAs=NoDrawType
      imageType=NoImage
      mirroring=NoMirror
      resourceName=None
      resourceObject=null
      tiling=NoTile
      imageSize=<0 0>
      margin=<ltrb 0 0 0 0>
      outlineSettings {
       bUseBrushTransparency=false
       roundingType=HalfHeightRadius
       width=0
       color {colorUseRule=UseColor_Specified, specifiedColor=<rgba 0 0 0 0>}
       cornerRadii=<0 0 0 0>
      }
      tintColor {colorUseRule=UseColor_Specified, specifiedColor=<rgba 1 1 1 1>}
      uVRegion {bIsValid=false, max=<0 0>, min=<0 0>}
     }
     horizontalTopSlotImage {
      bIsDynamicallyLoaded=false
      drawAs=NoDrawType
      imageType=NoImage
      mirroring=NoMirror
      resourceName=None
      resourceObject=null
      tiling=NoTile
      imageSize=<0 0>
      margin=<ltrb 0 0 0 0>
      outlineSettings {
       bUseBrushTransparency=false
       roundingType=HalfHeightRadius
       width=0
       color {colorUseRule=UseColor_Specified, specifiedColor=<rgba 0 0 0 0>}
       cornerRadii=<0 0 0 0>
      }
      tintColor {colorUseRule=UseColor_Specified, specifiedColor=<rgba 1 1 1 1>}
      uVRegion {bIsValid=false, max=<0 0>, min=<0 0>}
     }
     hoveredThumbImage {
      bIsDynamicallyLoaded=false
      drawAs=RoundedBox
      imageType=NoImage
      mirroring=NoMirror
      resourceName=None
      resourceObject=null
      tiling=NoTile
      imageSize=<0 0>
      margin=<ltrb 0 0 0 0>
      outlineSettings {
       bUseBrushTransparency=false
       roundingType=FixedRadius
       width=0
       color {colorUseRule=UseColor_Specified, specifiedColor=<rgba 0 0 0 0>}
       cornerRadii=<4 4 4 4>
      }
      tintColor {colorUseRule=UseColor_Specified, specifiedColor=<rgba 0.2159 0.2159 0.2159 1>}
      uVRegion {bIsValid=false, max=<0 0>, min=<0 0>}
     }
     normalThumbImage {
      bIsDynamicallyLoaded=false
      drawAs=RoundedBox
      imageType=NoImage
      mirroring=NoMirror
      resourceName=None
      resourceObject=null
      tiling=NoTile
      imageSize=<0 0>
      margin=<ltrb 0 0 0 0>
      outlineSettings {
       bUseBrushTransparency=false
       roundingType=FixedRadius
       width=0
       color {colorUseRule=UseColor_Specified, specifiedColor=<rgba 0 0 0 0>}
       cornerRadii=<4 4 4 4>
      }
      tintColor {colorUseRule=UseColor_Specified, specifiedColor=<rgba 0.0953 0.0953 0.0953 1>}
      uVRegion {bIsValid=false, max=<0 0>, min=<0 0>}
     }
     verticalBackgroundImage {
      bIsDynamicallyLoaded=false
      drawAs=NoDrawType
      imageType=NoImage
      mirroring=NoMirror
      resourceName=None
      resourceObject=null
      tiling=NoTile
      imageSize=<0 0>
      margin=<ltrb 0 0 0 0>
      outlineSettings {
       bUseBrushTransparency=false
       roundingType=HalfHeightRadius
       width=0
       color {colorUseRule=UseColor_Specified, specifiedColor=<rgba 0 0 0 0>}
       cornerRadii=<0 0 0 0>
      }
      tintColor {colorUseRule=UseColor_Specified, specifiedColor=<rgba 1 1 1 1>}
      uVRegion {bIsValid=false, max=<0 0>, min=<0 0>}
     }
     verticalBottomSlotImage {
      bIsDynamicallyLoaded=false
      drawAs=NoDrawType
      imageType=NoImage
      mirroring=NoMirror
      resourceName=None
      resourceObject=null
      tiling=NoTile
      imageSize=<0 0>
      margin=<ltrb 0 0 0 0>
      outlineSettings {
       bUseBrushTransparency=false
       roundingType=HalfHeightRadius
       width=0
       color {colorUseRule=UseColor_Specified, specifiedColor=<rgba 0 0 0 0>}
       cornerRadii=<0 0 0 0>
      }
      tintColor {colorUseRule=UseColor_Specified, specifiedColor=<rgba 1 1 1 1>}
      uVRegion {bIsValid=false, max=<0 0>, min=<0 0>}
     }
     verticalTopSlotImage {
      bIsDynamicallyLoaded=false
      drawAs=NoDrawType
      imageType=NoImage
      mirroring=NoMirror
      resourceName=None
      resourceObject=null
      tiling=NoTile
      imageSize=<0 0>
      margin=<ltrb 0 0 0 0>
      outlineSettings {
       bUseBrushTransparency=false
       roundingType=HalfHeightRadius
       width=0
       color {colorUseRule=UseColor_Specified, specifiedColor=<rgba 0 0 0 0>}
       cornerRadii=<0 0 0 0>
      }
      tintColor {colorUseRule=UseColor_Specified, specifiedColor=<rgba 1 1 1 1>}
      uVRegion {bIsValid=false, max=<0 0>, min=<0 0>}
     }
    }
    widgetStyle {
     barThickness=0
     bottomShadowBrush {
      bIsDynamicallyLoaded=false
      drawAs=Box
      imageType=FullColor
      mirroring=NoMirror
      resourceName=../../../Engine/Content/Slate/Common/ScrollBoxShadowBottom.png
      resourceObject=null
      tiling=NoTile
      imageSize=<16 8>
      margin=<ltrb 0.5 0 0.5 1>
      outlineSettings {
       bUseBrushTransparency=false
       roundingType=HalfHeightRadius
       width=0
       color {colorUseRule=UseColor_Specified, specifiedColor=<rgba 0 0 0 0>}
       cornerRadii=<0 0 0 0>
      }
      tintColor {colorUseRule=UseColor_Specified, specifiedColor=<rgba 1 1 1 1>}
      uVRegion {bIsValid=false, max=<0 0>, min=<0 0>}
     }
     horizontalScrolledContentPadding=<ltrb 0 0 0 0>
     leftShadowBrush {
      bIsDynamicallyLoaded=false
      drawAs=Box
      imageType=FullColor
      mirroring=NoMirror
      resourceName=../../../Engine/Content/Slate/Common/ScrollBoxShadowLeft.png
      resourceObject=null
      tiling=NoTile
      imageSize=<8 16>
      margin=<ltrb 1 0.5 0 0.5>
      outlineSettings {
       bUseBrushTransparency=false
       roundingType=HalfHeightRadius
       width=0
       color {colorUseRule=UseColor_Specified, specifiedColor=<rgba 0 0 0 0>}
       cornerRadii=<0 0 0 0>
      }
      tintColor {colorUseRule=UseColor_Specified, specifiedColor=<rgba 1 1 1 1>}
      uVRegion {bIsValid=false, max=<0 0>, min=<0 0>}
     }
     rightShadowBrush {
      bIsDynamicallyLoaded=false
      drawAs=Box
      imageType=FullColor
      mirroring=NoMirror
      resourceName=../../../Engine/Content/Slate/Common/ScrollBoxShadowRight.png
      resourceObject=null
      tiling=NoTile
      imageSize=<8 16>
      margin=<ltrb 0 0.5 1 0.5>
      outlineSettings {
       bUseBrushTransparency=false
       roundingType=HalfHeightRadius
       width=0
       color {colorUseRule=UseColor_Specified, specifiedColor=<rgba 0 0 0 0>}
       cornerRadii=<0 0 0 0>
      }
      tintColor {colorUseRule=UseColor_Specified, specifiedColor=<rgba 1 1 1 1>}
      uVRegion {bIsValid=false, max=<0 0>, min=<0 0>}
     }
     topShadowBrush {
      bIsDynamicallyLoaded=false
      drawAs=Box
      imageType=FullColor
      mirroring=NoMirror
      resourceName=../../../Engine/Content/Slate/Common/ScrollBoxShadowTop.png
      resourceObject=null
      tiling=NoTile
      imageSize=<16 8>
      margin=<ltrb 0.5 1 0.5 0>
      outlineSettings {
       bUseBrushTransparency=false
       roundingType=HalfHeightRadius
       width=0
       color {colorUseRule=UseColor_Specified, specifiedColor=<rgba 0 0 0 0>}
       cornerRadii=<0 0 0 0>
      }
      tintColor {colorUseRule=UseColor_Specified, specifiedColor=<rgba 1 1 1 1>}
      uVRegion {bIsValid=false, max=<0 0>, min=<0 0>}
     }
     verticalScrolledContentPadding=<ltrb 0 0 0 0>
    }
   }
   slot {class=HorizontalBoxSlot, properties {verticalAlignment=VAlign_Center, padding=<ltrb 5 0 0 0>, size {sizeRule=Fill, value=1}}}
  }
  cfg_Path {
   widgetClassPath=&EditorUtilityEditableText
   bIsVariable=true
   parent=EditorUtilityScrollBox_0
   widget=cfg_Path
   properties {
    hintText=The package path of an Asset
    toolTipText=The package path of the given asset you're looking to bookmark.
    widgetStyle {
     backgroundImageComposing {
      bIsDynamicallyLoaded=false
      drawAs=Border
      imageType=FullColor
      mirroring=NoMirror
      resourceName=../../../Engine/Content/Slate/Old/HyperlinkDotted.png
      resourceObject=null
      tiling=Both
      imageSize=<0 0>
      margin=<ltrb 0 0 0 0.1875>
      outlineSettings {
       bUseBrushTransparency=false
       roundingType=HalfHeightRadius
       width=0
       color {colorUseRule=UseColor_Specified, specifiedColor=<rgba 0 0 0 0>}
       cornerRadii=<0 0 0 0>
      }
      tintColor {colorUseRule=UseColor_Specified, specifiedColor=<rgba 1 1 1 1>}
      uVRegion {bIsValid=false, max=<0 0>, min=<0 0>}
     }
     backgroundImageSelected {
      bIsDynamicallyLoaded=false
      drawAs=Image
      imageType=NoImage
      mirroring=NoMirror
      resourceName=None
      resourceObject=null
      tiling=NoTile
      imageSize=<0 0>
      margin=<ltrb 0 0 0 0>
      outlineSettings {
       bUseBrushTransparency=false
       roundingType=HalfHeightRadius
       width=0
       color {colorUseRule=UseColor_Specified, specifiedColor=<rgba 0 0 0 0>}
       cornerRadii=<0 0 0 0>
      }
      tintColor {colorUseRule=UseColor_Specified, specifiedColor=<rgba 0 0.162 0.7454 1>}
      uVRegion {bIsValid=false, max=<0 0>, min=<0 0>}
     }
     caretImage {
      bIsDynamicallyLoaded=false
      drawAs=Image
      imageType=NoImage
      mirroring=NoMirror
      resourceName=None
      resourceObject=null
      tiling=NoTile
      imageSize=<0 0>
      margin=<ltrb 0 0 0 0>
      outlineSettings {
       bUseBrushTransparency=false
       roundingType=HalfHeightRadius
       width=0
       color {colorUseRule=UseColor_Specified, specifiedColor=<rgba 0 0 0 0>}
       cornerRadii=<0 0 0 0>
      }
      tintColor {colorUseRule=UseColor_Specified, specifiedColor=<rgba 1 1 1 1>}
      uVRegion {bIsValid=false, max=<0 0>, min=<0 0>}
     }
     colorAndOpacity {colorUseRule=UseColor_Foreground, specifiedColor=<rgba 1 1 1 1>}
     font {
      bForceMonospaced=false
      bMaterialIsStencil=false
      fontMaterial=null
      fontObject=&Roboto
      letterSpacing=-70
      monospacedWidth=1
      size=8.25
      skewAmount=0
      typefaceFontName=Regular
      outlineSettings {
       bApplyOutlineToDropShadows=false
       bMiteredCorners=false
       bSeparateFillAlpha=false
       outlineMaterial=null
       outlineSize=0
       outlineColor=<rgba 0 0 0 1>
      }
     }
     hintColorAndOpacity {colorUseRule=UseColor_Foreground, specifiedColor=<rgba 1 0 1 1>}
     selectedBackgroundColor {colorUseRule=UseColor_Foreground, specifiedColor=<rgba 1 0 1 1>}
    }
   }
   slot {class=ScrollBoxSlot, properties {verticalAlignment=VAlign_Center}}
  }
  --Category {
   widgetClassPath=&HorizontalBox
   parent=VerticalBox_118
   widget=--Category
   slot {class=VerticalBoxSlot, properties {verticalAlignment=VAlign_Top, padding=<ltrb 0 1 0 1>}}
  }
  INA_BTN_category {
   widgetClassPath=&EditorUtilityButton
   parent=--Category
   widget=INA_BTN_category
   properties {
    isFocusable=false
    toolTipText=Loads selected asset info to save new bookmark
    visibility=HitTestInvisible
    backgroundColor=<rgba 1 1 1 0>
    widgetStyle {
     clickedSlateSound {resourceObject=null}
     disabled {
      bIsDynamicallyLoaded=false
      drawAs=RoundedBox
      imageType=NoImage
      mirroring=NoMirror
      resourceName=None
      resourceObject=null
      tiling=NoTile
      imageSize=<0 0>
      margin=<ltrb 0 0 0 0>
      outlineSettings {
       bUseBrushTransparency=false
       roundingType=FixedRadius
       width=1
       color {colorUseRule=UseColor_ColorTable, specifiedColor=<rgba 0 0 0 0>}
       cornerRadii=<4 4 4 4>
      }
      tintColor {colorUseRule=UseColor_Specified, specifiedColor=<rgba 0.0395 0.0395 0.0395 1>}
      uVRegion {bIsValid=false, max=<0 0>, min=<0 0>}
     }
     disabledForeground {colorUseRule=UseColor_ColorTable, specifiedColor=<rgba 0 0 0 0>}
     hovered {
      bIsDynamicallyLoaded=false
      drawAs=NoDrawType
      imageType=NoImage
      mirroring=NoMirror
      resourceName=None
      resourceObject=null
      tiling=NoTile
      imageSize=<32 32>
      margin=<ltrb 0 0 0 0>
      outlineSettings {
       bUseBrushTransparency=false
       roundingType=FixedRadius
       width=1
       color {colorUseRule=UseColor_Specified, specifiedColor=<rgba 0.2083 0.2083 0.2083 1>}
       cornerRadii=<4 4 4 4>
      }
      tintColor {colorUseRule=UseColor_Specified, specifiedColor=<rgba 0.03 0.03 0.03 1>}
      uVRegion {bIsValid=false, max=<0 0>, min=<0 0>}
     }
     hoveredForeground {colorUseRule=UseColor_ColorTable, specifiedColor=<rgba 0 0 0 0>}
     hoveredSlateSound {resourceObject=null}
     normal {
      bIsDynamicallyLoaded=false
      drawAs=NoDrawType
      imageType=NoImage
      mirroring=NoMirror
      resourceName=None
      resourceObject=null
      tiling=NoTile
      imageSize=<32 32>
      margin=<ltrb 0 0 0 0>
      outlineSettings {
       bUseBrushTransparency=false
       roundingType=FixedRadius
       width=0.5
       color {colorUseRule=UseColor_Specified, specifiedColor=<rgba 0.0573 0.0573 0.0573 1>}
       cornerRadii=<2 2 2 2>
      }
      tintColor {colorUseRule=UseColor_Specified, specifiedColor=<rgba 0.0075 0.0075 0.0075 1>}
      uVRegion {bIsValid=false, max=<0 0>, min=<0 0>}
     }
     normalForeground {colorUseRule=UseColor_ColorTable, specifiedColor=<rgba 0 0 0 0>}
     normalPadding=<ltrb 4 0 4 0>
     pressed {
      bIsDynamicallyLoaded=false
      drawAs=RoundedBox
      imageType=NoImage
      mirroring=NoMirror
      resourceName=None
      resourceObject=null
      tiling=NoTile
      imageSize=<32 32>
      margin=<ltrb 0 0 0 0>
      outlineSettings {
       bUseBrushTransparency=false
       roundingType=FixedRadius
       width=1
       color {colorUseRule=UseColor_Specified, specifiedColor=<rgba 0.1198 0.0597 0.0168 1>}
       cornerRadii=<4 4 4 4>
      }
      tintColor {colorUseRule=UseColor_Specified, specifiedColor=<rgba 0.0365 0.0308 0.0258 1>}
      uVRegion {bIsValid=false, max=<0 0>, min=<0 0>}
     }
     pressedForeground {colorUseRule=UseColor_ColorTable, specifiedColor=<rgba 0 0 0 0>}
     pressedPadding=<ltrb 5.2 1 5 0.7>
     pressedSlateSound {resourceObject=null}
    }
   }
   slot {class=HorizontalBoxSlot, properties {horizontalAlignment=HAlign_Left, verticalAlignment=VAlign_Center, padding=<ltrb 1 0 0 0>}}
   bIsVariable=~
  }
  TextBlock_6 {
   widgetClassPath=&TextBlock
   parent=INA_BTN_category
   widget=TextBlock_6
   properties {
    justification=Center
    text="Category : "
    font {
     bForceMonospaced=false
     bMaterialIsStencil=false
     fontMaterial=null
     fontObject=&Roboto
     letterSpacing=0
     monospacedWidth=1
     size=9
     skewAmount=0
     typefaceFontName=Bold
     outlineSettings {
      bApplyOutlineToDropShadows=false
      bMiteredCorners=false
      bSeparateFillAlpha=false
      outlineMaterial=null
      outlineSize=0
      outlineColor=<rgba 0 0 0 1>
     }
    }
   }
   slot {class=ButtonSlot, properties {padding=<ltrb 0 2 1 2>}}
  }
  cfg_Category {
   widgetClassPath=&EditorUtilityEditableText
   bIsVariable=true
   parent=--Category
   widget=cfg_Category
   properties {
    hintText=The group the bookmark will appear with
    text=Default
    toolTipText=The category the bookmark will be added under
    widgetStyle {
     backgroundImageComposing {
      bIsDynamicallyLoaded=false
      drawAs=Border
      imageType=FullColor
      mirroring=NoMirror
      resourceName=../../../Engine/Content/Slate/Old/HyperlinkDotted.png
      resourceObject=null
      tiling=Both
      imageSize=<0 0>
      margin=<ltrb 0 0 0 0.1875>
      outlineSettings {
       bUseBrushTransparency=false
       roundingType=HalfHeightRadius
       width=0
       color {colorUseRule=UseColor_Specified, specifiedColor=<rgba 0 0 0 0>}
       cornerRadii=<0 0 0 0>
      }
      tintColor {colorUseRule=UseColor_Specified, specifiedColor=<rgba 1 1 1 1>}
      uVRegion {bIsValid=false, max=<0 0>, min=<0 0>}
     }
     backgroundImageSelected {
      bIsDynamicallyLoaded=false
      drawAs=Image
      imageType=NoImage
      mirroring=NoMirror
      resourceName=None
      resourceObject=null
      tiling=NoTile
      imageSize=<0 0>
      margin=<ltrb 0 0 0 0>
      outlineSettings {
       bUseBrushTransparency=false
       roundingType=HalfHeightRadius
       width=0
       color {colorUseRule=UseColor_Specified, specifiedColor=<rgba 0 0 0 0>}
       cornerRadii=<0 0 0 0>
      }
      tintColor {colorUseRule=UseColor_Specified, specifiedColor=<rgba 0 0.162 0.7454 1>}
      uVRegion {bIsValid=false, max=<0 0>, min=<0 0>}
     }
     caretImage {
      bIsDynamicallyLoaded=false
      drawAs=Image
      imageType=NoImage
      mirroring=NoMirror
      resourceName=None
      resourceObject=null
      tiling=NoTile
      imageSize=<0 0>
      margin=<ltrb 0 0 0 0>
      outlineSettings {
       bUseBrushTransparency=false
       roundingType=HalfHeightRadius
       width=0
       color {colorUseRule=UseColor_Specified, specifiedColor=<rgba 0 0 0 0>}
       cornerRadii=<0 0 0 0>
      }
      tintColor {colorUseRule=UseColor_Specified, specifiedColor=<rgba 1 1 1 1>}
      uVRegion {bIsValid=false, max=<0 0>, min=<0 0>}
     }
     colorAndOpacity {colorUseRule=UseColor_Foreground, specifiedColor=<rgba 1 1 1 1>}
     font {
      bForceMonospaced=false
      bMaterialIsStencil=false
      fontMaterial=null
      fontObject=&Roboto
      letterSpacing=-70
      monospacedWidth=1
      size=8.25
      skewAmount=0
      typefaceFontName=Regular
      outlineSettings {
       bApplyOutlineToDropShadows=false
       bMiteredCorners=false
       bSeparateFillAlpha=false
       outlineMaterial=null
       outlineSize=0
       outlineColor=<rgba 0 0 0 1>
      }
     }
     hintColorAndOpacity {colorUseRule=UseColor_Foreground, specifiedColor=<rgba 1 0 1 1>}
     selectedBackgroundColor {colorUseRule=UseColor_Foreground, specifiedColor=<rgba 1 0 1 1>}
    }
   }
   slot {class=HorizontalBoxSlot, properties {verticalAlignment=VAlign_Center, padding=<ltrb 5 2 2 2>, size {sizeRule=Fill, value=0.5}}}
  }
  --Color {
   widgetClassPath=&HorizontalBox
   parent=VerticalBox_118
   widget=--Color
   slot {class=VerticalBoxSlot, properties {verticalAlignment=VAlign_Top, padding=<ltrb 0 1 0 1>}}
  }
  TextBlock_7 {
   widgetClassPath=&TextBlock
   parent=--Color
   widget=TextBlock_7
   properties {
    justification=Center
    text="Bookmark Color : "
    font {
     bForceMonospaced=false
     bMaterialIsStencil=false
     fontMaterial=null
     fontObject=&Roboto
     letterSpacing=0
     monospacedWidth=1
     size=9
     skewAmount=0
     typefaceFontName=Bold
     outlineSettings {
      bApplyOutlineToDropShadows=false
      bMiteredCorners=false
      bSeparateFillAlpha=false
      outlineMaterial=null
      outlineSize=0
      outlineColor=<rgba 0 0 0 1>
     }
    }
   }
   slot {class=HorizontalBoxSlot, properties {horizontalAlignment=HAlign_Left, verticalAlignment=VAlign_Center, padding=<ltrb 4.5 0 0 0>}}
  }
  PV_Color {
   widgetClassPath=&SinglePropertyView
   parent=--Color
   widget=PV_Color
   properties {nameOverride=" ", propertyName=Color, renderTransform {angle=0, scale=<1 0.7>, shear=<0 0>, translation=<0 -0.1>}}
   slot {class=HorizontalBoxSlot, properties {size {sizeRule=Fill, value=1}}}
  }
  Border_68 {
   widgetClassPath=&Border
   parent=VerticalBox_118
   widget=Border_68
   properties {
    background {
     bIsDynamicallyLoaded=false
     drawAs=RoundedBox
     imageType=NoImage
     mirroring=NoMirror
     resourceName=None
     resourceObject=null
     tiling=NoTile
     imageSize=<32 32>
     margin=<ltrb 0 0 0 0>
     outlineSettings {
      bUseBrushTransparency=false
      roundingType=FixedRadius
      width=0
      color {colorUseRule=UseColor_Specified, specifiedColor=<rgba 0 0 0 0>}
      cornerRadii=<0 0 4 4>
     }
     tintColor {colorUseRule=UseColor_Specified, specifiedColor=<rgba 1 1 1 1>}
     uVRegion {bIsValid=false, max=<0 0>, min=<0 0>}
    }
    brushColor=<rgba 0.007 0.007 0.007 1>
   }
   slot {class=VerticalBoxSlot, properties {verticalAlignment=VAlign_Top, padding=<ltrb -4 0 -4 -2>, size {sizeRule=Fill, value=1}}}
  }
  HorizontalBox_2 {widgetClassPath=&HorizontalBox, parent=Border_68, widget=HorizontalBox_2, slot {class=BorderSlot}}
  BTN_AddBookmark {
   widgetClassPath=&EditorUtilityButton
   parent=HorizontalBox_2
   widget=BTN_AddBookmark
   properties {
    toolTipText="Create a new bookmark in the current preset using the above data.\r\nSaved automatically to the active preset."
    widgetStyle {
     clickedSlateSound {resourceObject=null}
     disabled {
      bIsDynamicallyLoaded=false
      drawAs=RoundedBox
      imageType=NoImage
      mirroring=NoMirror
      resourceName=None
      resourceObject=null
      tiling=NoTile
      imageSize=<0 0>
      margin=<ltrb 0 0 0 0>
      outlineSettings {
       bUseBrushTransparency=false
       roundingType=FixedRadius
       width=1
       color {colorUseRule=UseColor_ColorTable, specifiedColor=<rgba 0 0 0 0>}
       cornerRadii=<4 4 4 4>
      }
      tintColor {colorUseRule=UseColor_Specified, specifiedColor=<rgba 0.0395 0.0395 0.0395 1>}
      uVRegion {bIsValid=false, max=<0 0>, min=<0 0>}
     }
     disabledForeground {colorUseRule=UseColor_ColorTable, specifiedColor=<rgba 0 0 0 0>}
     hovered {
      bIsDynamicallyLoaded=false
      drawAs=RoundedBox
      imageType=NoImage
      mirroring=NoMirror
      resourceName=None
      resourceObject=null
      tiling=NoTile
      imageSize=<32 32>
      margin=<ltrb 0 0 0 0>
      outlineSettings {
       bUseBrushTransparency=false
       roundingType=FixedRadius
       width=1
       color {colorUseRule=UseColor_Specified, specifiedColor=<rgba 0.2083 0.2083 0.2083 1>}
       cornerRadii=<4 4 4 4>
      }
      tintColor {colorUseRule=UseColor_Specified, specifiedColor=<rgba 0.03 0.03 0.03 1>}
      uVRegion {bIsValid=false, max=<0 0>, min=<0 0>}
     }
     hoveredForeground {colorUseRule=UseColor_ColorTable, specifiedColor=<rgba 0 0 0 0>}
     hoveredSlateSound {resourceObject=null}
     normal {
      bIsDynamicallyLoaded=false
      drawAs=RoundedBox
      imageType=NoImage
      mirroring=NoMirror
      resourceName=None
      resourceObject=null
      tiling=NoTile
      imageSize=<32 32>
      margin=<ltrb 0 0 0 0>
      outlineSettings {
       bUseBrushTransparency=false
       roundingType=FixedRadius
       width=0.5
       color {colorUseRule=UseColor_Specified, specifiedColor=<rgba 0.0573 0.0573 0.0573 1>}
       cornerRadii=<2 2 2 2>
      }
      tintColor {colorUseRule=UseColor_Specified, specifiedColor=<rgba 0.0075 0.0075 0.0075 1>}
      uVRegion {bIsValid=false, max=<0 0>, min=<0 0>}
     }
     normalForeground {colorUseRule=UseColor_ColorTable, specifiedColor=<rgba 0 0 0 0>}
     normalPadding=<ltrb 4 0.7 4 0.7>
     pressed {
      bIsDynamicallyLoaded=false
      drawAs=RoundedBox
      imageType=NoImage
      mirroring=NoMirror
      resourceName=None
      resourceObject=null
      tiling=NoTile
      imageSize=<32 32>
      margin=<ltrb 0 0 0 0>
      outlineSettings {
       bUseBrushTransparency=false
       roundingType=FixedRadius
       width=1
       color {colorUseRule=UseColor_Specified, specifiedColor=<rgba 0.1198 0.0597 0.0168 1>}
       cornerRadii=<4 4 4 4>
      }
      tintColor {colorUseRule=UseColor_Specified, specifiedColor=<rgba 0.0365 0.0308 0.0258 1>}
      uVRegion {bIsValid=false, max=<0 0>, min=<0 0>}
     }
     pressedForeground {colorUseRule=UseColor_ColorTable, specifiedColor=<rgba 0 0 0 0>}
     pressedPadding=<ltrb 5.2 1 5 0.7>
     pressedSlateSound {resourceObject=null}
    }
   }
   slot {class=HorizontalBoxSlot, properties {padding=<ltrb 3 1 3 1>, size {sizeRule=Fill, value=1}}}
  }
  TX_Configurebookmark {
   widgetClassPath=&TextBlock
   bIsVariable=true
   parent=BTN_AddBookmark
   widget=TX_Configurebookmark
   properties {
    justification=Center
    text=Add Bookmark
    font {
     bForceMonospaced=false
     bMaterialIsStencil=false
     fontMaterial=null
     fontObject=&Roboto
     letterSpacing=0
     monospacedWidth=1
     size=9
     skewAmount=0
     typefaceFontName=Bold
     outlineSettings {
      bApplyOutlineToDropShadows=false
      bMiteredCorners=false
      bSeparateFillAlpha=false
      outlineMaterial=null
      outlineSize=0
      outlineColor=<rgba 0 0 0 1>
     }
    }
   }
   slot {class=ButtonSlot, properties {verticalAlignment=VAlign_Fill, padding=<ltrb 2 1.5 2 1>}}
  }
  BTN_WSLoadDataAssetConfig {
   widgetClassPath=&EditorUtilityButton
   parent=VerticalBox_0
   widget=BTN_WSLoadDataAssetConfig
   properties {
    toolTipText=View the Data Asset configuration menu.
    widgetStyle {
     clickedSlateSound {resourceObject=null}
     disabled {
      bIsDynamicallyLoaded=false
      drawAs=RoundedBox
      imageType=NoImage
      mirroring=NoMirror
      resourceName=None
      resourceObject=null
      tiling=NoTile
      imageSize=<0 0>
      margin=<ltrb 0 0 0 0>
      outlineSettings {
       bUseBrushTransparency=false
       roundingType=FixedRadius
       width=1
       color {colorUseRule=UseColor_ColorTable, specifiedColor=<rgba 0 0 0 0>}
       cornerRadii=<4 4 4 4>
      }
      tintColor {colorUseRule=UseColor_Specified, specifiedColor=<rgba 0.0395 0.0395 0.0395 1>}
      uVRegion {bIsValid=false, max=<0 0>, min=<0 0>}
     }
     disabledForeground {colorUseRule=UseColor_ColorTable, specifiedColor=<rgba 0 0 0 0>}
     hovered {
      bIsDynamicallyLoaded=false
      drawAs=RoundedBox
      imageType=NoImage
      mirroring=NoMirror
      resourceName=None
      resourceObject=null
      tiling=NoTile
      imageSize=<32 32>
      margin=<ltrb 0 0 0 0>
      outlineSettings {
       bUseBrushTransparency=false
       roundingType=FixedRadius
       width=1
       color {colorUseRule=UseColor_Specified, specifiedColor=<rgba 0.2083 0.2083 0.2083 1>}
       cornerRadii=<4 4 4 4>
      }
      tintColor {colorUseRule=UseColor_Specified, specifiedColor=<rgba 0.04 0.04 0.04 1>}
      uVRegion {bIsValid=false, max=<0 0>, min=<0 0>}
     }
     hoveredForeground {colorUseRule=UseColor_ColorTable, specifiedColor=<rgba 0 0 0 0>}
     hoveredSlateSound {resourceObject=null}
     normal {
      bIsDynamicallyLoaded=false
      drawAs=RoundedBox
      imageType=NoImage
      mirroring=NoMirror
      resourceName=None
      resourceObject=null
      tiling=NoTile
      imageSize=<32 32>
      margin=<ltrb 0 0 0 0>
      outlineSettings {
       bUseBrushTransparency=false
       roundingType=FixedRadius
       width=0.5
       color {colorUseRule=UseColor_Specified, specifiedColor=<rgba 0.07 0.07 0.07 1>}
       cornerRadii=<2 2 2 2>
      }
      tintColor {colorUseRule=UseColor_Specified, specifiedColor=<rgba 0.02 0.02 0.02 1>}
      uVRegion {bIsValid=false, max=<0 0>, min=<0 0>}
     }
     normalForeground {colorUseRule=UseColor_ColorTable, specifiedColor=<rgba 0 0 0 0>}
     normalPadding=<ltrb 4 0.7 4 0.7>
     pressed {
      bIsDynamicallyLoaded=false
      drawAs=RoundedBox
      imageType=NoImage
      mirroring=NoMirror
      resourceName=None
      resourceObject=null
      tiling=NoTile
      imageSize=<32 32>
      margin=<ltrb 0 0 0 0>
      outlineSettings {
       bUseBrushTransparency=false
       roundingType=FixedRadius
       width=1
       color {colorUseRule=UseColor_Specified, specifiedColor=<rgba 0.1198 0.0597 0.0168 1>}
       cornerRadii=<4 4 4 4>
      }
      tintColor {colorUseRule=UseColor_Specified, specifiedColor=<rgba 0.0365 0.0308 0.0258 1>}
      uVRegion {bIsValid=false, max=<0 0>, min=<0 0>}
     }
     pressedForeground {colorUseRule=UseColor_ColorTable, specifiedColor=<rgba 0 0 0 0>}
     pressedPadding=<ltrb 5.2 1 5 0.7>
     pressedSlateSound {resourceObject=null}
    }
   }
   slot {class=VerticalBoxSlot, properties {padding=<ltrb 2 4 2 2>}}
  }
  TextBlock {
   widgetClassPath=&TextBlock
   parent=BTN_WSLoadDataAssetConfig
   widget=TextBlock
   properties {
    clipping=ClipToBounds
    justification=Center
    text=Add / Update Workflow
    textOverflowPolicy=Ellipsis
    font {
     bForceMonospaced=false
     bMaterialIsStencil=false
     fontMaterial=null
     fontObject=&Roboto
     letterSpacing=0
     monospacedWidth=1
     size=6.75
     skewAmount=0
     typefaceFontName=Bold
     outlineSettings {
      bApplyOutlineToDropShadows=false
      bMiteredCorners=false
      bSeparateFillAlpha=false
      outlineMaterial=null
      outlineSize=0
      outlineColor=<rgba 0 0 0 1>
     }
    }
   }
   slot {class=ButtonSlot, properties {padding=<ltrb 2 1 2 1>}}
  }
  VerticalBox_1 {
   widgetClassPath=&VerticalBox
   parent=ConfigSwitcher
   widget=VerticalBox_1
   slot {class=WidgetSwitcherSlot, properties {padding=<ltrb 0 1 0 0>}}
  }
  Border_66 {
   widgetClassPath=&Border
   parent=VerticalBox_1
   widget=Border_66
   properties {
    background {
     bIsDynamicallyLoaded=false
     drawAs=RoundedBox
     imageType=NoImage
     mirroring=NoMirror
     resourceName=None
     resourceObject=null
     tiling=NoTile
     imageSize=<32 32>
     margin=<ltrb 0 0 0 0>
     outlineSettings {
      bUseBrushTransparency=false
      roundingType=FixedRadius
      width=0
      color {colorUseRule=UseColor_Specified, specifiedColor=<rgba 0 0 0 0>}
      cornerRadii=<3 3 4 4>
     }
     tintColor {colorUseRule=UseColor_Specified, specifiedColor=<rgba 1 1 1 1>}
     uVRegion {bIsValid=false, max=<0 0>, min=<0 0>}
    }
    brushColor=<rgba 0.0097 0.0097 0.0097 1>
    padding=<ltrb 0 0 0 0>
   }
   slot {class=VerticalBoxSlot, properties {padding=<ltrb 1 0 1 2>}}
  }
  VerticalBox_78 {
   widgetClassPath=&VerticalBox
   parent=Border_66
   widget=VerticalBox_78
   slot {class=BorderSlot, properties {padding=<ltrb 0 0 0 0>}}
  }
  HorizontalBox_4 {
   widgetClassPath=&HorizontalBox
   parent=VerticalBox_78
   widget=HorizontalBox_4
   slot {class=VerticalBoxSlot, properties {padding=<ltrb 2 4 2 2>, size {sizeRule=Fill, value=1}}}
  }
  TextBlock_5 {
   widgetClassPath=&TextBlock
   parent=HorizontalBox_4
   widget=TextBlock_5
   properties {
    justification=Center
    text="DataAsset Name: "
    font {
     bForceMonospaced=false
     bMaterialIsStencil=false
     fontMaterial=null
     fontObject=&Roboto
     letterSpacing=0
     monospacedWidth=1
     size=9
     skewAmount=0
     typefaceFontName=Bold
     outlineSettings {
      bApplyOutlineToDropShadows=false
      bMiteredCorners=false
      bSeparateFillAlpha=false
      outlineMaterial=null
      outlineSize=0
      outlineColor=<rgba 0 0 0 1>
     }
    }
   }
   slot {class=HorizontalBoxSlot, properties {horizontalAlignment=HAlign_Left, verticalAlignment=VAlign_Center, padding=<ltrb 4 0 0 0>}}
  }
  TB_BookCaseAssetName {
   widgetClassPath=&EditorUtilityEditableTextBox
   bIsVariable=true
   parent=HorizontalBox_4
   widget=TB_BookCaseAssetName
   properties {
    hintText=New BookShelf Asset Name
    toolTipText=The name of the Data Asset that will be created when clicking the button below. * DA_ prefix will be enforced
    widgetStyle {
     backgroundColor {colorUseRule=UseColor_Specified, specifiedColor=<rgba 1 1 1 1>}
     backgroundImageFocused {
      bIsDynamicallyLoaded=false
      drawAs=RoundedBox
      imageType=NoImage
      mirroring=NoMirror
      resourceName=None
      resourceObject=null
      tiling=NoTile
      imageSize=<0 0>
      margin=<ltrb 0 0 0 0>
      outlineSettings {
       bUseBrushTransparency=false
       roundingType=FixedRadius
       width=1
       color {colorUseRule=UseColor_ColorTable, specifiedColor=<rgba 0 0 0 0>}
       cornerRadii=<4 4 4 4>
      }
      tintColor {colorUseRule=UseColor_Specified, specifiedColor=<rgba 0.0048 0.0048 0.0048 1>}
      uVRegion {bIsValid=false, max=<0 0>, min=<0 0>}
     }
     backgroundImageHovered {
      bIsDynamicallyLoaded=false
      drawAs=RoundedBox
      imageType=NoImage
      mirroring=NoMirror
      resourceName=None
      resourceObject=null
      tiling=NoTile
      imageSize=<0 0>
      margin=<ltrb 0 0 0 0>
      outlineSettings {
       bUseBrushTransparency=false
       roundingType=FixedRadius
       width=1
       color {colorUseRule=UseColor_ColorTable, specifiedColor=<rgba 0 0 0 0>}
       cornerRadii=<4 4 4 4>
      }
      tintColor {colorUseRule=UseColor_Specified, specifiedColor=<rgba 0.0048 0.0048 0.0048 1>}
      uVRegion {bIsValid=false, max=<0 0>, min=<0 0>}
     }
     backgroundImageNormal {
      bIsDynamicallyLoaded=false
      drawAs=RoundedBox
      imageType=NoImage
      mirroring=NoMirror
      resourceName=None
      resourceObject=null
      tiling=NoTile
      imageSize=<32 32>
      margin=<ltrb 0 0 0 0>
      outlineSettings {
       bUseBrushTransparency=false
       roundingType=FixedRadius
       width=0.5
       color {colorUseRule=UseColor_Specified, specifiedColor=<rgba 0.0573 0.0573 0.0573 1>}
       cornerRadii=<2 2 2 2>
      }
      tintColor {colorUseRule=UseColor_Specified, specifiedColor=<rgba 0.0075 0.0075 0.0075 1>}
      uVRegion {bIsValid=false, max=<0 0>, min=<0 0>}
     }
     backgroundImageReadOnly {
      bIsDynamicallyLoaded=false
      drawAs=RoundedBox
      imageType=NoImage
      mirroring=NoMirror
      resourceName=None
      resourceObject=null
      tiling=NoTile
      imageSize=<0 0>
      margin=<ltrb 0 0 0 0>
      outlineSettings {
       bUseBrushTransparency=false
       roundingType=FixedRadius
       width=1
       color {colorUseRule=UseColor_ColorTable, specifiedColor=<rgba 0 0 0 0>}
       cornerRadii=<4 4 4 4>
      }
      tintColor {colorUseRule=UseColor_Specified, specifiedColor=<rgba 0.0284 0.0284 0.0284 1>}
      uVRegion {bIsValid=false, max=<0 0>, min=<0 0>}
     }
     focusedForegroundColor {colorUseRule=UseColor_Specified, specifiedColor=<rgba 1 1 1 1>}
     foregroundColor {colorUseRule=UseColor_Specified, specifiedColor=<rgba 0.5271 0.5271 0.5271 1>}
     hScrollBarPadding=<ltrb 0 0 0 0>
     padding=<ltrb 4 1 4 1>
     readOnlyForegroundColor {colorUseRule=UseColor_Foreground, specifiedColor=<rgba 1 0 1 1>}
     scrollBarStyle {
      thickness=8
      draggedThumbImage {
       bIsDynamicallyLoaded=false
       drawAs=RoundedBox
       imageType=NoImage
       mirroring=NoMirror
       resourceName=None
       resourceObject=null
       tiling=NoTile
       imageSize=<0 0>
       margin=<ltrb 0 0 0 0>
       outlineSettings {
        bUseBrushTransparency=false
        roundingType=FixedRadius
        width=0
        color {colorUseRule=UseColor_Specified, specifiedColor=<rgba 0 0 0 0>}
        cornerRadii=<4 4 4 4>
       }
       tintColor {colorUseRule=UseColor_Specified, specifiedColor=<rgba 0.2159 0.2159 0.2159 1>}
       uVRegion {bIsValid=false, max=<0 0>, min=<0 0>}
      }
      horizontalBackgroundImage {
       bIsDynamicallyLoaded=false
       drawAs=NoDrawType
       imageType=NoImage
       mirroring=NoMirror
       resourceName=None
       resourceObject=null
       tiling=NoTile
       imageSize=<0 0>
       margin=<ltrb 0 0 0 0>
       outlineSettings {
        bUseBrushTransparency=false
        roundingType=HalfHeightRadius
        width=0
        color {colorUseRule=UseColor_Specified, specifiedColor=<rgba 0 0 0 0>}
        cornerRadii=<0 0 0 0>
       }
       tintColor {colorUseRule=UseColor_Specified, specifiedColor=<rgba 1 1 1 1>}
       uVRegion {bIsValid=false, max=<0 0>, min=<0 0>}
      }
      horizontalBottomSlotImage {
       bIsDynamicallyLoaded=false
       drawAs=NoDrawType
       imageType=NoImage
       mirroring=NoMirror
       resourceName=None
       resourceObject=null
       tiling=NoTile
       imageSize=<0 0>
       margin=<ltrb 0 0 0 0>
       outlineSettings {
        bUseBrushTransparency=false
        roundingType=HalfHeightRadius
        width=0
        color {colorUseRule=UseColor_Specified, specifiedColor=<rgba 0 0 0 0>}
        cornerRadii=<0 0 0 0>
       }
       tintColor {colorUseRule=UseColor_Specified, specifiedColor=<rgba 1 1 1 1>}
       uVRegion {bIsValid=false, max=<0 0>, min=<0 0>}
      }
      horizontalTopSlotImage {
       bIsDynamicallyLoaded=false
       drawAs=NoDrawType
       imageType=NoImage
       mirroring=NoMirror
       resourceName=None
       resourceObject=null
       tiling=NoTile
       imageSize=<0 0>
       margin=<ltrb 0 0 0 0>
       outlineSettings {
        bUseBrushTransparency=false
        roundingType=HalfHeightRadius
        width=0
        color {colorUseRule=UseColor_Specified, specifiedColor=<rgba 0 0 0 0>}
        cornerRadii=<0 0 0 0>
       }
       tintColor {colorUseRule=UseColor_Specified, specifiedColor=<rgba 1 1 1 1>}
       uVRegion {bIsValid=false, max=<0 0>, min=<0 0>}
      }
      hoveredThumbImage {
       bIsDynamicallyLoaded=false
       drawAs=RoundedBox
       imageType=NoImage
       mirroring=NoMirror
       resourceName=None
       resourceObject=null
       tiling=NoTile
       imageSize=<0 0>
       margin=<ltrb 0 0 0 0>
       outlineSettings {
        bUseBrushTransparency=false
        roundingType=FixedRadius
        width=0
        color {colorUseRule=UseColor_Specified, specifiedColor=<rgba 0 0 0 0>}
        cornerRadii=<4 4 4 4>
       }
       tintColor {colorUseRule=UseColor_Specified, specifiedColor=<rgba 0.2159 0.2159 0.2159 1>}
       uVRegion {bIsValid=false, max=<0 0>, min=<0 0>}
      }
      normalThumbImage {
       bIsDynamicallyLoaded=false
       drawAs=RoundedBox
       imageType=NoImage
       mirroring=NoMirror
       resourceName=None
       resourceObject=null
       tiling=NoTile
       imageSize=<0 0>
       margin=<ltrb 0 0 0 0>
       outlineSettings {
        bUseBrushTransparency=false
        roundingType=FixedRadius
        width=0
        color {colorUseRule=UseColor_Specified, specifiedColor=<rgba 0 0 0 0>}
        cornerRadii=<4 4 4 4>
       }
       tintColor {colorUseRule=UseColor_Specified, specifiedColor=<rgba 0.0953 0.0953 0.0953 1>}
       uVRegion {bIsValid=false, max=<0 0>, min=<0 0>}
      }
      verticalBackgroundImage {
       bIsDynamicallyLoaded=false
       drawAs=NoDrawType
       imageType=NoImage
       mirroring=NoMirror
       resourceName=None
       resourceObject=null
       tiling=NoTile
       imageSize=<0 0>
       margin=<ltrb 0 0 0 0>
       outlineSettings {
        bUseBrushTransparency=false
        roundingType=HalfHeightRadius
        width=0
        color {colorUseRule=UseColor_Specified, specifiedColor=<rgba 0 0 0 0>}
        cornerRadii=<0 0 0 0>
       }
       tintColor {colorUseRule=UseColor_Specified, specifiedColor=<rgba 1 1 1 1>}
       uVRegion {bIsValid=false, max=<0 0>, min=<0 0>}
      }
      verticalBottomSlotImage {
       bIsDynamicallyLoaded=false
       drawAs=NoDrawType
       imageType=NoImage
       mirroring=NoMirror
       resourceName=None
       resourceObject=null
       tiling=NoTile
       imageSize=<0 0>
       margin=<ltrb 0 0 0 0>
       outlineSettings {
        bUseBrushTransparency=false
        roundingType=HalfHeightRadius
        width=0
        color {colorUseRule=UseColor_Specified, specifiedColor=<rgba 0 0 0 0>}
        cornerRadii=<0 0 0 0>
       }
       tintColor {colorUseRule=UseColor_Specified, specifiedColor=<rgba 1 1 1 1>}
       uVRegion {bIsValid=false, max=<0 0>, min=<0 0>}
      }
      verticalTopSlotImage {
       bIsDynamicallyLoaded=false
       drawAs=NoDrawType
       imageType=NoImage
       mirroring=NoMirror
       resourceName=None
       resourceObject=null
       tiling=NoTile
       imageSize=<0 0>
       margin=<ltrb 0 0 0 0>
       outlineSettings {
        bUseBrushTransparency=false
        roundingType=HalfHeightRadius
        width=0
        color {colorUseRule=UseColor_Specified, specifiedColor=<rgba 0 0 0 0>}
        cornerRadii=<0 0 0 0>
       }
       tintColor {colorUseRule=UseColor_Specified, specifiedColor=<rgba 1 1 1 1>}
       uVRegion {bIsValid=false, max=<0 0>, min=<0 0>}
      }
     }
     textStyle {
      overflowPolicy=Clip
      transformPolicy=None
      colorAndOpacity {colorUseRule=UseColor_Foreground, specifiedColor=<rgba 1 0 1 1>}
      font {
       bForceMonospaced=false
       bMaterialIsStencil=false
       fontMaterial=null
       fontObject=&Roboto
       letterSpacing=0
       monospacedWidth=1
       size=7.5
       skewAmount=0
       typefaceFontName=Regular
       outlineSettings {
        bApplyOutlineToDropShadows=false
        bMiteredCorners=false
        bSeparateFillAlpha=false
        outlineMaterial=null
        outlineSize=0
        outlineColor=<rgba 0 0 0 1>
       }
      }
      highlightColor {colorUseRule=UseColor_ColorTable, specifiedColor=<rgba 0 0 0 0>}
      highlightShape {
       bIsDynamicallyLoaded=false
       drawAs=Image
       imageType=NoImage
       mirroring=NoMirror
       resourceName=None
       resourceObject=null
       tiling=NoTile
       imageSize=<0 0>
       margin=<ltrb 0 0 0 0>
       outlineSettings {
        bUseBrushTransparency=false
        roundingType=HalfHeightRadius
        width=0
        color {colorUseRule=UseColor_Specified, specifiedColor=<rgba 0 0 0 0>}
        cornerRadii=<0 0 0 0>
       }
       tintColor {colorUseRule=UseColor_ColorTable, specifiedColor=<rgba 0 0 0 0>}
       uVRegion {bIsValid=false, max=<0 0>, min=<0 0>}
      }
      selectedBackgroundColor {colorUseRule=UseColor_ColorTable, specifiedColor=<rgba 0 0 0 0>}
      shadowColorAndOpacity=<rgba 0 0 0 1>
      shadowOffset=<0 0>
      strikeBrush {
       bIsDynamicallyLoaded=false
       drawAs=Image
       imageType=NoImage
       mirroring=NoMirror
       resourceName=None
       resourceObject=null
       tiling=NoTile
       imageSize=<32 32>
       margin=<ltrb 0 0 0 0>
       outlineSettings {
        bUseBrushTransparency=false
        roundingType=HalfHeightRadius
        width=0
        color {colorUseRule=UseColor_Specified, specifiedColor=<rgba 0 0 0 0>}
        cornerRadii=<0 0 0 0>
       }
       tintColor {colorUseRule=UseColor_Specified, specifiedColor=<rgba 1 1 1 1>}
       uVRegion {bIsValid=false, max=<0 0>, min=<0 0>}
      }
      underlineBrush {
       bIsDynamicallyLoaded=false
       drawAs=Image
       imageType=NoImage
       mirroring=NoMirror
       resourceName=None
       resourceObject=null
       tiling=NoTile
       imageSize=<32 32>
       margin=<ltrb 0 0 0 0>
       outlineSettings {
        bUseBrushTransparency=false
        roundingType=HalfHeightRadius
        width=0
        color {colorUseRule=UseColor_Specified, specifiedColor=<rgba 0 0 0 0>}
        cornerRadii=<0 0 0 0>
       }
       tintColor {colorUseRule=UseColor_Specified, specifiedColor=<rgba 1 1 1 1>}
       uVRegion {bIsValid=false, max=<0 0>, min=<0 0>}
      }
     }
     vScrollBarPadding=<ltrb 0 0 0 0>
    }
   }
   slot {class=HorizontalBoxSlot, properties {size {sizeRule=Fill, value=1}}}
  }
  HorizontalBox {
   widgetClassPath=&HorizontalBox
   parent=VerticalBox_78
   widget=HorizontalBox
   slot {class=VerticalBoxSlot, properties {padding=<ltrb 2 2 2 4>, size {sizeRule=Fill, value=1}}}
  }
  BTN_GetOpenFolder {
   widgetClassPath=&EditorUtilityButton
   parent=HorizontalBox
   widget=BTN_GetOpenFolder
   properties {
    toolTipText=Loads the open folder location as a valid package path.
    backgroundColor=<rgba 1 1 1 0.8>
    widgetStyle {
     clickedSlateSound {resourceObject=null}
     disabled {
      bIsDynamicallyLoaded=false
      drawAs=RoundedBox
      imageType=NoImage
      mirroring=NoMirror
      resourceName=None
      resourceObject=null
      tiling=NoTile
      imageSize=<0 0>
      margin=<ltrb 0 0 0 0>
      outlineSettings {
       bUseBrushTransparency=false
       roundingType=FixedRadius
       width=1
       color {colorUseRule=UseColor_ColorTable, specifiedColor=<rgba 0 0 0 0>}
       cornerRadii=<4 4 4 4>
      }
      tintColor {colorUseRule=UseColor_Specified, specifiedColor=<rgba 0.0395 0.0395 0.0395 1>}
      uVRegion {bIsValid=false, max=<0 0>, min=<0 0>}
     }
     disabledForeground {colorUseRule=UseColor_ColorTable, specifiedColor=<rgba 0 0 0 0>}
     hovered {
      bIsDynamicallyLoaded=false
      drawAs=RoundedBox
      imageType=NoImage
      mirroring=NoMirror
      resourceName=None
      resourceObject=null
      tiling=NoTile
      imageSize=<32 32>
      margin=<ltrb 0 0 0 0>
      outlineSettings {
       bUseBrushTransparency=false
       roundingType=FixedRadius
       width=1
       color {colorUseRule=UseColor_Specified, specifiedColor=<rgba 0.2083 0.2083 0.2083 1>}
       cornerRadii=<4 4 4 4>
      }
      tintColor {colorUseRule=UseColor_Specified, specifiedColor=<rgba 0.03 0.03 0.03 1>}
      uVRegion {bIsValid=false, max=<0 0>, min=<0 0>}
     }
     hoveredForeground {colorUseRule=UseColor_ColorTable, specifiedColor=<rgba 0 0 0 0>}
     hoveredSlateSound {resourceObject=null}
     normal {
      bIsDynamicallyLoaded=false
      drawAs=RoundedBox
      imageType=NoImage
      mirroring=NoMirror
      resourceName=None
      resourceObject=null
      tiling=NoTile
      imageSize=<32 32>
      margin=<ltrb 0 0 0 0>
      outlineSettings {
       bUseBrushTransparency=false
       roundingType=FixedRadius
       width=0.5
       color {colorUseRule=UseColor_Specified, specifiedColor=<rgba 0.0573 0.0573 0.0573 1>}
       cornerRadii=<2 2 2 2>
      }
      tintColor {colorUseRule=UseColor_Specified, specifiedColor=<rgba 0.0075 0.0075 0.0075 1>}
      uVRegion {bIsValid=false, max=<0 0>, min=<0 0>}
     }
     normalForeground {colorUseRule=UseColor_ColorTable, specifiedColor=<rgba 0 0 0 0>}
     normalPadding=<ltrb 4 0.7 4 0.7>
     pressed {
      bIsDynamicallyLoaded=false
      drawAs=RoundedBox
      imageType=NoImage
      mirroring=NoMirror
      resourceName=None
      resourceObject=null
      tiling=NoTile
      imageSize=<32 32>
      margin=<ltrb 0 0 0 0>
      outlineSettings {
       bUseBrushTransparency=false
       roundingType=FixedRadius
       width=1
       color {colorUseRule=UseColor_Specified, specifiedColor=<rgba 0.1198 0.0597 0.0168 1>}
       cornerRadii=<4 4 4 4>
      }
      tintColor {colorUseRule=UseColor_Specified, specifiedColor=<rgba 0.0365 0.0308 0.0258 1>}
      uVRegion {bIsValid=false, max=<0 0>, min=<0 0>}
     }
     pressedForeground {colorUseRule=UseColor_ColorTable, specifiedColor=<rgba 0 0 0 0>}
     pressedPadding=<ltrb 5.2 1 5 0.7>
     pressedSlateSound {resourceObject=null}
    }
   }
   slot {class=HorizontalBoxSlot, properties {horizontalAlignment=HAlign_Left, verticalAlignment=VAlign_Center}}
  }
  TextBlock_12 {
   widgetClassPath=&TextBlock
   parent=BTN_GetOpenFolder
   widget=TextBlock_12
   properties {
    justification=Center
    text=Path
    font {
     bForceMonospaced=false
     bMaterialIsStencil=false
     fontMaterial=null
     fontObject=&Roboto
     letterSpacing=0
     monospacedWidth=1
     size=9
     skewAmount=0
     typefaceFontName=Bold
     outlineSettings {
      bApplyOutlineToDropShadows=false
      bMiteredCorners=false
      bSeparateFillAlpha=false
      outlineMaterial=null
      outlineSize=0
      outlineColor=<rgba 0 0 0 1>
     }
    }
   }
   slot {class=ButtonSlot, properties {padding=<ltrb 0 2 0 2>}}
  }
  TB_BookCaseFolderPath {
   widgetClassPath=&EditorUtilityEditableTextBox
   bIsVariable=true
   parent=HorizontalBox
   widget=TB_BookCaseFolderPath
   properties {
    hintText=New BookShelf Folder Path
    toolTipText=The location of the Data Asset that will be created when clicking the button below.
    widgetStyle {
     backgroundColor {colorUseRule=UseColor_Specified, specifiedColor=<rgba 1 1 1 1>}
     backgroundImageFocused {
      bIsDynamicallyLoaded=false
      drawAs=RoundedBox
      imageType=NoImage
      mirroring=NoMirror
      resourceName=None
      resourceObject=null
      tiling=NoTile
      imageSize=<0 0>
      margin=<ltrb 0 0 0 0>
      outlineSettings {
       bUseBrushTransparency=false
       roundingType=FixedRadius
       width=1
       color {colorUseRule=UseColor_ColorTable, specifiedColor=<rgba 0 0 0 0>}
       cornerRadii=<4 4 4 4>
      }
      tintColor {colorUseRule=UseColor_Specified, specifiedColor=<rgba 0.0048 0.0048 0.0048 1>}
      uVRegion {bIsValid=false, max=<0 0>, min=<0 0>}
     }
     backgroundImageHovered {
      bIsDynamicallyLoaded=false
      drawAs=RoundedBox
      imageType=NoImage
      mirroring=NoMirror
      resourceName=None
      resourceObject=null
      tiling=NoTile
      imageSize=<0 0>
      margin=<ltrb 0 0 0 0>
      outlineSettings {
       bUseBrushTransparency=false
       roundingType=FixedRadius
       width=1
       color {colorUseRule=UseColor_ColorTable, specifiedColor=<rgba 0 0 0 0>}
       cornerRadii=<4 4 4 4>
      }
      tintColor {colorUseRule=UseColor_Specified, specifiedColor=<rgba 0.0048 0.0048 0.0048 1>}
      uVRegion {bIsValid=false, max=<0 0>, min=<0 0>}
     }
     backgroundImageNormal {
      bIsDynamicallyLoaded=false
      drawAs=RoundedBox
      imageType=NoImage
      mirroring=NoMirror
      resourceName=None
      resourceObject=null
      tiling=NoTile
      imageSize=<32 32>
      margin=<ltrb 0 0 0 0>
      outlineSettings {
       bUseBrushTransparency=false
       roundingType=FixedRadius
       width=0.5
       color {colorUseRule=UseColor_Specified, specifiedColor=<rgba 0.0573 0.0573 0.0573 1>}
       cornerRadii=<2 2 2 2>
      }
      tintColor {colorUseRule=UseColor_Specified, specifiedColor=<rgba 0.0075 0.0075 0.0075 1>}
      uVRegion {bIsValid=false, max=<0 0>, min=<0 0>}
     }
     backgroundImageReadOnly {
      bIsDynamicallyLoaded=false
      drawAs=RoundedBox
      imageType=NoImage
      mirroring=NoMirror
      resourceName=None
      resourceObject=null
      tiling=NoTile
      imageSize=<0 0>
      margin=<ltrb 0 0 0 0>
      outlineSettings {
       bUseBrushTransparency=false
       roundingType=FixedRadius
       width=1
       color {colorUseRule=UseColor_ColorTable, specifiedColor=<rgba 0 0 0 0>}
       cornerRadii=<4 4 4 4>
      }
      tintColor {colorUseRule=UseColor_Specified, specifiedColor=<rgba 0.0284 0.0284 0.0284 1>}
      uVRegion {bIsValid=false, max=<0 0>, min=<0 0>}
     }
     focusedForegroundColor {colorUseRule=UseColor_Specified, specifiedColor=<rgba 1 1 1 1>}
     foregroundColor {colorUseRule=UseColor_Specified, specifiedColor=<rgba 0.5271 0.5271 0.5271 1>}
     hScrollBarPadding=<ltrb 0 0 0 0>
     padding=<ltrb 4 1 4 1>
     readOnlyForegroundColor {colorUseRule=UseColor_Foreground, specifiedColor=<rgba 1 0 1 1>}
     scrollBarStyle {
      thickness=8
      draggedThumbImage {
       bIsDynamicallyLoaded=false
       drawAs=RoundedBox
       imageType=NoImage
       mirroring=NoMirror
       resourceName=None
       resourceObject=null
       tiling=NoTile
       imageSize=<0 0>
       margin=<ltrb 0 0 0 0>
       outlineSettings {
        bUseBrushTransparency=false
        roundingType=FixedRadius
        width=0
        color {colorUseRule=UseColor_Specified, specifiedColor=<rgba 0 0 0 0>}
        cornerRadii=<4 4 4 4>
       }
       tintColor {colorUseRule=UseColor_Specified, specifiedColor=<rgba 0.2159 0.2159 0.2159 1>}
       uVRegion {bIsValid=false, max=<0 0>, min=<0 0>}
      }
      horizontalBackgroundImage {
       bIsDynamicallyLoaded=false
       drawAs=NoDrawType
       imageType=NoImage
       mirroring=NoMirror
       resourceName=None
       resourceObject=null
       tiling=NoTile
       imageSize=<0 0>
       margin=<ltrb 0 0 0 0>
       outlineSettings {
        bUseBrushTransparency=false
        roundingType=HalfHeightRadius
        width=0
        color {colorUseRule=UseColor_Specified, specifiedColor=<rgba 0 0 0 0>}
        cornerRadii=<0 0 0 0>
       }
       tintColor {colorUseRule=UseColor_Specified, specifiedColor=<rgba 1 1 1 1>}
       uVRegion {bIsValid=false, max=<0 0>, min=<0 0>}
      }
      horizontalBottomSlotImage {
       bIsDynamicallyLoaded=false
       drawAs=NoDrawType
       imageType=NoImage
       mirroring=NoMirror
       resourceName=None
       resourceObject=null
       tiling=NoTile
       imageSize=<0 0>
       margin=<ltrb 0 0 0 0>
       outlineSettings {
        bUseBrushTransparency=false
        roundingType=HalfHeightRadius
        width=0
        color {colorUseRule=UseColor_Specified, specifiedColor=<rgba 0 0 0 0>}
        cornerRadii=<0 0 0 0>
       }
       tintColor {colorUseRule=UseColor_Specified, specifiedColor=<rgba 1 1 1 1>}
       uVRegion {bIsValid=false, max=<0 0>, min=<0 0>}
      }
      horizontalTopSlotImage {
       bIsDynamicallyLoaded=false
       drawAs=NoDrawType
       imageType=NoImage
       mirroring=NoMirror
       resourceName=None
       resourceObject=null
       tiling=NoTile
       imageSize=<0 0>
       margin=<ltrb 0 0 0 0>
       outlineSettings {
        bUseBrushTransparency=false
        roundingType=HalfHeightRadius
        width=0
        color {colorUseRule=UseColor_Specified, specifiedColor=<rgba 0 0 0 0>}
        cornerRadii=<0 0 0 0>
       }
       tintColor {colorUseRule=UseColor_Specified, specifiedColor=<rgba 1 1 1 1>}
       uVRegion {bIsValid=false, max=<0 0>, min=<0 0>}
      }
      hoveredThumbImage {
       bIsDynamicallyLoaded=false
       drawAs=RoundedBox
       imageType=NoImage
       mirroring=NoMirror
       resourceName=None
       resourceObject=null
       tiling=NoTile
       imageSize=<0 0>
       margin=<ltrb 0 0 0 0>
       outlineSettings {
        bUseBrushTransparency=false
        roundingType=FixedRadius
        width=0
        color {colorUseRule=UseColor_Specified, specifiedColor=<rgba 0 0 0 0>}
        cornerRadii=<4 4 4 4>
       }
       tintColor {colorUseRule=UseColor_Specified, specifiedColor=<rgba 0.2159 0.2159 0.2159 1>}
       uVRegion {bIsValid=false, max=<0 0>, min=<0 0>}
      }
      normalThumbImage {
       bIsDynamicallyLoaded=false
       drawAs=RoundedBox
       imageType=NoImage
       mirroring=NoMirror
       resourceName=None
       resourceObject=null
       tiling=NoTile
       imageSize=<0 0>
       margin=<ltrb 0 0 0 0>
       outlineSettings {
        bUseBrushTransparency=false
        roundingType=FixedRadius
        width=0
        color {colorUseRule=UseColor_Specified, specifiedColor=<rgba 0 0 0 0>}
        cornerRadii=<4 4 4 4>
       }
       tintColor {colorUseRule=UseColor_Specified, specifiedColor=<rgba 0.0953 0.0953 0.0953 1>}
       uVRegion {bIsValid=false, max=<0 0>, min=<0 0>}
      }
      verticalBackgroundImage {
       bIsDynamicallyLoaded=false
       drawAs=NoDrawType
       imageType=NoImage
       mirroring=NoMirror
       resourceName=None
       resourceObject=null
       tiling=NoTile
       imageSize=<0 0>
       margin=<ltrb 0 0 0 0>
       outlineSettings {
        bUseBrushTransparency=false
        roundingType=HalfHeightRadius
        width=0
        color {colorUseRule=UseColor_Specified, specifiedColor=<rgba 0 0 0 0>}
        cornerRadii=<0 0 0 0>
       }
       tintColor {colorUseRule=UseColor_Specified, specifiedColor=<rgba 1 1 1 1>}
       uVRegion {bIsValid=false, max=<0 0>, min=<0 0>}
      }
      verticalBottomSlotImage {
       bIsDynamicallyLoaded=false
       drawAs=NoDrawType
       imageType=NoImage
       mirroring=NoMirror
       resourceName=None
       resourceObject=null
       tiling=NoTile
       imageSize=<0 0>
       margin=<ltrb 0 0 0 0>
       outlineSettings {
        bUseBrushTransparency=false
        roundingType=HalfHeightRadius
        width=0
        color {colorUseRule=UseColor_Specified, specifiedColor=<rgba 0 0 0 0>}
        cornerRadii=<0 0 0 0>
       }
       tintColor {colorUseRule=UseColor_Specified, specifiedColor=<rgba 1 1 1 1>}
       uVRegion {bIsValid=false, max=<0 0>, min=<0 0>}
      }
      verticalTopSlotImage {
       bIsDynamicallyLoaded=false
       drawAs=NoDrawType
       imageType=NoImage
       mirroring=NoMirror
       resourceName=None
       resourceObject=null
       tiling=NoTile
       imageSize=<0 0>
       margin=<ltrb 0 0 0 0>
       outlineSettings {
        bUseBrushTransparency=false
        roundingType=HalfHeightRadius
        width=0
        color {colorUseRule=UseColor_Specified, specifiedColor=<rgba 0 0 0 0>}
        cornerRadii=<0 0 0 0>
       }
       tintColor {colorUseRule=UseColor_Specified, specifiedColor=<rgba 1 1 1 1>}
       uVRegion {bIsValid=false, max=<0 0>, min=<0 0>}
      }
     }
     textStyle {
      overflowPolicy=Clip
      transformPolicy=None
      colorAndOpacity {colorUseRule=UseColor_Foreground, specifiedColor=<rgba 1 0 1 1>}
      font {
       bForceMonospaced=false
       bMaterialIsStencil=false
       fontMaterial=null
       fontObject=&Roboto
       letterSpacing=0
       monospacedWidth=1
       size=7.5
       skewAmount=0
       typefaceFontName=Regular
       outlineSettings {
        bApplyOutlineToDropShadows=false
        bMiteredCorners=false
        bSeparateFillAlpha=false
        outlineMaterial=null
        outlineSize=0
        outlineColor=<rgba 0 0 0 1>
       }
      }
      highlightColor {colorUseRule=UseColor_ColorTable, specifiedColor=<rgba 0 0 0 0>}
      highlightShape {
       bIsDynamicallyLoaded=false
       drawAs=Image
       imageType=NoImage
       mirroring=NoMirror
       resourceName=None
       resourceObject=null
       tiling=NoTile
       imageSize=<0 0>
       margin=<ltrb 0 0 0 0>
       outlineSettings {
        bUseBrushTransparency=false
        roundingType=HalfHeightRadius
        width=0
        color {colorUseRule=UseColor_Specified, specifiedColor=<rgba 0 0 0 0>}
        cornerRadii=<0 0 0 0>
       }
       tintColor {colorUseRule=UseColor_ColorTable, specifiedColor=<rgba 0 0 0 0>}
       uVRegion {bIsValid=false, max=<0 0>, min=<0 0>}
      }
      selectedBackgroundColor {colorUseRule=UseColor_ColorTable, specifiedColor=<rgba 0 0 0 0>}
      shadowColorAndOpacity=<rgba 0 0 0 1>
      shadowOffset=<0 0>
      strikeBrush {
       bIsDynamicallyLoaded=false
       drawAs=Image
       imageType=NoImage
       mirroring=NoMirror
       resourceName=None
       resourceObject=null
       tiling=NoTile
       imageSize=<32 32>
       margin=<ltrb 0 0 0 0>
       outlineSettings {
        bUseBrushTransparency=false
        roundingType=HalfHeightRadius
        width=0
        color {colorUseRule=UseColor_Specified, specifiedColor=<rgba 0 0 0 0>}
        cornerRadii=<0 0 0 0>
       }
       tintColor {colorUseRule=UseColor_Specified, specifiedColor=<rgba 1 1 1 1>}
       uVRegion {bIsValid=false, max=<0 0>, min=<0 0>}
      }
      underlineBrush {
       bIsDynamicallyLoaded=false
       drawAs=Image
       imageType=NoImage
       mirroring=NoMirror
       resourceName=None
       resourceObject=null
       tiling=NoTile
       imageSize=<32 32>
       margin=<ltrb 0 0 0 0>
       outlineSettings {
        bUseBrushTransparency=false
        roundingType=HalfHeightRadius
        width=0
        color {colorUseRule=UseColor_Specified, specifiedColor=<rgba 0 0 0 0>}
        cornerRadii=<0 0 0 0>
       }
       tintColor {colorUseRule=UseColor_Specified, specifiedColor=<rgba 1 1 1 1>}
       uVRegion {bIsValid=false, max=<0 0>, min=<0 0>}
      }
     }
     vScrollBarPadding=<ltrb 0 0 0 0>
    }
   }
   slot {class=HorizontalBoxSlot, properties {size {sizeRule=Fill, value=1}}}
  }
  Border_6 {
   widgetClassPath=&Border
   parent=VerticalBox_78
   widget=Border_6
   properties {
    horizontalAlignment=HAlign_Center
    background {
     bIsDynamicallyLoaded=false
     drawAs=RoundedBox
     imageType=NoImage
     mirroring=NoMirror
     resourceName=None
     resourceObject=null
     tiling=NoTile
     imageSize=<32 32>
     margin=<ltrb 0 0 0 0>
     outlineSettings {
      bUseBrushTransparency=false
      roundingType=FixedRadius
      width=0
      color {colorUseRule=UseColor_Specified, specifiedColor=<rgba 0 0 0 0>}
      cornerRadii=<0 0 3 3>
     }
     tintColor {colorUseRule=UseColor_Specified, specifiedColor=<rgba 1 1 1 1>}
     uVRegion {bIsValid=false, max=<0 0>, min=<0 0>}
    }
    brushColor=<rgba 0.007 0.007 0.007 1>
    padding=<ltrb 0 2 0 2>
   }
   slot {class=VerticalBoxSlot, properties {padding=<ltrb 0 0 0 1>}}
  }
  BN_SaveNewBookCase {
   widgetClassPath=&EditorUtilityButton
   parent=Border_6
   widget=BN_SaveNewBookCase
   properties {
    toolTipText=Creates or updates a Data Asset from your loaded Bookmark Presets using the information above.
    widgetStyle {
     clickedSlateSound {resourceObject=null}
     disabled {
      bIsDynamicallyLoaded=false
      drawAs=RoundedBox
      imageType=NoImage
      mirroring=NoMirror
      resourceName=None
      resourceObject=null
      tiling=NoTile
      imageSize=<0 0>
      margin=<ltrb 0 0 0 0>
      outlineSettings {
       bUseBrushTransparency=false
       roundingType=FixedRadius
       width=1
       color {colorUseRule=UseColor_ColorTable, specifiedColor=<rgba 0 0 0 0>}
       cornerRadii=<4 4 4 4>
      }
      tintColor {colorUseRule=UseColor_Specified, specifiedColor=<rgba 0.0395 0.0395 0.0395 1>}
      uVRegion {bIsValid=false, max=<0 0>, min=<0 0>}
     }
     disabledForeground {colorUseRule=UseColor_ColorTable, specifiedColor=<rgba 0 0 0 0>}
     hovered {
      bIsDynamicallyLoaded=false
      drawAs=RoundedBox
      imageType=NoImage
      mirroring=NoMirror
      resourceName=None
      resourceObject=null
      tiling=NoTile
      imageSize=<32 32>
      margin=<ltrb 0 0 0 0>
      outlineSettings {
       bUseBrushTransparency=false
       roundingType=FixedRadius
       width=1
       color {colorUseRule=UseColor_Specified, specifiedColor=<rgba 0.2083 0.2083 0.2083 1>}
       cornerRadii=<4 4 4 4>
      }
      tintColor {colorUseRule=UseColor_Specified, specifiedColor=<rgba 0.03 0.03 0.03 1>}
      uVRegion {bIsValid=false, max=<0 0>, min=<0 0>}
     }
     hoveredForeground {colorUseRule=UseColor_ColorTable, specifiedColor=<rgba 0 0 0 0>}
     hoveredSlateSound {resourceObject=null}
     normal {
      bIsDynamicallyLoaded=false
      drawAs=RoundedBox
      imageType=NoImage
      mirroring=NoMirror
      resourceName=None
      resourceObject=null
      tiling=NoTile
      imageSize=<32 32>
      margin=<ltrb 0 0 0 0>
      outlineSettings {
       bUseBrushTransparency=false
       roundingType=FixedRadius
       width=0.5
       color {colorUseRule=UseColor_Specified, specifiedColor=<rgba 0.0573 0.0573 0.0573 1>}
       cornerRadii=<2 2 2 2>
      }
      tintColor {colorUseRule=UseColor_Specified, specifiedColor=<rgba 0.0075 0.0075 0.0075 1>}
      uVRegion {bIsValid=false, max=<0 0>, min=<0 0>}
     }
     normalForeground {colorUseRule=UseColor_ColorTable, specifiedColor=<rgba 0 0 0 0>}
     normalPadding=<ltrb 2 0.5 2 0.5>
     pressed {
      bIsDynamicallyLoaded=false
      drawAs=RoundedBox
      imageType=NoImage
      mirroring=NoMirror
      resourceName=None
      resourceObject=null
      tiling=NoTile
      imageSize=<32 32>
      margin=<ltrb 0 0 0 0>
      outlineSettings {
       bUseBrushTransparency=false
       roundingType=FixedRadius
       width=1
       color {colorUseRule=UseColor_Specified, specifiedColor=<rgba 0.1198 0.0597 0.0168 1>}
       cornerRadii=<4 4 4 4>
      }
      tintColor {colorUseRule=UseColor_Specified, specifiedColor=<rgba 0.0365 0.0308 0.0258 1>}
      uVRegion {bIsValid=false, max=<0 0>, min=<0 0>}
     }
     pressedForeground {colorUseRule=UseColor_ColorTable, specifiedColor=<rgba 0 0 0 0>}
     pressedPadding=<ltrb 3 1 3 0.75>
     pressedSlateSound {resourceObject=null}
    }
   }
   slot {class=BorderSlot, properties {horizontalAlignment=HAlign_Center, padding=<ltrb 0 2 0 2>}}
  }
  TextBlock_10 {
   widgetClassPath=&TextBlock
   parent=BN_SaveNewBookCase
   widget=TextBlock_10
   properties {
    clipping=ClipToBounds
    justification=Center
    text=Create DataAsset From Presets
    font {
     bForceMonospaced=false
     bMaterialIsStencil=false
     fontMaterial=null
     fontObject=&Roboto
     letterSpacing=0
     monospacedWidth=1
     size=9
     skewAmount=0
     typefaceFontName=Bold
     outlineSettings {
      bApplyOutlineToDropShadows=false
      bMiteredCorners=false
      bSeparateFillAlpha=false
      outlineMaterial=null
      outlineSize=0
      outlineColor=<rgba 0 0 0 1>
     }
    }
   }
   slot {class=ButtonSlot, properties {padding=<ltrb 4 1 4 1>}}
  }
  BTN_ClearSaveData {
   widgetClassPath=&EditorUtilityButton
   parent=VerticalBox_1
   widget=BTN_ClearSaveData
   properties {
    toolTipText=Resets all save data within this tool.
    widgetStyle {
     clickedSlateSound {resourceObject=null}
     disabled {
      bIsDynamicallyLoaded=false
      drawAs=RoundedBox
      imageType=NoImage
      mirroring=NoMirror
      resourceName=None
      resourceObject=null
      tiling=NoTile
      imageSize=<0 0>
      margin=<ltrb 0 0 0 0>
      outlineSettings {
       bUseBrushTransparency=false
       roundingType=FixedRadius
       width=1
       color {colorUseRule=UseColor_ColorTable, specifiedColor=<rgba 0 0 0 0>}
       cornerRadii=<4 4 4 4>
      }
      tintColor {colorUseRule=UseColor_Specified, specifiedColor=<rgba 0.0395 0.0395 0.0395 1>}
      uVRegion {bIsValid=false, max=<0 0>, min=<0 0>}
     }
     disabledForeground {colorUseRule=UseColor_ColorTable, specifiedColor=<rgba 0 0 0 0>}
     hovered {
      bIsDynamicallyLoaded=false
      drawAs=RoundedBox
      imageType=NoImage
      mirroring=NoMirror
      resourceName=None
      resourceObject=null
      tiling=NoTile
      imageSize=<32 32>
      margin=<ltrb 0 0 0 0>
      outlineSettings {
       bUseBrushTransparency=false
       roundingType=FixedRadius
       width=1
       color {colorUseRule=UseColor_Specified, specifiedColor=<rgba 0.2083 0.2083 0.2083 1>}
       cornerRadii=<4 4 4 4>
      }
      tintColor {colorUseRule=UseColor_Specified, specifiedColor=<rgba 0.03 0.03 0.03 1>}
      uVRegion {bIsValid=false, max=<0 0>, min=<0 0>}
     }
     hoveredForeground {colorUseRule=UseColor_ColorTable, specifiedColor=<rgba 0 0 0 0>}
     hoveredSlateSound {resourceObject=null}
     normal {
      bIsDynamicallyLoaded=false
      drawAs=RoundedBox
      imageType=NoImage
      mirroring=NoMirror
      resourceName=None
      resourceObject=null
      tiling=NoTile
      imageSize=<32 32>
      margin=<ltrb 0 0 0 0>
      outlineSettings {
       bUseBrushTransparency=false
       roundingType=FixedRadius
       width=0.5
       color {colorUseRule=UseColor_Specified, specifiedColor=<rgba 0.0573 0.0573 0.0573 1>}
       cornerRadii=<2 2 2 2>
      }
      tintColor {colorUseRule=UseColor_Specified, specifiedColor=<rgba 0.0075 0.0075 0.0075 1>}
      uVRegion {bIsValid=false, max=<0 0>, min=<0 0>}
     }
     normalForeground {colorUseRule=UseColor_ColorTable, specifiedColor=<rgba 0 0 0 0>}
     normalPadding=<ltrb 4 0.7 4 0.7>
     pressed {
      bIsDynamicallyLoaded=false
      drawAs=RoundedBox
      imageType=NoImage
      mirroring=NoMirror
      resourceName=None
      resourceObject=null
      tiling=NoTile
      imageSize=<32 32>
      margin=<ltrb 0 0 0 0>
      outlineSettings {
       bUseBrushTransparency=false
       roundingType=FixedRadius
       width=1
       color {colorUseRule=UseColor_Specified, specifiedColor=<rgba 0.1198 0.0597 0.0168 1>}
       cornerRadii=<4 4 4 4>
      }
      tintColor {colorUseRule=UseColor_Specified, specifiedColor=<rgba 0.0365 0.0308 0.0258 1>}
      uVRegion {bIsValid=false, max=<0 0>, min=<0 0>}
     }
     pressedForeground {colorUseRule=UseColor_ColorTable, specifiedColor=<rgba 0 0 0 0>}
     pressedPadding=<ltrb 5.2 1 5 0.7>
     pressedSlateSound {resourceObject=null}
    }
   }
   slot {class=VerticalBoxSlot, properties {padding=<ltrb 2 4 2 3>, size {sizeRule=Fill, value=1}}}
  }
  TextBlock_3 {
   widgetClassPath=&TextBlock
   parent=BTN_ClearSaveData
   widget=TextBlock_3
   properties {
    clipping=ClipToBounds
    justification=Center
    text=Reset Data
    textOverflowPolicy=Ellipsis
    font {
     bForceMonospaced=false
     bMaterialIsStencil=false
     fontMaterial=null
     fontObject=&Roboto
     letterSpacing=0
     monospacedWidth=1
     size=6.75
     skewAmount=0
     typefaceFontName=Bold
     outlineSettings {
      bApplyOutlineToDropShadows=false
      bMiteredCorners=false
      bSeparateFillAlpha=false
      outlineMaterial=null
      outlineSize=0
      outlineColor=<rgba 0 0 0 1>
     }
    }
   }
   slot {class=ButtonSlot, properties {padding=<ltrb 2 1 2 1>}}
  }
  BTN_WSLoadBookmarkConfig {
   widgetClassPath=&EditorUtilityButton
   parent=VerticalBox_1
   widget=BTN_WSLoadBookmarkConfig
   properties {
    toolTipText=View the bookmarks / presets editing menu
    widgetStyle {
     clickedSlateSound {resourceObject=null}
     disabled {
      bIsDynamicallyLoaded=false
      drawAs=RoundedBox
      imageType=NoImage
      mirroring=NoMirror
      resourceName=None
      resourceObject=null
      tiling=NoTile
      imageSize=<0 0>
      margin=<ltrb 0 0 0 0>
      outlineSettings {
       bUseBrushTransparency=false
       roundingType=FixedRadius
       width=1
       color {colorUseRule=UseColor_ColorTable, specifiedColor=<rgba 0 0 0 0>}
       cornerRadii=<4 4 4 4>
      }
      tintColor {colorUseRule=UseColor_Specified, specifiedColor=<rgba 0.0395 0.0395 0.0395 1>}
      uVRegion {bIsValid=false, max=<0 0>, min=<0 0>}
     }
     disabledForeground {colorUseRule=UseColor_ColorTable, specifiedColor=<rgba 0 0 0 0>}
     hovered {
      bIsDynamicallyLoaded=false
      drawAs=RoundedBox
      imageType=NoImage
      mirroring=NoMirror
      resourceName=None
      resourceObject=null
      tiling=NoTile
      imageSize=<32 32>
      margin=<ltrb 0 0 0 0>
      outlineSettings {
       bUseBrushTransparency=false
       roundingType=FixedRadius
       width=1
       color {colorUseRule=UseColor_Specified, specifiedColor=<rgba 0.2083 0.2083 0.2083 1>}
       cornerRadii=<4 4 4 4>
      }
      tintColor {colorUseRule=UseColor_Specified, specifiedColor=<rgba 0.04 0.04 0.04 1>}
      uVRegion {bIsValid=false, max=<0 0>, min=<0 0>}
     }
     hoveredForeground {colorUseRule=UseColor_ColorTable, specifiedColor=<rgba 0 0 0 0>}
     hoveredSlateSound {resourceObject=null}
     normal {
      bIsDynamicallyLoaded=false
      drawAs=RoundedBox
      imageType=NoImage
      mirroring=NoMirror
      resourceName=None
      resourceObject=null
      tiling=NoTile
      imageSize=<32 32>
      margin=<ltrb 0 0 0 0>
      outlineSettings {
       bUseBrushTransparency=false
       roundingType=FixedRadius
       width=0.5
       color {colorUseRule=UseColor_Specified, specifiedColor=<rgba 0.07 0.07 0.07 1>}
       cornerRadii=<2 2 2 2>
      }
      tintColor {colorUseRule=UseColor_Specified, specifiedColor=<rgba 0.02 0.02 0.02 1>}
      uVRegion {bIsValid=false, max=<0 0>, min=<0 0>}
     }
     normalForeground {colorUseRule=UseColor_ColorTable, specifiedColor=<rgba 0 0 0 0>}
     normalPadding=<ltrb 4 0.7 4 0.7>
     pressed {
      bIsDynamicallyLoaded=false
      drawAs=RoundedBox
      imageType=NoImage
      mirroring=NoMirror
      resourceName=None
      resourceObject=null
      tiling=NoTile
      imageSize=<32 32>
      margin=<ltrb 0 0 0 0>
      outlineSettings {
       bUseBrushTransparency=false
       roundingType=FixedRadius
       width=1
       color {colorUseRule=UseColor_Specified, specifiedColor=<rgba 0.1198 0.0597 0.0168 1>}
       cornerRadii=<4 4 4 4>
      }
      tintColor {colorUseRule=UseColor_Specified, specifiedColor=<rgba 0.0365 0.0308 0.0258 1>}
      uVRegion {bIsValid=false, max=<0 0>, min=<0 0>}
     }
     pressedForeground {colorUseRule=UseColor_ColorTable, specifiedColor=<rgba 0 0 0 0>}
     pressedPadding=<ltrb 5.2 1 5 0.7>
     pressedSlateSound {resourceObject=null}
    }
   }
   slot {class=VerticalBoxSlot, properties {padding=<ltrb 2 2 2 2>}}
  }
  TextBlock_9 {
   widgetClassPath=&TextBlock
   parent=BTN_WSLoadBookmarkConfig
   widget=TextBlock_9
   properties {
    clipping=ClipToBounds
    justification=Center
    text=Edit Bookmarks
    textOverflowPolicy=Ellipsis
    font {
     bForceMonospaced=false
     bMaterialIsStencil=false
     fontMaterial=null
     fontObject=&Roboto
     letterSpacing=0
     monospacedWidth=1
     size=6.75
     skewAmount=0
     typefaceFontName=Bold
     outlineSettings {
      bApplyOutlineToDropShadows=false
      bMiteredCorners=false
      bSeparateFillAlpha=false
      outlineMaterial=null
      outlineSize=0
      outlineColor=<rgba 0 0 0 1>
     }
    }
   }
   slot {class=ButtonSlot, properties {padding=<ltrb 2 1 2 1>}}
  }
 ]
}
